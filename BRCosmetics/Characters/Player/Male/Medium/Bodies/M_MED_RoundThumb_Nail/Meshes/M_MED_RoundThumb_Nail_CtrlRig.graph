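# Control Rig for the M_MED_RoundThumb_Nail player body that procedurally corrects accessory bones like shoulder pad, straps, pipes, jacket flaps, bag and knee pad.
# Nested Sequence nodes split the forward solve by body area.
# Spherical pose readers on upper arm, clavicle and thigh bones weight additive transform offsets, with per-limb debug-draw variables.
# Weighted parent constraints blend some accessories between bones, and one reader also writes the arm_up_r curve.
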ControlRigBlueprint /BRCosmetics/Characters/Player/Male/Medium/Bodies/M_MED_RoundThumb_Nail/Meshes/M_MED_RoundThumb_Nail_CtrlRig
graphs [
 ConstructionGraph {nodes [RigUnit_PrepareForExecution = &RigUnit_PrepareForExecution()]}
 RigVMFunctionLibrary {links [], nodes []}
 RigVMFunction_Sequence_2_1_ContainedGraph {nodes [
   n1 = RigVMFunctionEntryNode()
   n2 = &RigVMFunction_Sequence(ExecuteContext=@n1.ExecuteContext)
   n3 = &RigVMFunction_Sequence(ExecuteContext=@n2.B)
   RigVMFunctionReturnNode(A=@n2.A, B=@n3.A, C=@n3.B)
  ]}
 RigVMFunction_Sequence_3_1_ContainedGraph {nodes [
   n1 = RigVMFunctionEntryNode()
   n2 = &RigVMFunction_Sequence(ExecuteContext=@n1.ExecuteContext)
   n3 = &RigVMFunction_Sequence(ExecuteContext=@n2.B)
   n4 = &RigVMFunction_Sequence(ExecuteContext=@n3.B)
   RigVMFunctionReturnNode(A=@n2.A, B=@n3.A, C=@n4.A, D=@n4.B)
  ]}
 RigVMFunction_Sequence_3_1_ContainedGraph {nodes [
   n1 = RigVMFunctionEntryNode()
   n2 = &RigVMFunction_Sequence(ExecuteContext=@n1.ExecuteContext)
   n3 = &RigVMFunction_Sequence(ExecuteContext=@n2.B)
   n4 = &RigVMFunction_Sequence(ExecuteContext=@n3.B)
   RigVMFunctionReturnNode(A=@n2.A, B=@n3.A, C=@n4.A, D=@n4.B)
  ]}
 RigVMFunction_Sequence_3_1_ContainedGraph {nodes [
   n1 = RigVMFunctionEntryNode()
   n2 = &RigVMFunction_Sequence(ExecuteContext=@n1.ExecuteContext)
   n3 = &RigVMFunction_Sequence(ExecuteContext=@n2.B)
   RigVMFunctionReturnNode(A=@n2.A, B=@n3.A, C=@n3.B)
  ]}
 RigVMFunction_Sequence_4_1_ContainedGraph {nodes [
   n1 = RigVMFunctionEntryNode()
   n2 = &RigVMFunction_Sequence(ExecuteContext=@n1.ExecuteContext)
   n3 = &RigVMFunction_Sequence(ExecuteContext=@n2.B)
   RigVMFunctionReturnNode(A=@n2.A, B=@n3.A, C=@n3.B)
  ]}
 RigVMFunction_Sequence_5_1_ContainedGraph {nodes [
   n1 = RigVMFunctionEntryNode()
   n2 = &RigVMFunction_Sequence(ExecuteContext=@n1.ExecuteContext)
   n3 = &RigVMFunction_Sequence(ExecuteContext=@n2.B)
   RigVMFunctionReturnNode(A=@n2.A, B=@n3.A, C=@n3.B)
  ]}
 RigVMFunction_Sequence_5_2_ContainedGraph {nodes [
   n1 = RigVMFunctionEntryNode()
   n2 = &RigVMFunction_Sequence(ExecuteContext=@n1.ExecuteContext)
   n3 = &RigVMFunction_Sequence(ExecuteContext=@n2.B)
   RigVMFunctionReturnNode(A=@n2.A, B=@n3.A, C=@n3.B)
  ]}
 RigVMFunction_Sequence_7_1_ContainedGraph {nodes [
   n1 = RigVMFunctionEntryNode()
   n2 = &RigVMFunction_Sequence(ExecuteContext=@n1.ExecuteContext)
   n3 = &RigVMFunction_Sequence(ExecuteContext=@n2.B)
   RigVMFunctionReturnNode(A=@n2.A, B=@n3.A, C=@n3.B)
  ]}
 RigVMFunction_Sequence_8_1_ContainedGraph {nodes [
   n1 = RigVMFunctionEntryNode()
   n2 = &RigVMFunction_Sequence(ExecuteContext=@n1.ExecuteContext)
   n3 = &RigVMFunction_Sequence(ExecuteContext=@n2.B)
   n4 = &RigVMFunction_Sequence(ExecuteContext=@n3.B)
   RigVMFunctionReturnNode(A=@n2.A, B=@n3.A, C=@n4.A, D=@n4.B)
  ]}
 RigVMFunction_Sequence_9_1_ContainedGraph {nodes [
   n1 = RigVMFunctionEntryNode()
   n2 = &RigVMFunction_Sequence(ExecuteContext=@n1.ExecuteContext)
   n3 = &RigVMFunction_Sequence(ExecuteContext=@n2.B)
   RigVMFunctionReturnNode(A=@n2.A, B=@n3.A, C=@n3.B)
  ]}
 RigVMFunction_Sequence_9_ContainedGraph {nodes [
   n1 = RigVMFunctionEntryNode()
   RigVMFunction_Sequence = &RigVMFunction_Sequence(ExecuteContext=@n1.ExecuteContext)
   n2 = &RigVMFunction_Sequence(ExecuteContext=@RigVMFunction_Sequence.B)
   n3 = &RigVMFunction_Sequence(ExecuteContext=@n2.B)
   RigVMFunctionReturnNode(A=@RigVMFunction_Sequence.A, B=@n2.A, C=@n3.A, D=@n3.B)
  ]}
 RigVMModel {nodes [
   RigUnit_BeginExecution = &RigUnit_BeginExecution()
   n1 = &RigVMFunction_Sequence(ExecuteContext=@RigUnit_BeginExecution.ExecutePin)
   n2 = &RigVMFunction_Sequence(ExecuteContext=@n1.A)
   RigVMFunction_Sequence_5_2_2 = RigVMAggregateNode(ExecuteContext=@n2.A)
   n3 = &RigVMFunction_Sequence(ExecuteContext=@RigVMFunction_Sequence_5_2_2.A)
   &RigUnit_ParentConstraint(
    AdvancedSettings=(RotationOrderForFilter=XZY)
    Child=(Type=Bone,Name="extra_shoulder_strap_09")
    Parents="((Item=(Type=Bone,Name=\"upperarm_r\"),Weight=0.0071),(Item=(Type=Bone,Name=\"deform_upperarm_bk_r\"),Weight=0.0113),(Item=(Type=Bone,Name=\"deform_knee_in_r\"),Weight=0.0267),(Item=(Type=Bone,Name=\"upperarm_twist_01_r\"),Weight=0.027),(Item=(Type=Bone,Name=\"spin ...(truncated)"
    Weight=1
    ExecutePin=@n3.A
   )
   > &RigUnit_ParentConstraint(
    AdvancedSettings=(RotationOrderForFilter=XZY)
    Child=(Type=Bone,Name="extra_shoulder_strap_08")
    Parents="((Item=(Type=Bone,Name=\"upperarm_twist_01_r\"),Weight=0.0033),(Item=(Type=Bone,Name=\"spine_02\"),Weight=0.0235),(Item=(Type=Bone,Name=\"clavicle_r\"),Weight=0.0298),(Item=(Type=Bone,Name=\"deform_upperarm_fr_r\"),Weight=0.0315),(Item=(Type=Bone,Name=\"spine_03\"), ...(truncated)"
    Weight=1
   )
   > &RigUnit_ModifyTransforms(
    ItemToModify=((Item=(Type=Bone,Name="extra_shoulder_strap_08"),Transform=(Rotation=(X=-0.0041,Y=-0.0052,Z=-0.2065,W=0.9784),Translation=(X=-4.1271,Y=1.41,Z=1.5324),Scale3D=(X=1,Y=1,Z=1))))
    Weight=1
   )
   > &RigUnit_ParentConstraint(
    AdvancedSettings=(RotationOrderForFilter=XZY)
    Child=(Type=Bone,Name="extra_shoulder_strap_02")
    Parents="((Item=(Type=Bone,Name=\"deform_pec_r\"),Weight=0.0392),(Item=(Type=Bone,Name=\"spine_04\"),Weight=0.0722),(Item=(Type=Bone,Name=\"deform_upperarm_up_r\"),Weight=0.1231),(Item=(Type=Bone,Name=\"spine_05\"),Weight=0.1295),(Item=(Type=Bone,Name=\"upperarm_twist_01_r\" ...(truncated)"
    Weight=1
   )
   > &RigUnit_ModifyTransforms(
    ItemToModify=((Item=(Type=Bone,Name="extra_shoulder_strap_02"),Transform=(Rotation=(X=-0.0041,Y=-0.0052,Z=-0.2065,W=0.9784),Translation=(X=-0.3689,Y=-0.0425,Z=-0.0835),Scale3D=(X=1,Y=1,Z=1))))
    Weight=1
   )
   n4 = &RigUnit_SphericalPoseReader(
    ActiveRegionScaleFactors=(PositiveWidth=1,NegativeWidth=1,PositiveHeight=1,NegativeHeight=1)
    ActiveRegionSize=0.05
    Debug=(bDrawDebug=true,bDraw2D=False,bDrawLocalAxes=False,DebugScale=25,DebugSegments=20,DebugThickness=0.25)
    DriverAxis=(X=-1,Y=0,Z=0)
    DriverItem=(Type=Bone,Name="upperarm_r")
    FalloffRegionScaleFactors=(PositiveWidth=1,NegativeWidth=1,PositiveHeight=2,NegativeHeight=1)
    FalloffSize=0.3
    OptionalParentItem=(Type=Bone)
    RotationOffset=(X=-120,Y=8.961,Z=11.4016)
    ExecutePin=@n3.B
    Debug.bDrawDebug=RigVMVariableNode() {variable=Degub_arm_r}.Value
   )
   > &RigUnit_ModifyTransforms(
    ItemToModify=((Item=(Type=Bone,Name="extra_shoulder_pad"),Transform=(Rotation=(X=0.2443,Y=-0.1432,Z=0.1637,W=0.945),Translation=(X=6.6783,Y=1.6979,Z=-0.2512),Scale3D=(X=1,Y=1,Z=1))))
    Weight=@n4.OutputParam
   )
   > &RigUnit_ModifyTransforms(
    ItemToModify=((Item=(Type=Bone,Name="extra_shoulder_strap_03"),Transform=(Rotation=(X=-0.3308,Y=0.1319,Z=-0.182,W=0.9165),Translation=(X=2.4206,Y=2.9268,Z=-1.264),Scale3D=(X=1,Y=1,Z=1))))
    Weight=@n4.OutputParam
   )
   > &RigUnit_ModifyTransforms(
    ItemToModify=((Item=(Type=Bone,Name="extra_shoulder_strap_02"),Transform=(Rotation=(X=-0.061,Y=-0.0174,Z=0.2552,W=0.9648),Translation=(X=0.7353,Y=-0.195,Z=0.9452),Scale3D=(X=1,Y=1,Z=1))))
    Weight=@n4.OutputParam
   )
   > &RigUnit_ModifyTransforms(
    ItemToModify=((Item=(Type=Bone,Name="extra_shoulder_strap_09"),Transform=(Rotation=(X=0,Y=0,Z=0,W=1),Translation=(X=0,Y=2,Z=0),Scale3D=(X=1,Y=1,Z=1))))
    Weight=@n4.OutputParam
   )
   RigVMFunction_Sequence_4_1 = RigVMAggregateNode(ExecuteContext=@RigVMFunction_Sequence_5_2_2.B)
   n5 = &RigUnit_SphericalPoseReader(
    ActiveRegionScaleFactors=(PositiveWidth=1,NegativeWidth=1,PositiveHeight=1,NegativeHeight=1)
    ActiveRegionSize=0.02
    Debug=(bDrawDebug=true,bDraw2D=False,bDrawLocalAxes=False,DebugScale=25,DebugSegments=20,DebugThickness=0.25)
    DriverAxis=(X=0,Y=0,Z=-1)
    DriverItem=(Type=Bone,Name="upperarm_r")
    FalloffRegionScaleFactors=(PositiveWidth=1,NegativeWidth=1,PositiveHeight=1,NegativeHeight=1)
    FalloffSize=0.2
    OptionalParentItem=(Type=Bone)
    RotationOffset=(X=120.1965,Y=15.951,Z=-12.5528)
    ExecutePin=@RigVMFunction_Sequence_4_1.A
    Debug.bDrawDebug=RigVMVariableNode() {variable=Degub_arm_r}.Value
   )
   > &RigUnit_ModifyTransforms(
    ItemToModify=((Item=(Type=Bone,Name="extra_shoulder_pad"),Transform=(Rotation=(X=0,Y=0.0967,Z=0,W=0.9953),Translation=(X=0,Y=0,Z=0),Scale3D=(X=1,Y=1,Z=1))))
    Weight=@n5.OutputParam
   )
   n6 = &RigUnit_SphericalPoseReader(
    ActiveRegionScaleFactors=(PositiveWidth=2,NegativeWidth=1,PositiveHeight=2,NegativeHeight=2)
    ActiveRegionSize=0.05
    Debug=(bDrawDebug=true,bDraw2D=False,bDrawLocalAxes=False,DebugScale=25,DebugSegments=20,DebugThickness=0.25)
    DriverAxis=(X=-1,Y=0,Z=0)
    DriverItem=(Type=Bone,Name="upperarm_r")
    FalloffRegionScaleFactors=(PositiveWidth=2,NegativeWidth=0.5,PositiveHeight=1,NegativeHeight=1)
    FalloffSize=0.2
    OptionalParentItem=(Type=Bone)
    RotationOffset=(X=-100.8064,Y=-80,Z=-70.5359)
    ExecutePin=@RigVMFunction_Sequence_4_1.B
    Debug.bDrawDebug=RigVMVariableNode() {variable=Degub_arm_r}.Value
   )
   > &RigUnit_ModifyTransforms(
    ItemToModify=((Item=(Type=Bone,Name="extra_shoulder_pad"),Transform=(Rotation=(X=0,Y=-0.0623,Z=0,W=0.9981),Translation=(X=-2.0632,Y=-1.094,Z=0.082),Scale3D=(X=1,Y=1,Z=1))))
    Weight=@n6.OutputParam
   )
   > &RigUnit_ModifyTransforms(
    ItemToModify=((Item=(Type=Bone,Name="extra_shoulder_strap_09"),Transform=(Rotation=(X=-0.0366,Y=0.071,Z=0.0371,W=0.9961),Translation=(X=-1.4206,Y=2.7375,Z=0.0435),Scale3D=(X=1,Y=1,Z=1))))
    Weight=@n6.OutputParam
   )
   n7 = &RigUnit_SphericalPoseReader(
    ActiveRegionScaleFactors=(PositiveWidth=2,NegativeWidth=2,PositiveHeight=2,NegativeHeight=1)
    ActiveRegionSize=0.05
    Debug=(bDrawDebug=true,bDraw2D=False,bDrawLocalAxes=False,DebugScale=25,DebugSegments=20,DebugThickness=0.25)
    DriverAxis=(X=-1,Y=0,Z=0)
    DriverItem=(Type=Bone,Name="upperarm_r")
    FalloffRegionScaleFactors=(PositiveWidth=1.5,NegativeWidth=1.5,PositiveHeight=3,NegativeHeight=1)
    FalloffSize=0.2
    OptionalParentItem=(Type=Bone)
    RotationOffset=(X=-160,Y=-1.9124,Z=-80)
    Debug.bDrawDebug=RigVMVariableNode() {variable=Degub_arm_r}.Value
    ExecutePin=@RigVMFunction_Sequence_4_1.C
   )
   > /Script/ControlRig.RigUnit_SetCurveValue(Curve=arm_up_r, Value=@n7.OutputParam) {is=struct RigVMUnitNode}
   > &RigUnit_ModifyTransforms(
    ItemToModify=((Item=(Type=Bone,Name="extra_shoulder_pad"),Transform=(Rotation=(X=0.0407,Y=-0.2846,Z=0.0913,W=0.9534),Translation=(X=1.0543,Y=-0.7846,Z=-3.5972),Scale3D=(X=1,Y=1,Z=1))))
    Weight=@n7.OutputParam
   )
   > &RigUnit_ModifyTransforms(
    ItemToModify=((Item=(Type=Bone,Name="extra_shoulder_strap_02"),Transform=(Rotation=(X=-0.061,Y=-0.0174,Z=0.2552,W=0.9648),Translation=(X=0.7353,Y=-0.195,Z=0.9452),Scale3D=(X=1,Y=1,Z=1))))
    Weight=@n7.OutputParam
   )
   > &RigUnit_ModifyTransforms(
    ItemToModify=((Item=(Type=Bone,Name="extra_shoulder_strap_09"),Transform=(Rotation=(X=0,Y=0,Z=0,W=1),Translation=(X=-1.0562,Y=2.0363,Z=0.3049),Scale3D=(X=1,Y=1,Z=1))))
    Weight=@n7.OutputParam
   )
   RigVMFunction_Sequence_2_1 = RigVMAggregateNode(ExecuteContext=@RigVMFunction_Sequence_5_2_2.C)
   n8 = &RigUnit_SphericalPoseReader(
    ActiveRegionScaleFactors=(PositiveWidth=1,NegativeWidth=1,PositiveHeight=1,NegativeHeight=1)
    ActiveRegionSize=0.03
    Debug=(bDrawDebug=true,bDraw2D=False,bDrawLocalAxes=False,DebugScale=25,DebugSegments=20,DebugThickness=0.25)
    DriverAxis=(X=-1,Y=0,Z=0)
    DriverItem=(Type=Bone,Name="upperarm_r")
    FalloffRegionScaleFactors=(PositiveWidth=1,NegativeWidth=1,PositiveHeight=1,NegativeHeight=1)
    FalloffSize=0.2
    OptionalParentItem=(Type=Bone)
    RotationOffset=(X=-27.0558,Y=-52.3519,Z=-70)
    Debug.bDrawDebug=RigVMVariableNode() {variable=Degub_arm_r}.Value
    ExecutePin=@RigVMFunction_Sequence_2_1.B
   )
   > &RigUnit_ModifyTransforms(
    ItemToModify=((Item=(Type=Bone,Name="extra_shoulder_pad"),Transform=(Rotation=(X=0,Y=0,Z=0,W=1),Translation=(X=0,Y=0.4001,Z=0),Scale3D=(X=1,Y=1,Z=1))))
    Weight=@n8.OutputParam
   )
   > &RigUnit_ModifyTransforms(
    ItemToModify=((Item=(Type=Bone,Name="extra_shoulder_strap_03"),Transform=(Rotation=(X=0.1305,Y=0,Z=0,W=0.9914),Translation=(X=0,Y=-0.3697,Z=0.2494),Scale3D=(X=1,Y=1,Z=1))))
    Weight=@n8.OutputParam
   )
   n9 = &RigUnit_SphericalPoseReader(
    ActiveRegionScaleFactors=(PositiveWidth=1,NegativeWidth=1,PositiveHeight=1,NegativeHeight=1)
    ActiveRegionSize=0.02
    Debug=(bDrawDebug=true,bDraw2D=False,bDrawLocalAxes=False,DebugScale=25,DebugSegments=20,DebugThickness=0.25)
    DriverAxis=(X=-1,Y=0,Z=0)
    DriverItem=(Type=Bone,Name="upperarm_r")
    FalloffRegionScaleFactors=(PositiveWidth=1,NegativeWidth=1,PositiveHeight=1,NegativeHeight=1)
    FalloffSize=0.3
    OptionalParentItem=(Type=Bone)
    RotationOffset=(X=-30.0493,Y=-10.3001,Z=-70)
    Debug.bDrawDebug=RigVMVariableNode() {variable=Degub_arm_r}.Value
    ExecutePin=@RigVMFunction_Sequence_2_1.A
   )
   > &RigUnit_ModifyTransforms(
    ItemToModify=((Item=(Type=Bone,Name="extra_shoulder_pad"),Transform=(Rotation=(X=0,Y=0,Z=0,W=1),Translation=(X=-1.9858,Y=-1.3544,Z=0),Scale3D=(X=1,Y=1,Z=1))))
    Weight=@n9.OutputParam
   )
   > &RigUnit_ModifyTransforms(
    ItemToModify=((Item=(Type=Bone,Name="extra_shoulder_strap_03"),Transform=(Rotation=(X=0.2164,Y=0,Z=0,W=0.9763),Translation=(X=0,Y=0.4371,Z=-0.9374),Scale3D=(X=1,Y=1,Z=1))))
    Weight=@n9.OutputParam
   )
   &RigUnit_ModifyTransforms(
    ItemToModify=((Item=(Type=Bone,Name="extra_shoulder_strap_02"),Transform=(Rotation=(X=0.0423,Y=0.0019,Z=0.2834,W=0.9581),Translation=(X=0,Y=0,Z=0),Scale3D=(X=1,Y=1,Z=1))))
    Weight=1
    ExecutePin=@RigVMFunction_Sequence_2_1.C
   )
   > &RigUnit_ParentConstraint(
    AdvancedSettings=(InterpolationType=Average,RotationOrderForFilter=XZY)
    Child=(Type=Bone,Name="extra_shoulder_strap_03")
    Filter=(TranslationFilter=(bX=false,bY=true,bZ=false),RotationFilter=(bX=false,bY=false,bZ=false),ScaleFilter=(bX=True,bY=True,bZ=True))
    Parents=((Item=(Type=Bone,Name="extra_shoulder_pad"),Weight=1))
    Weight=0.4
   )
   n10 = &RigVMFunction_MathDoubleMax(A=@n4.OutputParam, B=@n6.OutputParam)
   > &RigUnit_ParentConstraint(
    AdvancedSettings=(InterpolationType=Average,RotationOrderForFilter=XZY)
    Child=(Type=Bone,Name="extra_shoulder_strap_03")
    Filter=(TranslationFilter=(bX=true,bY=true,bZ=true),RotationFilter=(bX=true,bY=true,bZ=true),ScaleFilter=(bX=True,bY=True,bZ=True))
    Parents=((Item=(Type=Bone,Name="extra_shoulder_pad"),Weight=1))
    Weight=&RigVMFunction_MathDoubleRemap(
     SourceMaximum=1
     SourceMinimum=0
     TargetMaximum=0
     TargetMinimum=1
     bClamp=False
     Value=&RigVMFunction_MathDoubleMax(A=@n10.Result, B=@n7.OutputParam).Result
    ).Result
   )
   RigVMFunction_Sequence_3_1_1_2 = RigVMAggregateNode(ExecuteContext=@n2.B)
   RigVMFunction_Sequence_3_2_2 = RigVMAggregateNode(ExecuteContext=@RigVMFunction_Sequence_3_1_1_2.A)
   n11 = &RigUnit_SphericalPoseReader(
    ActiveRegionScaleFactors=(PositiveWidth=1,NegativeWidth=1,PositiveHeight=1,NegativeHeight=1)
    ActiveRegionSize=0.01
    Debug=(bDrawDebug=true,bDraw2D=False,bDrawLocalAxes=False,DebugScale=25,DebugSegments=20,DebugThickness=0.25)
    DriverAxis=(X=1,Y=0,Z=0)
    DriverItem=(Type=Bone,Name="clavicle_l")
    FalloffRegionScaleFactors=(PositiveWidth=1,NegativeWidth=1,PositiveHeight=1,NegativeHeight=1)
    FalloffSize=0.05
    OptionalParentItem=(Type=Bone)
    RotationOffset=(X=38,Y=3.0599,Z=-88.4832)
    Debug.bDrawDebug=RigVMVariableNode() {variable=Degub_arm_l}.Value
    ExecutePin=@RigVMFunction_Sequence_3_2_2.D
   )
   > &RigUnit_ModifyTransforms(
    ItemToModify=((Item=(Type=Bone,Name="extra_pipe_02"),Transform=(Rotation=(X=0,Y=0,Z=0,W=1),Translation=(X=0,Y=0.7533,Z=0),Scale3D=(X=1,Y=1,Z=1))))
    Weight=@n11.OutputParam
   )
   > &RigUnit_ModifyTransforms(
    ItemToModify=((Item=(Type=Bone,Name="extra_shoulder_strap_06"),Transform=(Rotation=(X=0,Y=0,Z=0,W=1),Translation=(X=-0.2849,Y=-0.0043,Z=-0.9669),Scale3D=(X=1,Y=1,Z=1))))
    ItemToModify.0.Transform.Rotation=&RigVMFunction_MathQuaternionFromEuler(Euler=(X=5,Y=0,Z=0)).Result
    Weight=@n11.OutputParam
   )
   > &RigUnit_ModifyTransforms(
    ItemToModify=((Item=(Type=Bone,Name="extra_shoulder_strap_07"),Transform=(Rotation=(X=0,Y=0,Z=0,W=1),Translation=(X=-0.5795,Y=-0.7928,Z=-0.5461),Scale3D=(X=1,Y=1,Z=1))))
    ItemToModify.0.Transform.Rotation=&RigVMFunction_MathQuaternionFromEuler(Euler=(X=15,Y=0,Z=0)).Result
    Weight=@n11.OutputParam
   )
   > &RigUnit_ModifyTransforms(
    ItemToModify=((Item=(Type=Bone,Name="extra_shoulder_strap_05"),Transform=(Rotation=(X=0,Y=0,Z=0,W=1),Translation=(X=0.2429,Y=-0.2881,Z=-0.2351),Scale3D=(X=1,Y=1,Z=1))))
    ItemToModify.0.Transform.Rotation=&RigVMFunction_MathQuaternionFromEuler(Euler=(X=10,Y=0,Z=0)).Result
    Weight=@n11.OutputParam
   )
   > &RigUnit_ModifyTransforms(
    ItemToModify=((Item=(Type=Bone,Name="extra_pipe_01"),Transform=(Rotation=(X=0,Y=0,Z=0,W=1),Translation=(X=-0.0827,Y=0.1554,Z=-0.0345),Scale3D=(X=1,Y=1,Z=1))))
    ItemToModify.0.Transform.Rotation=&RigVMFunction_MathQuaternionFromEuler(Euler=(X=-15,Y=0,Z=0)).Result
    Weight=@n11.OutputParam
   )
   n12 = &RigUnit_SphericalPoseReader(
    ActiveRegionScaleFactors=(PositiveWidth=1,NegativeWidth=1,PositiveHeight=1,NegativeHeight=1)
    ActiveRegionSize=0.01
    Debug=(bDrawDebug=true,bDraw2D=False,bDrawLocalAxes=False,DebugScale=25,DebugSegments=20,DebugThickness=0.25)
    DriverAxis=(X=1,Y=0,Z=0)
    DriverItem=(Type=Bone,Name="clavicle_l")
    FalloffRegionScaleFactors=(PositiveWidth=1,NegativeWidth=1,PositiveHeight=1,NegativeHeight=1)
    FalloffSize=0.1
    OptionalParentItem=(Type=Bone)
    RotationOffset=(X=40.9099,Y=16.0149,Z=-88.4832)
    Debug.bDrawDebug=RigVMVariableNode() {variable=Degub_arm_l}.Value
    ExecutePin=@RigVMFunction_Sequence_3_2_2.C
   )
   > &RigUnit_ModifyTransforms(
    ItemToModify=((Item=(Type=Bone,Name="extra_pipe_02"),Transform=(Rotation=(X=0,Y=0,Z=0,W=1),Translation=(X=0,Y=0,Z=0),Scale3D=(X=1,Y=1,Z=1))))
    Weight=@n12.OutputParam
   )
   > &RigUnit_ModifyTransforms(
    ItemToModify=((Item=(Type=Bone,Name="extra_shoulder_strap_06"),Transform=(Rotation=(X=0,Y=0,Z=0,W=1),Translation=(X=-0.162,Y=0.1177,Z=-2.3621),Scale3D=(X=1,Y=1,Z=1))))
    ItemToModify.0.Transform.Rotation=&RigVMFunction_MathQuaternionFromEuler(Euler=(X=5,Y=-5,Z=0)).Result
    Weight=@n12.OutputParam
   )
   > &RigUnit_ModifyTransforms(
    ItemToModify=((Item=(Type=Bone,Name="extra_shoulder_strap_07"),Transform=(Rotation=(X=0,Y=0,Z=0,W=1),Translation=(X=-0.0024,Y=-0.0089,Z=-1.6845),Scale3D=(X=1,Y=1,Z=1))))
    ItemToModify.0.Transform.Rotation=&RigVMFunction_MathQuaternionFromEuler(Euler=(X=15,Y=-5,Z=0)).Result
    Weight=@n12.OutputParam
   )
   > &RigUnit_ModifyTransforms(
    ItemToModify=((Item=(Type=Bone,Name="extra_shoulder_strap_05"),Transform=(Rotation=(X=0,Y=0,Z=0,W=1),Translation=(X=-0.5819,Y=-0.2729,Z=-0.8656),Scale3D=(X=1,Y=1,Z=1))))
    ItemToModify.0.Transform.Rotation=&RigVMFunction_MathQuaternionFromEuler(Euler=(X=-5,Y=10,Z=0)).Result
    Weight=@n12.OutputParam
   )
   > &RigUnit_ModifyTransforms(
    ItemToModify=((Item=(Type=Bone,Name="extra_pipe_01"),Transform=(Rotation=(X=0,Y=0,Z=0,W=1),Translation=(X=-1.856,Y=-0.2374,Z=-1.4609),Scale3D=(X=1,Y=1,Z=1))))
    ItemToModify.0.Transform.Rotation=&RigVMFunction_MathQuaternionFromEuler(Euler=(X=-5,Y=0,Z=0)).Result
    Weight=@n12.OutputParam
   )
   n13 = &RigUnit_SphericalPoseReader(
    ActiveRegionScaleFactors=(PositiveWidth=1,NegativeWidth=1,PositiveHeight=0.5,NegativeHeight=2)
    ActiveRegionSize=0.1
    Debug=(bDrawDebug=true,bDraw2D=False,bDrawLocalAxes=False,DebugScale=25,DebugSegments=20,DebugThickness=0.25)
    DriverAxis=(X=1,Y=0,Z=0)
    DriverItem=(Type=Bone,Name="clavicle_l")
    FalloffRegionScaleFactors=(PositiveWidth=1,NegativeWidth=1,PositiveHeight=0.5,NegativeHeight=1)
    FalloffSize=0.3
    OptionalParentItem=(Type=Bone)
    RotationOffset=(X=48.9861,Y=6.8705,Z=-94.7416)
    Debug.bDrawDebug=RigVMVariableNode() {variable=Degub_arm_l}.Value
    ExecutePin=@RigVMFunction_Sequence_3_2_2.A
   )
   > &RigUnit_ParentConstraint(
    AdvancedSettings=(RotationOrderForFilter=XZY)
    Child=(Type=Bone,Name="extra_pipe_02")
    Filter=(TranslationFilter=(bX=True,bY=True,bZ=True),RotationFilter=(bX=true,bY=true,bZ=true),ScaleFilter=(bX=True,bY=True,bZ=True))
    Parents="((Item=(Type=Bone,Name=\"upperarm_l\"),Weight=0.0444),(Item=(Type=Bone,Name=\"deform_upperarm_fr_l\"),Weight=0.1074),(Item=(Type=Bone,Name=\"spine_05\"),Weight=0.1113),(Item=(Type=Bone,Name=\"upperarm_twist_02_l\"),Weight=0.1402),(Item=(Type=Bone,Name=\"clavicle_l\" ...(truncated)"
    Weight=@n13.OutputParam
   )
   > &RigUnit_ModifyTransforms(
    ItemToModify=((Item=(Type=Bone,Name="extra_pipe_02"),Transform=(Rotation=(X=0.0561,Y=0.0017,Z=-0.156,W=0.9862),Translation=(X=-0.1028,Y=1.3859,Z=3.5615),Scale3D=(X=1,Y=1,Z=1))))
    Weight=@n13.OutputParam
   )
   n14 = &RigUnit_SphericalPoseReader(
    ActiveRegionScaleFactors=(PositiveWidth=1,NegativeWidth=1,PositiveHeight=1,NegativeHeight=1)
    ActiveRegionSize=0.01
    Debug=(bDrawDebug=true,bDraw2D=False,bDrawLocalAxes=False,DebugScale=25,DebugSegments=20,DebugThickness=0.25)
    DriverAxis=(X=1,Y=0,Z=0)
    DriverItem=(Type=Bone,Name="clavicle_l")
    FalloffRegionScaleFactors=(PositiveWidth=1,NegativeWidth=1,PositiveHeight=1,NegativeHeight=1)
    FalloffSize=0.1
    OptionalParentItem=(Type=Bone)
    RotationOffset=(X=48.9861,Y=6.8705,Z=-94.7416)
    Debug.bDrawDebug=RigVMVariableNode() {variable=Degub_arm_l}.Value
    ExecutePin=@RigVMFunction_Sequence_3_2_2.B
   )
   > &RigUnit_ModifyTransforms(
    ItemToModify=((Item=(Type=Bone,Name="extra_pipe_02"),Transform=(Rotation=(X=0,Y=0,Z=0,W=1),Translation=(X=0,Y=0,Z=0),Scale3D=(X=1,Y=1,Z=1))))
    Weight=@n14.OutputParam
   )
   > &RigUnit_ModifyTransforms(
    ItemToModify=((Item=(Type=Bone,Name="extra_shoulder_strap_06"),Transform=(Rotation=(X=0,Y=0,Z=0,W=1),Translation=(X=-0.2849,Y=-0.0043,Z=-0.9669),Scale3D=(X=1,Y=1,Z=1))))
    ItemToModify.0.Transform.Rotation=&RigVMFunction_MathQuaternionFromEuler(Euler=(X=5,Y=-5,Z=0)).Result
    Weight=@n14.OutputParam
   )
   > &RigUnit_ModifyTransforms(
    ItemToModify=((Item=(Type=Bone,Name="extra_shoulder_strap_07"),Transform=(Rotation=(X=0,Y=0,Z=0,W=1),Translation=(X=-0.3666,Y=-0.5215,Z=-1.5586),Scale3D=(X=1,Y=1,Z=1))))
    ItemToModify.0.Transform.Rotation=&RigVMFunction_MathQuaternionFromEuler(Euler=(X=15,Y=-15,Z=0)).Result
    Weight=@n14.OutputParam
   )
   > &RigUnit_ModifyTransforms(
    ItemToModify=((Item=(Type=Bone,Name="extra_shoulder_strap_05"),Transform=(Rotation=(X=0,Y=0,Z=0,W=1),Translation=(X=0.3934,Y=-0.2411,Z=-0.502),Scale3D=(X=1,Y=1,Z=1))))
    ItemToModify.0.Transform.Rotation=&RigVMFunction_MathQuaternionFromEuler(Euler=(X=0,Y=10,Z=0)).Result
    Weight=@n14.OutputParam
   )
   > &RigUnit_ModifyTransforms(
    ItemToModify=((Item=(Type=Bone,Name="extra_pipe_01"),Transform=(Rotation=(X=0,Y=0,Z=0,W=1),Translation=(X=-2.0056,Y=-0.1242,Z=-1.0782),Scale3D=(X=1,Y=1,Z=1))))
    ItemToModify.0.Transform.Rotation=&RigVMFunction_MathQuaternionFromEuler(Euler=(X=-15,Y=0,Z=0)).Result
    Weight=@n14.OutputParam
   )
   n15 = &RigUnit_SphericalPoseReader(
    ActiveRegionScaleFactors=(PositiveWidth=1,NegativeWidth=1,PositiveHeight=1,NegativeHeight=1)
    ActiveRegionSize=0.03
    Debug=(bDrawDebug=true,bDraw2D=False,bDrawLocalAxes=False,DebugScale=25,DebugSegments=20,DebugThickness=0.25)
    DriverAxis=(X=1,Y=0,Z=0)
    DriverItem=(Type=Bone,Name="clavicle_l")
    FalloffRegionScaleFactors=(PositiveWidth=1,NegativeWidth=1,PositiveHeight=1,NegativeHeight=1)
    FalloffSize=0.1
    OptionalParentItem=(Type=Bone)
    RotationOffset=(X=55,Y=-33,Z=-60)
    Debug.bDrawDebug=RigVMVariableNode() {variable=Degub_arm_l}.Value
    ExecutePin=@RigVMFunction_Sequence_3_1_1_2.B
   )
   > &RigUnit_ModifyTransforms(
    ItemToModify=((Item=(Type=Bone,Name="extra_pipe_02"),Transform=(Rotation=(X=0,Y=0,Z=0,W=1),Translation=(X=0.5691,Y=0,Z=0),Scale3D=(X=1,Y=1,Z=1))))
    Weight=@n15.OutputParam
   )
   > &RigUnit_ModifyTransforms(
    ItemToModify=((Item=(Type=Bone,Name="extra_shoulder_strap_06"),Transform=(Rotation=(X=0,Y=0,Z=0,W=1),Translation=(X=-0.2243,Y=0.0558,Z=-1.6541),Scale3D=(X=1,Y=1,Z=1))))
    ItemToModify.0.Transform.Rotation=&RigVMFunction_MathQuaternionFromEuler(Euler=(X=5,Y=-5,Z=0)).Result
    Weight=@n15.OutputParam
   )
   > &RigUnit_ModifyTransforms(
    ItemToModify=((Item=(Type=Bone,Name="extra_shoulder_strap_07"),Transform=(Rotation=(X=0,Y=0,Z=0,W=1),Translation=(X=-0.0204,Y=-0.0353,Z=-1.4838),Scale3D=(X=1,Y=1,Z=1))))
    ItemToModify.0.Transform.Rotation=&RigVMFunction_MathQuaternionFromEuler(Euler=(X=10,Y=-5,Z=0)).Result
    Weight=@n15.OutputParam
   )
   > &RigUnit_ModifyTransforms(
    ItemToModify=((Item=(Type=Bone,Name="extra_shoulder_strap_05"),Transform=(Rotation=(X=0,Y=0,Z=0,W=1),Translation=(X=-0.5095,Y=-0.2382,Z=-0.4692),Scale3D=(X=1,Y=1,Z=1))))
    ItemToModify.0.Transform.Rotation=&RigVMFunction_MathQuaternionFromEuler(Euler=(X=-5,Y=10,Z=0)).Result
    Weight=@n15.OutputParam
   )
   > &RigUnit_ModifyTransforms(
    ItemToModify=((Item=(Type=Bone,Name="extra_pipe_01"),Transform=(Rotation=(X=0,Y=0,Z=0,W=1),Translation=(X=-0.3904,Y=-0.1811,Z=0.3465),Scale3D=(X=1,Y=1,Z=1))))
    ItemToModify.0.Transform.Rotation=&RigVMFunction_MathQuaternionFromEuler(Euler=(X=-5,Y=0,Z=0)).Result
    Weight=@n15.OutputParam
   )
   > &RigUnit_ModifyTransforms(
    ItemToModify=((Item=(Type=Bone,Name="extra_shoulder_strap_01"),Transform=(Rotation=(X=0,Y=0,Z=0,W=1),Translation=(X=0,Y=0,Z=-0.7961),Scale3D=(X=1,Y=1,Z=1))))
    ItemToModify.0.Transform.Rotation=&RigVMFunction_MathQuaternionFromEuler(Euler=(X=0,Y=0,Z=0)).Result
    Weight=@n15.OutputParam
   )
   n16 = &RigUnit_SphericalPoseReader(
    ActiveRegionScaleFactors=(PositiveWidth=1,NegativeWidth=1,PositiveHeight=1,NegativeHeight=1)
    ActiveRegionSize=0.03
    Debug=(bDrawDebug=true,bDraw2D=False,bDrawLocalAxes=False,DebugScale=25,DebugSegments=20,DebugThickness=0.25)
    DriverAxis=(X=1,Y=0,Z=0)
    DriverItem=(Type=Bone,Name="clavicle_l")
    FalloffRegionScaleFactors=(PositiveWidth=1,NegativeWidth=1,PositiveHeight=1,NegativeHeight=1)
    FalloffSize=0.1
    OptionalParentItem=(Type=Bone)
    RotationOffset=(X=50,Y=-50,Z=-65)
    Debug.bDrawDebug=RigVMVariableNode() {variable=Degub_arm_l}.Value
    ExecutePin=@RigVMFunction_Sequence_3_1_1_2.C
   )
   > &RigUnit_ModifyTransforms(
    ItemToModify=((Item=(Type=Bone,Name="extra_pipe_01"),Transform=(Rotation=(X=0,Y=0,Z=0,W=1),Translation=(X=-1.009,Y=0.0424,Z=-0.9235),Scale3D=(X=1,Y=1,Z=1))))
    ItemToModify.0.Transform.Rotation=&RigVMFunction_MathQuaternionFromEuler(Euler=(X=0,Y=0,Z=10)).Result
    Weight=@n16.OutputParam
   )
   n17 = &RigVMFunction_Sequence(ExecuteContext=@n1.B)
   n18 = &RigVMFunction_Sequence(ExecuteContext=@n17.B)
   RigVMFunction_Sequence_9 = RigVMAggregateNode(ExecuteContext=@n18.A)
   n19 = &RigUnit_SphericalPoseReader(
    ActiveRegionScaleFactors=(PositiveWidth=1,NegativeWidth=1,PositiveHeight=2,NegativeHeight=2)
    ActiveRegionSize=0.01
    Debug=(bDrawDebug=true,bDraw2D=False,bDrawLocalAxes=False,DebugScale=25,DebugSegments=20,DebugThickness=0.25)
    DriverAxis=(X=-1,Y=0,Z=0)
    DriverItem=(Type=Bone,Name="thigh_l")
    FalloffRegionScaleFactors=(PositiveWidth=1,NegativeWidth=1,PositiveHeight=2,NegativeHeight=2)
    FalloffSize=0.2
    FlipHeightScaling=False
    FlipWidthScaling=False
    OptionalParentItem=(Type=Bone,Name="None")
    RotationOffset=(X=105.0011,Y=-15.1094,Z=35.2898)
    Debug.bDrawDebug=RigVMVariableNode() {variable=Degub_thigh_l}.Value
    ExecutePin=@RigVMFunction_Sequence_9.B
   )
   > &RigUnit_ModifyTransforms(
    ItemToModify=((Item=(Type=Bone,Name="extra_jaket_side_fr_l"),Transform=(Rotation=(X=-0.0045,Y=-0.0153,Z=-0.5194,W=0.8544),Translation=(X=-1.0433,Y=-0.4732,Z=-1.916),Scale3D=(X=1,Y=1,Z=1))))
    Weight=@n19.OutputParam
   )
   > &RigUnit_ModifyTransforms(
    ItemToModify=((Item=(Type=Bone,Name="extra_jaket_side_l"),Transform=(Rotation=(X=0.2201,Y=0.0718,Z=0.2978,W=0.9261),Translation=(X=0,Y=-3.7343,Z=0),Scale3D=(X=1,Y=1,Z=1))))
    Weight=@n19.OutputParam
   )
   n20 = &RigUnit_SphericalPoseReader(
    ActiveRegionScaleFactors=(PositiveWidth=1,NegativeWidth=1,PositiveHeight=2,NegativeHeight=2)
    ActiveRegionSize=0.01
    Debug=(bDrawDebug=true,bDraw2D=False,bDrawLocalAxes=False,DebugScale=25,DebugSegments=20,DebugThickness=0.25)
    DriverAxis=(X=-1,Y=0,Z=0)
    DriverItem=(Type=Bone,Name="thigh_l")
    FalloffRegionScaleFactors=(PositiveWidth=1,NegativeWidth=1,PositiveHeight=2,NegativeHeight=2)
    FalloffSize=0.2
    FlipHeightScaling=False
    FlipWidthScaling=False
    OptionalParentItem=(Type=Bone,Name="None")
    RotationOffset=(X=105.0011,Y=-15.1094,Z=65.3123)
    Debug.bDrawDebug=RigVMVariableNode() {variable=Degub_thigh_l}.Value
    ExecutePin=@RigVMFunction_Sequence_9.A
   )
   > &RigUnit_ModifyTransforms(
    ItemToModify=((Item=(Type=Bone,Name="extra_knee_pad"),Transform=(Rotation=(X=0,Y=0,Z=0.1053,W=0.9944),Translation=(X=1.6358,Y=0.296,Z=0),Scale3D=(X=1,Y=1,Z=1))))
    Weight=@n20.OutputParam
   )
   n21 = &RigUnit_SphericalPoseReader(
    ActiveRegionScaleFactors=(PositiveWidth=1,NegativeWidth=1,PositiveHeight=1,NegativeHeight=1)
    ActiveRegionSize=0.03
    Debug=(bDrawDebug=true,bDraw2D=False,bDrawLocalAxes=False,DebugScale=25,DebugSegments=20,DebugThickness=0.25)
    DriverAxis=(X=-1,Y=0,Z=0)
    DriverItem=(Type=Bone,Name="thigh_l")
    FalloffRegionScaleFactors=(PositiveWidth=1,NegativeWidth=1,PositiveHeight=1,NegativeHeight=1)
    FalloffSize=0.2
    FlipHeightScaling=False
    FlipWidthScaling=False
    OptionalParentItem=(Type=Bone,Name="None")
    RotationOffset=(X=121.7015,Y=32.7303,Z=-10.9851)
    Debug.bDrawDebug=RigVMVariableNode() {variable=Degub_thigh_l}.Value
    ExecutePin=@RigVMFunction_Sequence_9.D
   )
   > &RigUnit_ModifyTransforms(
    ItemToModify=((Item=(Type=Bone,Name="extra_jaket_side_fr_l"),Transform=(Rotation=(X=-0.111,Y=-0.0284,Z=-0.29,W=0.9501),Translation=(X=-9.1129,Y=-2.5329,Z=-2.1038),Scale3D=(X=1,Y=1,Z=1))))
    Weight=@n21.OutputParam
   )
   > &RigUnit_ModifyTransforms(
    ItemToModify=((Item=(Type=Bone,Name="extra_jaket_side_l"),Transform=(Rotation=(X=0.4592,Y=-0.2834,Z=-0.1577,W=0.827),Translation=(X=-0.8514,Y=-3.0469,Z=0.5123),Scale3D=(X=1,Y=1,Z=1))))
    Weight=@n21.OutputParam
   )
   > &RigUnit_ModifyTransforms(
    ItemToModify=((Item=(Type=Bone,Name="extra_jaket_fr"),Transform=(Rotation=(X=0.4463,Y=-0.1716,Z=-0.5314,W=0.6992),Translation=(X=0,Y=0,Z=0),Scale3D=(X=1,Y=1,Z=1))))
    Weight=@n21.OutputParam
   )
   > &RigUnit_ModifyTransforms(
    ItemToModify=((Item=(Type=Bone,Name="extra_bag_fr_l"),Transform=(Rotation=(X=0.142,Y=-0.5352,Z=-0.3444,W=0.7581),Translation=(X=5.3187,Y=-1.5728,Z=3.5843),Scale3D=(X=1,Y=1,Z=1))))
    Weight=@n21.OutputParam
   )
   n22 = &RigUnit_SphericalPoseReader(
    ActiveRegionScaleFactors=(PositiveWidth=1,NegativeWidth=1,PositiveHeight=2,NegativeHeight=2)
    ActiveRegionSize=0.01
    Debug=(bDrawDebug=true,bDraw2D=False,bDrawLocalAxes=False,DebugScale=25,DebugSegments=20,DebugThickness=0.25)
    DriverAxis=(X=-1,Y=0,Z=0)
    DriverItem=(Type=Bone,Name="thigh_l")
    FalloffRegionScaleFactors=(PositiveWidth=1,NegativeWidth=1,PositiveHeight=2,NegativeHeight=2)
    FalloffSize=0.15
    FlipHeightScaling=False
    FlipWidthScaling=False
    OptionalParentItem=(Type=Bone,Name="None")
    RotationOffset=(X=112.8661,Y=-5.6481,Z=2.3289)
    Debug.bDrawDebug=RigVMVariableNode() {variable=Degub_thigh_l}.Value
    ExecutePin=@RigVMFunction_Sequence_9.C
   )
   > &RigUnit_ModifyTransforms(
    ItemToModify=((Item=(Type=Bone,Name="extra_jaket_side_fr_l"),Transform=(Rotation=(X=0.3346,Y=-0.085,Z=-0.2051,W=0.9158),Translation=(X=-3.1121,Y=-4.3445,Z=-2.4342),Scale3D=(X=1,Y=1,Z=1))))
    Weight=@n22.OutputParam
   )
   > &RigUnit_ModifyTransforms(
    ItemToModify=((Item=(Type=Bone,Name="extra_jaket_side_l"),Transform=(Rotation=(X=0.1236,Y=-0.4181,Z=-0.2888,W=0.8524),Translation=(X=2.585,Y=1.7728,Z=0.1004),Scale3D=(X=1,Y=1,Z=1))))
    Weight=@n22.OutputParam
   )
   > &RigUnit_ModifyTransforms(
    ItemToModify=((Item=(Type=Bone,Name="extra_jaket_fr"),Transform=(Rotation=(X=-0.5278,Y=0.3423,Z=-0.3604,W=0.6887),Translation=(X=0,Y=0,Z=0),Scale3D=(X=1,Y=1,Z=1))))
    Weight=@n22.OutputParam
   )
   RigVMFunction_Sequence_9_1 = RigVMAggregateNode(ExecuteContext=@n18.B)
   &RigUnit_ParentConstraint(
    AdvancedSettings=(InterpolationType=Average,RotationOrderForFilter=XZY)
    Child=(Type=Bone,Name="extra_jaket_fr")
    Filter=(TranslationFilter=(bX=false,bY=false,bZ=false),RotationFilter=(bX=True,bY=True,bZ=True),ScaleFilter=(bX=True,bY=True,bZ=True))
    Parents=((Item=(Type=Bone,Name="extra_jaket_fr_r"),Weight=0.5),(Item=(Type=Bone,Name="extra_jaket_side_fr_l"),Weight=0.5))
    Weight=0.5
    ExecutePin=@RigVMFunction_Sequence_9_1.C
   )
   n23 = &RigUnit_SphericalPoseReader(
    ActiveRegionScaleFactors=(PositiveWidth=1,NegativeWidth=1,PositiveHeight=2,NegativeHeight=2)
    ActiveRegionSize=0.01
    Debug=(bDrawDebug=true,bDraw2D=False,bDrawLocalAxes=False,DebugScale=25,DebugSegments=20,DebugThickness=0.25)
    DriverAxis=(X=-1,Y=0,Z=0)
    DriverItem=(Type=Bone,Name="thigh_l")
    FalloffRegionScaleFactors=(PositiveWidth=1,NegativeWidth=1,PositiveHeight=1,NegativeHeight=1)
    FalloffSize=0.15
    FlipHeightScaling=False
    FlipWidthScaling=False
    OptionalParentItem=(Type=Bone,Name="None")
    RotationOffset=(X=84.63,Y=0,Z=3)
    Debug.bDrawDebug=RigVMVariableNode() {variable=Degub_thigh_l}.Value
    ExecutePin=@RigVMFunction_Sequence_9_1.B
   )
   > &RigUnit_ModifyTransforms(
    ItemToModify=((Item=(Type=Bone,Name="extra_jaket_side_fr_l"),Transform=(Rotation=(X=0.0762,Y=-0.1126,Z=-0.7143,W=0.6865),Translation=(X=-1.5178,Y=-2.5094,Z=-2.3666),Scale3D=(X=1,Y=1,Z=1))))
    Weight=@n23.OutputParam
   )
   > &RigUnit_ModifyTransforms(
    ItemToModify=((Item=(Type=Bone,Name="extra_jaket_side_l"),Transform=(Rotation=(X=0.0629,Y=0.0314,Z=0.3094,W=0.9483),Translation=(X=-0.1102,Y=-4.2322,Z=2.4089),Scale3D=(X=1,Y=1,Z=1))))
    Weight=@n23.OutputParam
   )
   n24 = &RigUnit_SphericalPoseReader(
    ActiveRegionScaleFactors=(PositiveWidth=1,NegativeWidth=1,PositiveHeight=1,NegativeHeight=1)
    ActiveRegionSize=0.01
    Debug=(bDrawDebug=true,bDraw2D=False,bDrawLocalAxes=False,DebugScale=25,DebugSegments=20,DebugThickness=0.25)
    DriverAxis=(X=-1,Y=0,Z=0)
    DriverItem=(Type=Bone,Name="thigh_l")
    FalloffRegionScaleFactors=(PositiveWidth=1,NegativeWidth=1,PositiveHeight=1,NegativeHeight=1)
    FalloffSize=0.15
    FlipHeightScaling=False
    FlipWidthScaling=False
    OptionalParentItem=(Type=Bone,Name="None")
    RotationOffset=(X=133.5992,Y=-5.6481,Z=42.1651)
    Debug.bDrawDebug=RigVMVariableNode() {variable=Degub_thigh_l}.Value
    ExecutePin=@RigVMFunction_Sequence_9_1.A
   )
   > &RigUnit_ModifyTransforms(
    ItemToModify=((Item=(Type=Bone,Name="extra_jaket_side_l"),Transform=(Rotation=(X=0.0188,Y=-0.2256,Z=-0.3593,W=0.9053),Translation=(X=2.5851,Y=1.7728,Z=0.1005),Scale3D=(X=1,Y=1,Z=1))))
    Weight=@n24.OutputParam
   )
   RigVMFunction_Sequence_7_1 = RigVMAggregateNode(ExecuteContext=@n17.A)
   RigVMFunction_Sequence_3_3 = RigVMAggregateNode(ExecuteContext=@RigVMFunction_Sequence_7_1.A)
   n25 = &RigUnit_SphericalPoseReader(
    ActiveRegionScaleFactors=(PositiveWidth=1,NegativeWidth=1,PositiveHeight=1,NegativeHeight=1)
    ActiveRegionSize=0.02
    Debug=(bDrawDebug=true,bDraw2D=False,bDrawLocalAxes=False,DebugScale=25,DebugSegments=20,DebugThickness=0.25)
    DriverAxis=(X=1,Y=0,Z=0)
    DriverItem=(Type=Bone,Name="thigh_r")
    FalloffRegionScaleFactors=(PositiveWidth=1,NegativeWidth=1,PositiveHeight=1,NegativeHeight=1)
    FalloffSize=0.2
    FlipHeightScaling=False
    FlipWidthScaling=False
    OptionalParentItem=(Type=None,Name="None")
    RotationOffset=(X=-69.8033,Y=-5.4014,Z=69.9223)
    Debug.bDrawDebug=RigVMVariableNode() {variable=Degub_thigh_r}.Value
    ExecutePin=@RigVMFunction_Sequence_3_3.A
   )
   > &RigUnit_ModifyTransforms(
    ItemToModify=((Item=(Type=Bone,Name="extra_jaket_fr_r"),Transform=(Rotation=(X=0.0166,Y=0.0317,Z=0.0915,W=-0.9952),Translation=(X=0.7816,Y=-0.4585,Z=-2.1496),Scale3D=(X=1,Y=1,Z=1))))
    Weight=@n25.OutputParam
   )
   > &RigUnit_ModifyTransforms(
    ItemToModify=((Item=(Type=Bone,Name="extra_jaket_side_fr_r"),Transform=(Rotation=(X=0,Y=0,Z=-0.0436,W=0.999),Translation=(X=0,Y=0,Z=0),Scale3D=(X=1,Y=1,Z=1))))
    Weight=@n25.OutputParam
   )
   n26 = &RigUnit_SphericalPoseReader(
    ActiveRegionScaleFactors=(PositiveWidth=1,NegativeWidth=1,PositiveHeight=1,NegativeHeight=1)
    ActiveRegionSize=0.02
    Debug=(bDrawDebug=true,bDraw2D=False,bDrawLocalAxes=False,DebugScale=25,DebugSegments=20,DebugThickness=0.25)
    DriverAxis=(X=1,Y=0,Z=0)
    DriverItem=(Type=Bone,Name="thigh_r")
    FalloffRegionScaleFactors=(PositiveWidth=1,NegativeWidth=1,PositiveHeight=1,NegativeHeight=1)
    FalloffSize=0.2
    FlipHeightScaling=False
    FlipWidthScaling=False
    OptionalParentItem=(Type=None,Name="None")
    RotationOffset=(X=-102.6683,Y=0,Z=52.3614)
    Debug.bDrawDebug=RigVMVariableNode() {variable=Degub_thigh_r}.Value
    ExecutePin=@RigVMFunction_Sequence_3_3.B
   )
   > &RigUnit_ModifyTransforms(
    ItemToModify=((Item=(Type=Bone,Name="extra_jaket_fr_r"),Transform=(Rotation=(X=0.0634,Y=-0.0365,Z=0.0742,W=-0.9946),Translation=(X=2.1922,Y=0.4131,Z=-1.2223),Scale3D=(X=1,Y=1,Z=1))))
    Weight=@n26.OutputParam
   )
   > &RigUnit_ModifyTransforms(
    ItemToModify=((Item=(Type=Bone,Name="extra_jaket_side_fr_r"),Transform=(Rotation=(X=0.0136,Y=-0.0005,Z=-0.0343,W=0.9993),Translation=(X=0,Y=0,Z=0),Scale3D=(X=1,Y=1,Z=1))))
    Weight=@n26.OutputParam
   )
   n27 = &RigUnit_SphericalPoseReader(
    ActiveRegionScaleFactors=(PositiveWidth=1,NegativeWidth=1,PositiveHeight=1,NegativeHeight=1)
    ActiveRegionSize=0.02
    Debug=(bDrawDebug=true,bDraw2D=False,bDrawLocalAxes=False,DebugScale=25,DebugSegments=20,DebugThickness=0.25)
    DriverAxis=(X=1,Y=0,Z=0)
    DriverItem=(Type=Bone,Name="thigh_r")
    FalloffRegionScaleFactors=(PositiveWidth=1,NegativeWidth=1,PositiveHeight=1,NegativeHeight=1)
    FalloffSize=0.2
    FlipHeightScaling=False
    FlipWidthScaling=False
    OptionalParentItem=(Type=Bone,Name="None")
    RotationOffset=(X=-124.6118,Y=0,Z=6.2291)
    Debug.bDrawDebug=RigVMVariableNode() {variable=Degub_thigh_r}.Value
    ExecutePin=@RigVMFunction_Sequence_3_3.C
   )
   > &RigUnit_ModifyTransforms(
    ItemToModify=((Item=(Type=Bone,Name="extra_jaket_fr_r"),Transform=(Rotation=(X=0.1179,Y=-0.1511,Z=0.4158,W=-0.889),Translation=(X=-3.9034,Y=-7.3141,Z=-10.2517),Scale3D=(X=1,Y=1,Z=1))))
    Weight=@n27.OutputParam
   )
   > &RigUnit_ModifyTransforms(
    ItemToModify=((Item=(Type=Bone,Name="extra_jaket_side_fr_r"),Transform=(Rotation=(X=-0.1467,Y=0.0935,Z=0.0952,W=0.9801),Translation=(X=0.3017,Y=-0.2115,Z=-2.5899),Scale3D=(X=1,Y=1,Z=1))))
    Weight=@n27.OutputParam
   )
   > &RigUnit_ModifyTransforms(
    ItemToModify=((Item=(Type=Bone,Name="extra_jaket_side_r"),Transform=(Rotation=(X=0.0255,Y=0.0323,Z=0.0934,W=0.9948),Translation=(X=0,Y=0,Z=0),Scale3D=(X=1,Y=1,Z=1))))
    Weight=@n27.OutputParam
   )
   > &RigUnit_ModifyTransforms(
    ItemToModify=((Item=(Type=Bone,Name="extra_jaket_fr"),Transform=(Rotation=(X=0.5366,Y=-0.3119,Z=-0.7767,W=0.1079),Translation=(X=-2.3441,Y=-1.6079,Z=0.9703),Scale3D=(X=1,Y=1,Z=1))))
    Weight=@n27.OutputParam
   )
   RigVMFunction_Sequence_5_3 = RigVMAggregateNode(ExecuteContext=@RigVMFunction_Sequence_7_1.B)
   n28 = &RigUnit_SphericalPoseReader(
    ActiveRegionScaleFactors=(PositiveWidth=1,NegativeWidth=1,PositiveHeight=1,NegativeHeight=1)
    ActiveRegionSize=0.01
    Debug=(bDrawDebug=true,bDraw2D=False,bDrawLocalAxes=False,DebugScale=25,DebugSegments=20,DebugThickness=0.25)
    DriverAxis=(X=1,Y=0,Z=0)
    DriverItem=(Type=Bone,Name="thigh_r")
    FalloffRegionScaleFactors=(PositiveWidth=1,NegativeWidth=1,PositiveHeight=1,NegativeHeight=1)
    FalloffSize=0.2
    FlipHeightScaling=False
    FlipWidthScaling=False
    OptionalParentItem=(Type=Bone,Name="None")
    RotationOffset=(X=-47.557,Y=-9.1245,Z=38.5768)
    Debug.bDrawDebug=RigVMVariableNode() {variable=Degub_thigh_r}.Value
    ExecutePin=@RigVMFunction_Sequence_5_3.A
   )
   > &RigUnit_ModifyTransforms(
    ItemToModify=((Item=(Type=Bone,Name="extra_jaket_fr_r"),Transform=(Rotation=(X=-0.1821,Y=0.0797,Z=0.2882,W=-0.9367),Translation=(X=1.0444,Y=-5.5148,Z=0.2398),Scale3D=(X=1,Y=1,Z=1))))
    Weight=@n28.OutputParam
   )
   > &RigUnit_ModifyTransforms(
    ItemToModify=((Item=(Type=Bone,Name="extra_jaket_side_fr_r"),Transform=(Rotation=(X=-0.0937,Y=0.0092,Z=-0.401,W=0.9112),Translation=(X=2.1213,Y=-1.6092,Z=-0.3294),Scale3D=(X=1,Y=1,Z=1))))
    Weight=@n28.OutputParam
   )
   > &RigUnit_ModifyTransforms(
    ItemToModify=((Item=(Type=Bone,Name="extra_jaket_side_r"),Transform=(Rotation=(X=-0.0866,Y=0.0716,Z=-0.1725,W=0.9786),Translation=(X=0,Y=0,Z=0),Scale3D=(X=1,Y=1,Z=1))))
    Weight=@n28.OutputParam
   )
   n29 = &RigUnit_SphericalPoseReader(
    ActiveRegionScaleFactors=(PositiveWidth=1,NegativeWidth=1,PositiveHeight=1,NegativeHeight=1)
    ActiveRegionSize=0.01
    Debug=(bDrawDebug=true,bDraw2D=False,bDrawLocalAxes=False,DebugScale=25,DebugSegments=20,DebugThickness=0.25)
    DriverAxis=(X=1,Y=0,Z=0)
    DriverItem=(Type=Bone,Name="thigh_r")
    FalloffRegionScaleFactors=(PositiveWidth=1,NegativeWidth=1,PositiveHeight=1,NegativeHeight=1)
    FalloffSize=0.2
    FlipHeightScaling=False
    FlipWidthScaling=False
    OptionalParentItem=(Type=Bone,Name="None")
    RotationOffset=(X=-85.926,Y=5.3521,Z=32.0759)
    Debug.bDrawDebug=RigVMVariableNode() {variable=Degub_thigh_r}.Value
    ExecutePin=@RigVMFunction_Sequence_5_3.B
   )
   > &RigUnit_ModifyTransforms(
    ItemToModify=((Item=(Type=Bone,Name="extra_jaket_fr_r"),Transform=(Rotation=(X=-0.211,Y=-0.0037,Z=0.4119,W=-0.8864),Translation=(X=-1.2237,Y=-6.9369,Z=-2.5727),Scale3D=(X=1,Y=1,Z=1))))
    Weight=@n29.OutputParam
   )
   > &RigUnit_ModifyTransforms(
    ItemToModify=((Item=(Type=Bone,Name="extra_jaket_side_fr_r"),Transform=(Rotation=(X=-0.1041,Y=0.087,Z=-0.1271,W=0.9826),Translation=(X=1.8719,Y=-2.455,Z=-2.5868),Scale3D=(X=1,Y=1,Z=1))))
    Weight=@n29.OutputParam
   )
   > &RigUnit_ModifyTransforms(
    ItemToModify=((Item=(Type=Bone,Name="extra_jaket_side_r"),Transform=(Rotation=(X=0.0241,Y=0.0892,Z=-0.0426,W=0.9948),Translation=(X=0,Y=0,Z=0),Scale3D=(X=1,Y=1,Z=1))))
    Weight=@n29.OutputParam
   )
   > &RigUnit_ModifyTransforms(
    ItemToModify=((Item=(Type=Bone,Name="extra_jaket_fr"),Transform=(Rotation=(X=0.488,Y=-0.3744,Z=-0.3351,W=0.7137),Translation=(X=0,Y=0,Z=0),Scale3D=(X=1,Y=1,Z=1))))
    Weight=@n29.OutputParam
   )
   n30 = &RigUnit_SphericalPoseReader(
    ActiveRegionScaleFactors=(PositiveWidth=1,NegativeWidth=1,PositiveHeight=1,NegativeHeight=1)
    ActiveRegionSize=0.01
    Debug=(bDrawDebug=true,bDraw2D=False,bDrawLocalAxes=False,DebugScale=25,DebugSegments=20,DebugThickness=0.25)
    DriverAxis=(X=1,Y=0,Z=0)
    DriverItem=(Type=Bone,Name="thigh_r")
    FalloffRegionScaleFactors=(PositiveWidth=1,NegativeWidth=1,PositiveHeight=1,NegativeHeight=1)
    FalloffSize=0.1
    FlipHeightScaling=False
    FlipWidthScaling=False
    OptionalParentItem=(Type=Bone,Name="None")
    RotationOffset=(X=-67.4083,Y=-20.1291,Z=10.5349)
    Debug.bDrawDebug=RigVMVariableNode() {variable=Degub_thigh_r}.Value
    ExecutePin=@RigVMFunction_Sequence_5_3.C
   )
   > &RigUnit_ModifyTransforms(
    ItemToModify=((Item=(Type=Bone,Name="extra_jaket_fr_r"),Transform=(Rotation=(X=-0.0223,Y=0.1073,Z=0.2166,W=-0.9701),Translation=(X=1.8175,Y=-4.066,Z=1.0921),Scale3D=(X=1,Y=1,Z=1))))
    Weight=@n30.OutputParam
   )
   > &RigUnit_ModifyTransforms(
    ItemToModify=((Item=(Type=Bone,Name="extra_jaket_side_fr_r"),Transform=(Rotation=(X=-0.0975,Y=0.0249,Z=-0.1813,W=0.9783),Translation=(X=0.8298,Y=-0.3534,Z=-0.9997),Scale3D=(X=1,Y=1,Z=1))))
    Weight=@n30.OutputParam
   )
   > &RigUnit_ModifyTransforms(
    ItemToModify=((Item=(Type=Bone,Name="extra_jaket_side_r"),Transform=(Rotation=(X=0.0352,Y=0.0248,Z=-0.101,W=0.994),Translation=(X=0,Y=0,Z=0),Scale3D=(X=1,Y=1,Z=1))))
    Weight=@n30.OutputParam
   )
   RigVMFunction_Sequence_8 = RigVMAggregateNode(ExecuteContext=@RigVMFunction_Sequence_7_1.C)
   &RigUnit_ParentConstraint(
    AdvancedSettings=(InterpolationType=Average,RotationOrderForFilter=XZY)
    Child=(Type=Bone,Name="dyn_grenade_01_base")
    Filter=(TranslationFilter=(bX=True,bY=True,bZ=True),RotationFilter=(bX=True,bY=True,bZ=True),ScaleFilter=(bX=false,bY=false,bZ=false))
    Parents=((Item=(Type=Bone,Name="extra_jaket_fr_r"),Weight=1))
    Weight=1
    ExecutePin=@RigVMFunction_Sequence_8.D
   )
   > &RigUnit_ParentConstraint(
    AdvancedSettings=(InterpolationType=Average,RotationOrderForFilter=XZY)
    Child=(Type=Bone,Name="dyn_grenade_03_base")
    Filter=(TranslationFilter=(bX=false,bY=false,bZ=false),RotationFilter=(bX=True,bY=True,bZ=True),ScaleFilter=(bX=false,bY=false,bZ=false))
    Parents=((Item=(Type=Bone,Name="extra_jaket_side_fr_r"),Weight=1))
    Weight=1
   )
   > &RigUnit_ParentConstraint(
    AdvancedSettings=(InterpolationType=Average,RotationOrderForFilter=XZY)
    Child=(Type=Bone,Name="dyn_grenade_02_base")
    Filter=(TranslationFilter=(bX=True,bY=True,bZ=True),RotationFilter=(bX=True,bY=True,bZ=True),ScaleFilter=(bX=false,bY=false,bZ=false))
    Parents=((Item=(Type=Bone,Name="dyn_grenade_01_base"),Weight=0.5),(Item=(Type=Bone,Name="dyn_grenade_03_base"),Weight=0.5))
    Weight=1
   )
   > &RigUnit_ParentConstraint(
    AdvancedSettings=(InterpolationType=Average,RotationOrderForFilter=XZY)
    Child=(Type=Bone,Name="dyn_grenade_04_base")
    Filter=(TranslationFilter=(bX=True,bY=True,bZ=True),RotationFilter=(bX=True,bY=True,bZ=True),ScaleFilter=(bX=false,bY=false,bZ=false))
    Parents=((Item=(Type=Bone,Name="dyn_grenade_03_base"),Weight=0.5),(Item=(Type=Bone,Name="dyn_grenade_05_base"),Weight=0.5))
    Weight=1
   )
   n31 = &RigUnit_SphericalPoseReader(
    ActiveRegionScaleFactors=(PositiveWidth=1,NegativeWidth=1,PositiveHeight=1,NegativeHeight=1)
    ActiveRegionSize=0.03
    Debug=(bDrawDebug=true,bDraw2D=False,bDrawLocalAxes=False,DebugScale=25,DebugSegments=20,DebugThickness=0.25)
    DriverAxis=(X=1,Y=0,Z=0)
    DriverItem=(Type=Bone,Name="thigh_r")
    FalloffRegionScaleFactors=(PositiveWidth=1,NegativeWidth=1,PositiveHeight=1,NegativeHeight=1)
    FalloffSize=0.3
    FlipHeightScaling=False
    FlipWidthScaling=False
    OptionalParentItem=(Type=Bone,Name="None")
    RotationOffset=(X=-51.1302,Y=-39.1067,Z=-58.0471)
    Debug.bDrawDebug=RigVMVariableNode() {variable=Degub_thigh_r}.Value
    ExecutePin=@RigVMFunction_Sequence_8.A
   )
   > &RigUnit_ModifyTransforms(
    ItemToModify=((Item=(Type=Bone,Name="extra_jaket_fr_r"),Transform=(Rotation=(X=0.0055,Y=-0.0312,Z=-0.7842,W=0.6197),Translation=(X=-16.7589,Y=-10.1449,Z=-2.2726),Scale3D=(X=1,Y=1,Z=1))))
    Weight=@n31.OutputParam
   )
   > &RigUnit_ModifyTransforms(
    ItemToModify=((Item=(Type=Bone,Name="extra_jaket_side_fr_r"),Transform=(Rotation=(X=-0.2602,Y=0.0383,Z=-0.526,W=0.8088),Translation=(X=-3.7495,Y=-8.1807,Z=-0.2604),Scale3D=(X=1,Y=1,Z=1))))
    Weight=@n31.OutputParam
   )
   > &RigUnit_ModifyTransforms(
    ItemToModify=((Item=(Type=Bone,Name="extra_jaket_side_r"),Transform=(Rotation=(X=-0.1155,Y=0.0171,Z=-0.0974,W=0.9884),Translation=(X=0,Y=0,Z=0),Scale3D=(X=1,Y=1,Z=1))))
    Weight=@n31.OutputParam
   )
   > &RigUnit_ModifyTransforms(
    ItemToModify=((Item=(Type=Bone,Name="extra_jaket_fr"),Transform=(Rotation=(X=0.5507,Y=-0.2782,Z=-0.4018,W=0.6767),Translation=(X=0,Y=0,Z=0),Scale3D=(X=1,Y=1,Z=1))))
    Weight=@n31.OutputParam
   )
   > &RigUnit_ModifyTransforms(
    ItemToModify=((Item=(Type=Bone,Name="extra_bag_fr_r"),Transform=(Rotation=(X=-0.0032,Y=0.0457,Z=0.0692,W=0.9966),Translation=(X=0.031,Y=0.0043,Z=2.7833),Scale3D=(X=1,Y=1,Z=1))))
    Weight=@n31.OutputParam
   )
   > &RigUnit_ModifyTransforms(
    ItemToModify=((Item=(Type=Bone,Name="dyn_grenade_03_base"),Transform=(Rotation=(X=0.5507,Y=-0.2783,Z=-0.4017,W=0.6767),Translation=(X=-1.2201,Y=1.2568,Z=-0.2347),Scale3D=(X=1,Y=1,Z=1))))
    Weight=@n31.OutputParam
   )
   n32 = &RigUnit_SphericalPoseReader(
    ActiveRegionScaleFactors=(PositiveWidth=1,NegativeWidth=1,PositiveHeight=1,NegativeHeight=1)
    ActiveRegionSize=0.01
    Debug=(bDrawDebug=true,bDraw2D=False,bDrawLocalAxes=False,DebugScale=25,DebugSegments=20,DebugThickness=0.25)
    DriverAxis=(X=1,Y=0,Z=0)
    DriverItem=(Type=Bone,Name="thigh_r")
    FalloffRegionScaleFactors=(PositiveWidth=1,NegativeWidth=1,PositiveHeight=1,NegativeHeight=1)
    FalloffSize=0.12
    FlipHeightScaling=False
    FlipWidthScaling=False
    OptionalParentItem=(Type=Bone,Name="None")
    RotationOffset=(X=-100.6244,Y=-29.5089,Z=15.1354)
    Debug.bDrawDebug=RigVMVariableNode() {variable=Degub_thigh_r}.Value
    ExecutePin=@RigVMFunction_Sequence_8.C
   )
   > &RigUnit_ModifyTransforms(
    ItemToModify=((Item=(Type=Bone,Name="extra_jaket_fr_r"),Transform=(Rotation=(X=0.1586,Y=0.0583,Z=-0.3868,W=0.9065),Translation=(X=-2.104,Y=-3.4598,Z=-4.2482),Scale3D=(X=1,Y=1,Z=1))))
    Weight=@n32.OutputParam
   )
   > &RigUnit_ModifyTransforms(
    ItemToModify=((Item=(Type=Bone,Name="extra_jaket_side_fr_r"),Transform=(Rotation=(X=-0.127,Y=0.1081,Z=-0.148,W=0.9748),Translation=(X=0.1519,Y=0.1137,Z=-1.5441),Scale3D=(X=1,Y=1,Z=1))))
    Weight=@n32.OutputParam
   )
   > &RigUnit_ModifyTransforms(
    ItemToModify=((Item=(Type=Bone,Name="extra_jaket_fr"),Transform=(Rotation=(X=0.7479,Y=-0.1073,Z=-0.3572,W=0.5491),Translation=(X=0,Y=0,Z=0),Scale3D=(X=1,Y=1,Z=1))))
    Weight=@n32.OutputParam
   )
   n33 = &RigUnit_SphericalPoseReader(
    ActiveRegionScaleFactors=(PositiveWidth=1,NegativeWidth=2,PositiveHeight=2,NegativeHeight=1)
    ActiveRegionSize=0.02
    Debug=(bDrawDebug=true,bDraw2D=False,bDrawLocalAxes=False,DebugScale=25,DebugSegments=20,DebugThickness=0.25)
    DriverAxis=(X=1,Y=0,Z=0)
    DriverItem=(Type=Bone,Name="thigh_r")
    FalloffRegionScaleFactors=(PositiveWidth=1,NegativeWidth=1,PositiveHeight=2,NegativeHeight=1)
    FalloffSize=0.1
    FlipHeightScaling=False
    FlipWidthScaling=False
    OptionalParentItem=(Type=Bone,Name="None")
    RotationOffset=(X=-46.011,Y=-36.0038,Z=-22.91)
    Debug.bDrawDebug=RigVMVariableNode() {variable=Degub_thigh_r}.Value
    ExecutePin=@RigVMFunction_Sequence_8.B
   )
   > &RigUnit_ModifyTransforms(
    ItemToModify=((Item=(Type=Bone,Name="extra_jaket_fr_r"),Transform=(Rotation=(X=0,Y=0,Z=-0.208,W=0.9781),Translation=(X=1.31,Y=0,Z=0),Scale3D=(X=1,Y=1,Z=1))))
    Weight=@n33.OutputParam
   )
   > &RigUnit_ModifyTransforms(
    ItemToModify=((Item=(Type=Bone,Name="extra_jaket_side_fr_r"),Transform=(Rotation=(X=0.1502,Y=-0.083,Z=-0.399,W=0.9008),Translation=(X=1.2608,Y=1.746,Z=-2.1744),Scale3D=(X=1,Y=1,Z=1))))
    Weight=@n33.OutputParam
   )
   > &RigUnit_ModifyTransforms(
    ItemToModify=((Item=(Type=Bone,Name="extra_jaket_side_r"),Transform=(Rotation=(X=-0.1534,Y=0.0669,Z=-0.2011,W=0.9652),Translation=(X=0,Y=0,Z=0),Scale3D=(X=1,Y=1,Z=1))))
    Weight=@n33.OutputParam
   )
   &RigUnit_ModifyTransforms(
    ExecutePin=()
    ItemToModify=((Item=(Type=Bone,Name="extra_shoulder_pad"),Transform=(Rotation=(X=0.186,Y=-0.1262,Z=0.1678,W=0.9598),Translation=(X=4.6527,Y=1.9063,Z=-1.25),Scale3D=(X=1,Y=1,Z=1))))
    Weight=1
   )
  ]}
]
hierarchy [
 root {loc=<0 0 0>, rot=<0 0 0 1>}
 pelvis {parent=root, loc=<0 -0.7071 93.1191>, rot=<0.0127 0.707 -0.0127 -0.707>}
 spine_01 {parent=pelvis, loc=<6.7494 -0.6388 0>, rot=<0 0 0.034 -0.9994>}
 spine_02 {parent=spine_01, loc=<7.2518 -0.035 0>, rot=<0 0 0.0238 -0.9997>}
 spine_03 {parent=spine_02, loc=<6.0104 0 0>, rot=<0 0 0.0149 -0.9999>}
 spine_04 {parent=spine_03, loc=<8.1024 0 0>, rot=<0 0 -0.0132 -0.9999>}
 spine_05 {parent=spine_04, loc=<10.0146 0 0>, rot=<0 0 -0.0015 -1>}
 attach_cape {parent=spine_05, loc=<9.0178 -11.7147 0.0002>, rot=<0.0616 -0.7044 -0.0616 -0.7044>, scale=<0.9999 1.0001 1>}
 neck_01 {parent=spine_05, loc=<11.3138 0.1529 0>, rot=<0 0 -0.2003 -0.9797>}
 neck_02 {parent=neck_01, loc=<4.5627 0 0>, rot=<0 0 0.0504 -0.9987>}
 head {parent=neck_02, loc=<5.3914 0 0>, rot=<0 0 0.115 -0.9934>}
 attach_backpack {parent=spine_05, loc=<-0.1212 -13.4738 0.0002>, rot=<-0.0093 -0.707 0.0092 -0.707>, scale=<0.9999 1.0001 1>}
 clavicle_l {parent=spine_05, loc=<4.3575 -0.1466 -2.3223>, rot=<0.0465 -0.7289 -0.009 -0.6829>}
 upperarm_l {parent=clavicle_l, loc=<15.5111 0 0>, rot=<0.0226 -0.3234 -0.0088 -0.9459>}
 lowerarm_l {parent=upperarm_l, loc=<26.6514 0 0>, rot=<0.0391 -0.0099 -0.2453 -0.9686>}
 hand_l {parent=lowerarm_l, loc=<22.3491 0 0>, rot=<-0.7081 -0.0507 0.0975 -0.6975>}
 thumb_01_l {parent=hand_l, loc=<1.3118 -0.3191 -2.459>, rot=<0.6546 -0.4042 0.0664 -0.6354>}
 thumb_02_l {parent=thumb_01_l, loc=<5.9336 -0.0448 0.0401>, rot=<-0.0037 0.0213 0.298 -0.9543>}
 thumb_03_l {parent=thumb_02_l, loc=<2.2223 0.3682 -0.0907>, rot=<0.0118 -0.013 0.0066 -0.9998>}
 index_metacarpal_l {parent=hand_l, loc=<1.6535 0.0183 -2.96>, rot=<0.0158 -0.0124 -0.0173 -0.9996>}
 index_01_l {parent=index_metacarpal_l, loc=<8.3008 0.0357 -0.0524>, rot=<-0.0044 0.0485 0.0924 -0.9945>}
 index_02_l {parent=index_01_l, loc=<3.3587 -0.3004 0.0452>, rot=<-0.002 0.0027 0.3225 -0.9466>}
 index_03_l {parent=index_02_l, loc=<2.5578 -0.123 -0.0252>, rot=<-0.002 0.0067 0.2558 -0.9667>}
 middle_metacarpal_l {parent=hand_l, loc=<2.1856 -0.1725 -0.5984>, rot=<0.006 0.0168 -0.0329 -0.9993>}
 middle_01_l {parent=middle_metacarpal_l, loc=<7.3275 -0.2447 0.0129>, rot=<-0.0208 0.0042 0.0858 -0.9961>}
 middle_02_l {parent=middle_01_l, loc=<4.0123 0.1401 0.1175>, rot=<-0.0218 0.0258 0.3512 -0.9357>}
 middle_03_l {parent=middle_02_l, loc=<2.6077 0.0534 0.1373>, rot=<0.0082 0.0188 0.1708 -0.9851>}
 ring_metacarpal_l {parent=hand_l, loc=<2.2074 -0.3625 1.3638>, rot=<-0.0284 0.0722 -0.0458 -0.9959>}
 ring_01_l {parent=ring_metacarpal_l, loc=<7.0295 -0.6197 -0.1446>, rot=<-0.0489 -0.0252 0.1892 -0.9804>}
 ring_02_l {parent=ring_01_l, loc=<3.6174 0.4647 -0.1027>, rot=<0 0 0.2806 -0.9598>}
 ring_03_l {parent=ring_02_l, loc=<2.4152 0.114 0.0199>, rot=<0 0 0.1189 -0.9929>}
 pinky_metacarpal_l {parent=hand_l, loc=<1.7616 -0.9753 3.3814>, rot=<-0.0397 0.0814 -0.069 -0.9935>}
 pinky_01_l {parent=pinky_metacarpal_l, loc=<7.1709 -0.4831 -0.1094>, rot=<-0.0654 -0.0035 0.2242 -0.9723>}
 pinky_02_l {parent=pinky_01_l, loc=<2.8274 0.0893 -0.1043>, rot=<0.0037 0.0011 0.2854 -0.9584>}
 pinky_03_l {parent=pinky_02_l, loc=<2.2046 0.0307 -0.0078>, rot=<-0.0002 0.0098 0.1445 -0.9895>}
 deform_wrist_dn_base_l {parent=hand_l, loc=<0 -2.2342 0>, rot=<0 0 0 1>}
 deform_wrist_dn_l {parent=deform_wrist_dn_base_l, loc=<0 0 0>, rot=<0 0 0 1>}
 deform_wrist_up_base_l {parent=hand_l, loc=<0 2.2682 0>, rot=<0 0 0 1>}
 deform_wrist_up_l {parent=deform_wrist_up_base_l, loc=<0 0 0>, rot=<0 0 0 1>}
 lowerarm_twist_02_l {parent=lowerarm_l, loc=<9.6199 0 0>, rot=<0 0 0 -1>}
 lowerarm_twist_01_l {parent=lowerarm_l, loc=<17.8549 0 0>, rot=<0 0 0 -1>}
 deform_elbow_in_base_l {parent=lowerarm_l, loc=<0.2112 5.259 0.8274>, rot=<0 0 0 -1>}
 deform_elbow_in_l {parent=deform_elbow_in_base_l, loc=<0 0 0>, rot=<0 0 0 1>}
 deform_elbow_out_base_l {parent=lowerarm_l, loc=<0 -4 0>, rot=<0 0 0 -1>}
 deform_elbow_out_l {parent=deform_elbow_out_base_l, loc=<0 0 0>, rot=<0 0 0 1>}
 upperarm_twist_01_l {parent=upperarm_l, loc=<0.0041 0 0>, rot=<0 0 0 1>}
 deform_upperarm_bk_base_l {parent=upperarm_twist_01_l, loc=<1.0813 -7.2193 -1.9453>, rot=<0 0 0 1>}
 deform_upperarm_bk_l {parent=deform_upperarm_bk_base_l, loc=<0 0 0>, rot=<0 0 0 1>}
 deform_upperarm_fr_base_l {parent=upperarm_twist_01_l, loc=<-0.3691 7.0506 -2.765>, rot=<0 0 0 1>}
 deform_upperarm_fr_l {parent=deform_upperarm_fr_base_l, loc=<0 0 0>, rot=<0 0 0 1>}
 deform_upperarm_up_base_l {parent=upperarm_twist_01_l, loc=<-3.2216 -1.1391 3.3311>, rot=<0 0 0 1>}
 deform_upperarm_up_l {parent=deform_upperarm_up_base_l, loc=<0 0 0>, rot=<0 0 0 1>}
 upperarm_twist_02_l {parent=upperarm_l, loc=<13.7941 0 0>, rot=<0 0 0 1>}
 extra_bolt_01 {parent=clavicle_l, loc=<10.8301 -6.7253 8.0875>, rot=<-0.2911 -0.7907 -0.0836 -0.5321>}
 extra_shoulder_strap_05 {parent=clavicle_l, loc=<9.0644 -5.9584 9.6151>, rot=<-0.3081 -0.7297 0.0444 -0.6088>}
 extra_pipe_01 {parent=extra_shoulder_strap_05, loc=<0.7617 0.9888 4.1504>, rot=<0.0636 -0.1232 0.0117 -0.9903>}
 extra_shoulder_strap_01 {parent=clavicle_l, loc=<11.354 3.288 7.603>, rot=<0.1887 -0.7601 0.1403 -0.6057>}
 extra_pipe_02 {parent=extra_shoulder_strap_01, loc=<-0.9253 -0.3228 2.2437>, rot=<0.0191 -0.0856 -0.0455 -0.9951>}
 extra_shoulder_strap_06 {parent=clavicle_l, loc=<3.09 11.9792 3.7827>, rot=<0.5673 -0.6843 -0.1358 -0.4375>}
 extra_shoulder_strap_07 {parent=clavicle_l, loc=<7.0808 9.9514 5.3697>, rot=<0.4639 -0.741 -0.0698 -0.4805>}
 clavicle_r {parent=spine_05, loc=<4.3575 -0.1466 2.3223>, rot=<0.7289 0.0465 -0.6829 0.009>}
 upperarm_r {parent=clavicle_r, loc=<-15.5111 0 0>, rot=<0.0226 -0.3234 -0.0088 -0.9459>}
 lowerarm_r {parent=upperarm_r, loc=<-26.6514 0 0>, rot=<0.0391 -0.0099 -0.2453 -0.9686>}
 hand_r {parent=lowerarm_r, loc=<-22.3491 0 0>, rot=<-0.7081 -0.0507 0.0975 -0.6975>}
 thumb_01_r {parent=hand_r, loc=<-1.3118 0.3191 2.459>, rot=<0.6546 -0.4042 0.0664 -0.6354>}
 thumb_02_r {parent=thumb_01_r, loc=<-5.9336 0.0448 -0.0401>, rot=<-0.0037 0.0213 0.298 -0.9543>}
 thumb_03_r {parent=thumb_02_r, loc=<-2.2223 -0.3682 0.0907>, rot=<0.0118 -0.013 0.0066 -0.9998>}
 index_metacarpal_r {parent=hand_r, loc=<-1.6535 -0.0183 2.96>, rot=<0.0158 -0.0124 -0.0173 -0.9996>}
 index_01_r {parent=index_metacarpal_r, loc=<-8.3008 -0.0357 0.0524>, rot=<-0.0044 0.0485 0.0924 -0.9945>}
 index_02_r {parent=index_01_r, loc=<-3.3587 0.3004 -0.0452>, rot=<-0.002 0.0027 0.3225 -0.9466>}
 index_03_r {parent=index_02_r, loc=<-2.5578 0.123 0.0252>, rot=<-0.002 0.0067 0.2558 -0.9667>}
 middle_metacarpal_r {parent=hand_r, loc=<-2.1856 0.1725 0.5984>, rot=<0.006 0.0168 -0.0329 -0.9993>}
 middle_01_r {parent=middle_metacarpal_r, loc=<-7.3275 0.2447 -0.0129>, rot=<-0.0208 0.0042 0.0858 -0.9961>}
 middle_02_r {parent=middle_01_r, loc=<-4.0123 -0.1401 -0.1175>, rot=<-0.0218 0.0258 0.3512 -0.9357>}
 middle_03_r {parent=middle_02_r, loc=<-2.6077 -0.0534 -0.1373>, rot=<0.0082 0.0188 0.1708 -0.9851>}
 ring_metacarpal_r {parent=hand_r, loc=<-2.2074 0.3625 -1.3638>, rot=<-0.0284 0.0722 -0.0458 -0.9959>}
 ring_01_r {parent=ring_metacarpal_r, loc=<-7.0295 0.6197 0.1446>, rot=<-0.0489 -0.0252 0.1892 -0.9804>}
 ring_02_r {parent=ring_01_r, loc=<-3.6174 -0.4647 0.1027>, rot=<0 0 0.2806 -0.9598>}
 ring_03_r {parent=ring_02_r, loc=<-2.4152 -0.114 -0.0199>, rot=<0 0 0.1189 -0.9929>}
 pinky_metacarpal_r {parent=hand_r, loc=<-1.7616 0.9753 -3.3814>, rot=<-0.0397 0.0814 -0.069 -0.9935>}
 pinky_01_r {parent=pinky_metacarpal_r, loc=<-7.1709 0.4831 0.1094>, rot=<-0.0654 -0.0035 0.2242 -0.9723>}
 pinky_02_r {parent=pinky_01_r, loc=<-2.8274 -0.0893 0.1043>, rot=<0.0037 0.0011 0.2854 -0.9584>}
 pinky_03_r {parent=pinky_02_r, loc=<-2.2046 -0.0307 0.0078>, rot=<-0.0002 0.0098 0.1445 -0.9895>}
 deform_wrist_dn_base_r {parent=hand_r, loc=<-0.0004 2.2345 -0.0001>, rot=<0 0 0 1>}
 deform_wrist_dn_r {parent=deform_wrist_dn_base_r, loc=<0 0 0>, rot=<0 0 0 1>}
 deform_wrist_up_base_r {parent=hand_r, loc=<0.0003 -2.2684 0.0001>, rot=<0 0 0 1>}
 deform_wrist_up_r {parent=deform_wrist_up_base_r, loc=<0 0 0>, rot=<0 0 0 1>}
 lowerarm_twist_02_r {parent=lowerarm_r, loc=<-9.6199 0 0>, rot=<0 0 0 -1>}
 lowerarm_twist_01_r {parent=lowerarm_r, loc=<-17.8549 0 0>, rot=<0 0 0 -1>}
 deform_elbow_in_base_r {parent=lowerarm_r, loc=<-0.2111 -5.2591 -0.8275>, rot=<0 0 0 -1>}
 deform_elbow_in_r {parent=deform_elbow_in_base_r, loc=<-0.0001 0 0.0002>, rot=<0 0 0 1>}
 deform_elbow_out_base_r {parent=lowerarm_r, loc=<0 4 -0.0001>, rot=<0 0 0 -1>}
 deform_elbow_out_r {parent=deform_elbow_out_base_r, loc=<0 0 0>, rot=<0 0 0 1>}
 upperarm_twist_01_r {parent=upperarm_r, loc=<-0.0041 0 0>, rot=<0 0 0 -1>}
 deform_upperarm_bk_base_r {parent=upperarm_twist_01_r, loc=<-1.0814 7.2192 1.9456>, rot=<0 0 0 -1>}
 deform_upperarm_bk_r {parent=deform_upperarm_bk_base_r, loc=<0 0 0>, rot=<0 0 0 1>}
 deform_upperarm_fr_base_r {parent=upperarm_twist_01_r, loc=<0.3689 -7.0506 2.7652>, rot=<0 0 0 -1>}
 deform_upperarm_fr_r {parent=deform_upperarm_fr_base_r, loc=<0 0 0>, rot=<0 0 0 1>}
 deform_upperarm_up_base_r {parent=upperarm_twist_01_r, loc=<3.2214 1.1391 -3.3308>, rot=<0 0 0 -1>}
 deform_upperarm_up_r {parent=deform_upperarm_up_base_r, loc=<0 0 0>, rot=<0 0 0 -1>}
 upperarm_twist_02_r {parent=upperarm_r, loc=<-13.7941 0 0>, rot=<0 0 0 -1>}
 extra_shoulder_pad {parent=upperarm_r, loc=<-1.3409 -0.1533 -4.7404>, rot=<0 0 0 1>}
 extra_shoulder_strap_04 {parent=extra_shoulder_pad, loc=<5.3533 9.539 2.0508>, rot=<0.9283 -0.1876 -0.0771 -0.3118>}
 extra_shoulder_strap_02 {parent=clavicle_r, loc=<-14.7574 -7.8329 -2.9728>, rot=<0.4402 -0.7706 0.2557 0.3834>}
 extra_shoulder_strap_03 {parent=clavicle_r, loc=<-13.8898 10.2279 -4.3474>, rot=<0.0395 0.3591 0.4283 -0.8283>}
 extra_shoulder_strap_08 {parent=clavicle_r, loc=<-5.106 -11.5864 -2.2096>, rot=<0.7528 -0.4702 0.1805 0.4239>}
 extra_shoulder_strap_09 {parent=clavicle_r, loc=<-5.063 12.7495 -3.6865>, rot=<0.2335 0.2578 0.6097 -0.7122>}
 deform_pec_aim_base_l {parent=spine_04, loc=<7.8732 12.6388 -2.4123>, rot=<0.0102 -0.4602 0.3979 -0.7936>}
 deform_pec_aim_l {parent=deform_pec_aim_base_l, loc=<0 0 0>, rot=<0 0 0 1>}
 deform_pec_base_l {parent=deform_pec_aim_l, loc=<10.766 0 0>, rot=<0 0 0 1>}
 deform_pec_l {parent=deform_pec_base_l, loc=<0.055 0 0>, rot=<0 0 0 1>}
 deform_pec_aim_base_r {parent=spine_04, loc=<7.8727 12.6388 2.4123>, rot=<-0.4602 -0.0102 0.7936 0.3979>}
 deform_pec_aim_r {parent=deform_pec_aim_base_r, loc=<0 0 0>, rot=<0 0 0 1>}
 deform_pec_base_r {parent=deform_pec_aim_r, loc=<-10.7662 -0.0006 -0.0007>, rot=<0 0 0 1>}
 deform_pec_r {parent=deform_pec_base_r, loc=<0 0 0>, rot=<0 0 0 1>}
 extra_chest_pocket {parent=spine_04, loc=<2.9206 11.9227 -10.9337>, rot=<0.6887 0.2602 0.6638 0.1317>}
 thigh_l {parent=pelvis, loc=<-2.5693 -0.0109 -7.7501>, rot=<0.0012 0.0572 0.0155 -0.9982>}
 calf_l {parent=thigh_l, loc=<-39.0484 0 0>, rot=<0.0006 -0.0161 -0.0391 -0.9991>}
 foot_l {parent=calf_l, loc=<-42.561 0.1077 0.4292>, rot=<0.0418 -0.0455 0.0343 -0.9975>}
 ball_l {parent=foot_l, loc=<-6.1599 13.8264 0.6806>, rot=<0.0001 -0.0001 -0.7063 -0.7079>}
 calf_twist_02_l {parent=calf_l, loc=<-18.8064 -0.0967 0.0397>, rot=<0.0237 0.0095 -0.0035 -0.9997>}
 calf_twist_01_l {parent=calf_l, loc=<-35.813 -0.0967 0.0397>, rot=<0.0237 0.0095 -0.0035 -0.9997>}
 deform_knee_in_base_l {parent=calf_l, loc=<1.298 -5.2957 -0.0496>, rot=<0 0 0 -1>}
 deform_knee_in_l {parent=deform_knee_in_base_l, loc=<0 0 0>, rot=<0 0 0 1>}
 deform_knee_out_base_l {parent=calf_l, loc=<0 4 0>, rot=<0 0 0 -1>}
 deform_knee_out_l {parent=deform_knee_out_base_l, loc=<0 0 0>, rot=<0 0 0 1>}
 extra_knee_pad {parent=calf_l, loc=<-1.1367 4.7265 -0.7528>, rot=<0 0 0 1>}
 thigh_twist_01_l {parent=thigh_l, loc=<-21.6303 0 0>, rot=<0.0013 0 0 -1>}
 deform_groin_base_l {parent=thigh_l, loc=<-2.1908 8.1142 -1.7442>, rot=<0 0 0 -1>}
 deform_groin_l {parent=deform_groin_base_l, loc=<0 0 0>, rot=<0 0 0 1>}
 dyn_leg_strap_base {parent=thigh_l, loc=<-15.3417 3.5432 -9.1817>, rot=<0.704 -0.6881 -0.1253 -0.1236>}
 dyn_leg_strap {parent=dyn_leg_strap_base, loc=<0.1 0 0>, rot=<0 0 0 1>}
 thigh_r {parent=pelvis, loc=<-2.5693 -0.0109 7.7501>, rot=<0.0572 -0.0012 0.9982 0.0155>}
 calf_r {parent=thigh_r, loc=<39.0484 0 0>, rot=<0.0006 -0.0161 -0.0391 -0.9991>}
 foot_r {parent=calf_r, loc=<42.561 -0.1077 -0.4292>, rot=<0.0418 -0.0455 0.0343 -0.9975>}
 ball_r {parent=foot_r, loc=<6.1599 -13.8264 -0.6806>, rot=<0.0001 -0.0001 -0.7063 -0.7079>}
 calf_twist_02_r {parent=calf_r, loc=<18.8064 0.0967 -0.0397>, rot=<0.0237 0.0095 -0.0035 -0.9997>}
 calf_twist_01_r {parent=calf_r, loc=<35.813 0.0967 -0.0397>, rot=<0.0237 0.0095 -0.0035 -0.9997>}
 deform_knee_in_base_r {parent=calf_r, loc=<-1.298 5.2957 0.0496>, rot=<0 0 0 -1>}
 deform_knee_in_r {parent=deform_knee_in_base_r, loc=<0 0 0>, rot=<0 0 0 1>}
 deform_knee_out_base_r {parent=calf_r, loc=<0 -4 0>, rot=<0 0 0 -1>}
 deform_knee_out_r {parent=deform_knee_out_base_r, loc=<0 0 0>, rot=<0 0 0 1>}
 thigh_twist_01_r {parent=thigh_r, loc=<21.6303 0 0>, rot=<0.0013 0 0 -1>}
 deform_groin_base_r {parent=thigh_r, loc=<2.1908 -8.1142 1.7442>, rot=<0 0 0 -1>}
 deform_groin_r {parent=deform_groin_base_r, loc=<0 0 0>, rot=<0 0 0 1>}
 thigh_r_ctrl {controlType=EULER_TRANSFORM, parent=thigh_r, type=Control, loc=<0 0 0>, rot=<0 0 0 1>}
 deform_glute_base_l {parent=pelvis, loc=<-8.8064 -8.8222 -8.4287>, rot=<0.0275 0.0501 -0.4472 -0.8926>}
 deform_glute_l {parent=deform_glute_base_l, loc=<0 0 0>, rot=<0 0 0 1>}
 deform_glute_base_r {parent=pelvis, loc=<-8.8062 -8.8222 8.4287>, rot=<0.0501 -0.0275 0.8926 -0.4472>}
 deform_glute_r {parent=deform_glute_base_r, loc=<0 0 0>, rot=<0 0 0 1>}
 extra_bag_fr_l {parent=pelvis, loc=<9.4393 6.586 -12.0567>, rot=<-0.2223 -0.6673 -0.1958 -0.6834>}
 dyn_strap_fr_l_base {parent=extra_bag_fr_l, loc=<3.8895 0.3039 -2.1392>, rot=<0.5713 -0.4391 -0.4159 -0.5548>}
 dyn_strap_fr_l {parent=dyn_strap_fr_l_base, loc=<0.1 0 0>, rot=<0 0 0 1>}
 extra_bag_bk_l {parent=pelvis, loc=<9.181 -0.6202 -14.692>, rot=<-0.0344 -0.7178 -0.0194 -0.6951>}
 dyn_strap_bk_l_base {parent=extra_bag_bk_l, loc=<3.924 0.0433 -1.9584>, rot=<0.5754 -0.411 -0.3954 -0.5862>}
 dyn_strap_bk_l {parent=dyn_strap_bk_l_base, loc=<0.0992 -0.0126 0.0002>, rot=<0 0 0 1>}
 extra_bag_fr_r {parent=pelvis, loc=<9.4393 6.586 12.0567>, rot=<0.6834 0.1958 0.6673 0.2223>}
 dyn_strap_fr_r_base {parent=extra_bag_fr_r, loc=<3.8393 -0.3023 -2.141>, rot=<0.5548 -0.4159 -0.4391 -0.5713>}
 dyn_strap_fr_r {parent=dyn_strap_fr_r_base, loc=<0.1001 0 0>, rot=<0 0 0 1>}
 extra_bag_bk_r {parent=pelvis, loc=<9.181 -0.6202 14.692>, rot=<0.6951 0.0194 0.7178 0.0344>}
 dyn_strap_bk_r_base {parent=extra_bag_bk_r, loc=<3.9315 -0.0431 -1.9576>, rot=<0.5862 -0.3954 -0.411 -0.5754>}
 dyn_strap_bk_r {parent=dyn_strap_bk_r_base, loc=<0.1002 0.0001 0>, rot=<0 0 0 1>}
 extra_jaket_fr_r {parent=pelvis, loc=<0.1222 9.124 11.2086>, rot=<-0.086 -0.2044 0.9704 0.0955>}
 dyn_grenade_01_base {parent=extra_jaket_fr_r, loc=<0.1725 -1.6799 0.6529>, rot=<0.0558 0 0 -0.9984>}
 dyn_grenade_01 {parent=dyn_grenade_01_base, loc=<-0.2 0 0>, rot=<0 0 0 1>}
 extra_jaket_side_fr_r {parent=pelvis, loc=<4.5808 3.1742 13.9198>, rot=<0.0789 -0.4993 0.8393 0.2002>}
 dyn_grenade_02_base {parent=extra_jaket_side_fr_r, loc=<6.7575 -0.8098 -2.6239>, rot=<-0.0434 -0.0679 -0.1293 -0.9883>}
 dyn_grenade_02 {parent=dyn_grenade_02_base, loc=<-0.2 0 0>, rot=<0 0 0 1>}
 dyn_grenade_03_base {parent=extra_jaket_side_fr_r, loc=<6.9854 -0.5393 1.6>, rot=<0.1246 0.0348 -0.1354 -0.9823>}
 dyn_grenade_03 {parent=dyn_grenade_03_base, loc=<-0.2 0 0>, rot=<0 0 0 1>}
 dyn_grenade_04_base {parent=extra_jaket_side_fr_r, loc=<5.8442 0.8317 5.4167>, rot=<0.2646 0.1271 -0.1793 -0.939>}
 dyn_grenade_04 {parent=dyn_grenade_04_base, loc=<-0.2 0 0>, rot=<0 0 0 1>}
 extra_jaket_side_fr_r_trg {parent=pelvis, loc=<-1.669 4.9456 16.2095>, rot=<0.0789 -0.4993 0.8393 0.2002>}
 extra_jaket_side_r {parent=pelvis, loc=<4.5791 -1.4161 14.108>, rot=<0.1542 -0.7539 0.6292 0.109>}
 dyn_grenade_05_base {parent=extra_jaket_side_r, loc=<6.3074 -1.4134 2.4113>, rot=<0.8004 0.0892 0.0004 -0.5928>}
 dyn_grenade_05 {parent=dyn_grenade_05_base, loc=<-0.2 0 0>, rot=<0 0 0 1>}
 extra_jaket_side_r_trg {parent=pelvis, loc=<-1.8443 -2.0749 16.5868>, rot=<0.1542 -0.7539 0.6292 0.109>}
 extra_jaket_side_bk_r {parent=pelvis, loc=<5.6941 -7.9961 9.4007>, rot=<0.1561 -0.9374 0.3049 0.0621>}
 extra_jaket_side_bk_r_trg {parent=pelvis, loc=<-1.2268 -9.8599 10.6229>, rot=<0.1561 -0.9374 0.3049 0.0621>}
 extra_jaket_bk {parent=pelvis, loc=<5.7675 -10.1472 -0.0139>, rot=<0.1101 -0.9919 0.0632 0.0093>}
 extra_jaket_bk_trg {parent=pelvis, loc=<-1.9982 -11.8766 0.244>, rot=<0.1091 -0.994 -0.0083 0.0009>}
 extra_jaket_side_bk_l {parent=pelvis, loc=<5.6941 -7.9961 -9.4007>, rot=<0.1468 -0.957 -0.2452 -0.0502>}
 extra_jaket_side_bk_l_trg {parent=pelvis, loc=<-1.5066 -9.9357 -10.6856>, rot=<0.1468 -0.957 -0.2452 -0.0502>}
 extra_jaket_side_l {parent=pelvis, loc=<5.5722 1.7507 -12.8517>, rot=<-0.1129 0.641 0.7463 0.1393>}
 extra_jaket_side_l_trg {parent=pelvis, loc=<-1.5061 2.2413 -15.4821>, rot=<-0.1129 0.641 0.7463 0.1393>}
 extra_jaket_fr {parent=pelvis, loc=<4.9596 10.6623 -0.0113>, rot=<0.0093 -0.0005 0.9995 0.0295>}
 extra_jaket_fr_trg {parent=pelvis, loc=<-2.5925 11.1149 0.1411>, rot=<0.0093 -0.0005 0.9995 0.0295>}
 pelvis_ctrl {controlType=EULER_TRANSFORM, parent=pelvis, type=Control, loc=<0 0 0>, rot=<0 0 0 1>}
 extra_jaket_side_fr_l {parent=pelvis, loc=<4.9502 9.3476 -7.1801>, rot=<-0.0225 0.1766 0.9821 0.0607>}
 arm_up_r {type=Curve, loc=<0 0 0>, rot=<0 0 0 1>}
]
variables [
 Alpha_crouch {type=double, default=0}
 Alpha_parachite {type=double, default=0}
 Alpha_targeting {type=double, default=0}
 Degub_arm_l {type=bool, default=False}
 Degub_arm_r {type=bool, default=True}
 Degub_thigh_l {type=bool, default=False}
 Degub_thigh_r {type=bool, default=False}
 crouch {type=bool, default=False}
 targeting {type=bool, default=False}
]
