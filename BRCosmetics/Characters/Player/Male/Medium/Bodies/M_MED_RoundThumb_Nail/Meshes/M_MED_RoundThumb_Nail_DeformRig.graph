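# Deformation Control Rig for the M_MED_RoundThumb_Nail medium male player cosmetic body, built to adjust corrective deform joints.
# The forward solve is a single call to the shared DeformRig function, configured with curve switches, per-region toggles, debug flags and tuned offset transforms.
# The hierarchy adds twist bones and base/child deform pairs at major joints, pecs and glutes, plus outfit-specific extra and dyn bones for straps, grenades, bags and jacket.
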ControlRigBlueprint /BRCosmetics/Characters/Player/Male/Medium/Bodies/M_MED_RoundThumb_Nail/Meshes/M_MED_RoundThumb_Nail_DeformRig
graphs [
 RigVMFunctionLibrary {links [], nodes []}
 RigVMModel {nodes [
   RigUnit_BeginExecution = &RigUnit_BeginExecution()
   FN_DeformRig_FNC = RigVMFunctionReferenceNode(
    DeformRig_ Curve="(deformRig_l_upperarm_up_alpha_72_#1=False,deformRig_r_upperarm_up_alpha_73_#2=False,deformRig_l_upperarm_fr_alpha_74_#3=False,deformRig_r_upperarm_fr_alpha_75_CB51B8 ...(truncated)"
    DeformRig_Bool="(upperarm_1_#4=true,elbow_3_#5=True,wrist_7_#6=True,groin_8_#7=True,glute_17_#8=True,knee_9_40C408324CA261 ...(truncated)"
    DeformRig_Debug=(upperarm_debug_18_#4=False,wrist_debug_20_#6=False,lat_debug_26_#9=False,groinGlute_debug_27_#7=False)
    DeformRig_Value="(upperarm_up_18_#10=(Rotation=(X=0,Y=0,Z=0,W=1),Translation=(X=0,Y=0,Z=8),Scale3D=(X=0.7,Y=1,Z=1)),upperarm_fwd_fr_19_#11=(Rotation=(X=0,Y=0,Z=0,W=1),Translation=(X=0,Y=10,Z=0),Scale3D=(X=0.5,Y=1,Z= ...(truncated)"
    ExecuteContext=@RigUnit_BeginExecution.ExecutePin
   )
  ]}
]
hierarchy [
 root {loc=<0 0 0>, rot=<0 0 0 1>}
 pelvis {parent=root, loc=<0 -0.7071 93.1191>, rot=<0.0127 0.707 -0.0127 -0.707>}
 spine_01 {parent=pelvis, loc=<6.7494 -0.6388 0>, rot=<0 0 0.034 -0.9994>}
 spine_02 {parent=spine_01, loc=<7.2518 -0.035 0>, rot=<0 0 0.0238 -0.9997>}
 spine_03 {parent=spine_02, loc=<6.0104 0 0>, rot=<0 0 0.0149 -0.9999>}
 spine_04 {parent=spine_03, loc=<8.1024 0 0>, rot=<0 0 -0.0132 -0.9999>}
 spine_05 {parent=spine_04, loc=<10.0146 0 0>, rot=<0 0 -0.0015 -1>}
 attach_cape {parent=spine_05, loc=<9.017 -11.714 0>, rot=<0.0616 -0.7044 -0.0616 -0.7044>}
 neck_01 {parent=spine_05, loc=<11.3138 0.1529 0>, rot=<0 0 -0.2003 -0.9797>}
 neck_02 {parent=neck_01, loc=<4.5627 0 0>, rot=<0 0 0.0504 -0.9987>}
 head {parent=neck_02, loc=<5.3914 0 0>, rot=<0 0 0.115 -0.9934>}
 attach_backpack {parent=spine_05, loc=<-0.123 -13.473 0>, rot=<-0.0093 -0.707 0.0093 -0.707>}
 clavicle_l {parent=spine_05, loc=<4.3575 -0.1466 -2.3223>, rot=<0.0465 -0.7289 -0.009 -0.6829>}
 upperarm_l {parent=clavicle_l, loc=<15.5111 0 0>, rot=<0.0226 -0.3234 -0.0088 -0.9459>}
 lowerarm_l {parent=upperarm_l, loc=<26.6514 0 0>, rot=<0.0391 -0.0099 -0.2453 -0.9686>}
 hand_l {parent=lowerarm_l, loc=<22.3491 0 0>, rot=<-0.7081 -0.0507 0.0975 -0.6975>}
 thumb_01_l {parent=hand_l, loc=<1.3118 -0.3191 -2.459>, rot=<0.6546 -0.4042 0.0664 -0.6354>}
 thumb_02_l {parent=thumb_01_l, loc=<5.9336 -0.0448 0.0401>, rot=<-0.0037 0.0213 0.298 -0.9543>}
 thumb_03_l {parent=thumb_02_l, loc=<2.2223 0.3682 -0.0907>, rot=<0.0118 -0.013 0.0066 -0.9998>}
 index_metacarpal_l {parent=hand_l, loc=<1.6535 0.0183 -2.96>, rot=<0.0158 -0.0124 -0.0173 -0.9996>}
 index_01_l {parent=index_metacarpal_l, loc=<8.3008 0.0357 -0.0524>, rot=<-0.0044 0.0485 0.0924 -0.9945>}
 index_02_l {parent=index_01_l, loc=<3.3587 -0.3004 0.0452>, rot=<-0.002 0.0027 0.3225 -0.9466>}
 index_03_l {parent=index_02_l, loc=<2.5578 -0.123 -0.0252>, rot=<-0.002 0.0067 0.2558 -0.9667>}
 middle_metacarpal_l {parent=hand_l, loc=<2.1856 -0.1725 -0.5984>, rot=<0.006 0.0168 -0.0329 -0.9993>}
 middle_01_l {parent=middle_metacarpal_l, loc=<7.3275 -0.2447 0.0129>, rot=<-0.0208 0.0042 0.0858 -0.9961>}
 middle_02_l {parent=middle_01_l, loc=<4.0123 0.1401 0.1175>, rot=<-0.0218 0.0258 0.3512 -0.9357>}
 middle_03_l {parent=middle_02_l, loc=<2.6077 0.0534 0.1373>, rot=<0.0082 0.0188 0.1708 -0.9851>}
 ring_metacarpal_l {parent=hand_l, loc=<2.2074 -0.3625 1.3638>, rot=<-0.0284 0.0722 -0.0458 -0.9959>}
 ring_01_l {parent=ring_metacarpal_l, loc=<7.0295 -0.6197 -0.1446>, rot=<-0.0489 -0.0252 0.1892 -0.9804>}
 ring_02_l {parent=ring_01_l, loc=<3.6174 0.4647 -0.1027>, rot=<0 0 0.2806 -0.9598>}
 ring_03_l {parent=ring_02_l, loc=<2.4152 0.114 0.0199>, rot=<0 0 0.1189 -0.9929>}
 pinky_metacarpal_l {parent=hand_l, loc=<1.7616 -0.9753 3.3814>, rot=<-0.0397 0.0814 -0.069 -0.9935>}
 pinky_01_l {parent=pinky_metacarpal_l, loc=<7.1709 -0.4831 -0.1094>, rot=<-0.0654 -0.0035 0.2242 -0.9723>}
 pinky_02_l {parent=pinky_01_l, loc=<2.8274 0.0893 -0.1043>, rot=<0.0037 0.0011 0.2854 -0.9584>}
 pinky_03_l {parent=pinky_02_l, loc=<2.2046 0.0307 -0.0078>, rot=<-0.0002 0.0098 0.1445 -0.9895>}
 deform_wrist_dn_base_l {parent=hand_l, loc=<0 -2.2342 0>, rot=<0 0 0 1>}
 deform_wrist_dn_l {parent=deform_wrist_dn_base_l, loc=<0 0 0>, rot=<0 0 0 1>}
 deform_wrist_up_base_l {parent=hand_l, loc=<0 2.2682 0>, rot=<0 0 0 1>}
 deform_wrist_up_l {parent=deform_wrist_up_base_l, loc=<0 0 0>, rot=<0 0 0 1>}
 lowerarm_twist_02_l {parent=lowerarm_l, loc=<9.6199 0 0>, rot=<0 0 0 -1>}
 lowerarm_twist_01_l {parent=lowerarm_l, loc=<17.8549 0 0>, rot=<0 0 0 -1>}
 deform_elbow_in_base_l {parent=lowerarm_l, loc=<0.2112 5.259 0.8274>, rot=<0 0 0 -1>}
 deform_elbow_in_l {parent=deform_elbow_in_base_l, loc=<0 0 0>, rot=<0 0 0 1>}
 deform_elbow_out_base_l {parent=lowerarm_l, loc=<0 -4 0>, rot=<0 0 0 -1>}
 deform_elbow_out_l {parent=deform_elbow_out_base_l, loc=<0 0 0>, rot=<0 0 0 1>}
 upperarm_twist_01_l {parent=upperarm_l, loc=<0.0041 0 0>, rot=<0 0 0 1>}
 deform_upperarm_bk_base_l {parent=upperarm_twist_01_l, loc=<1.0813 -7.2193 -1.9453>, rot=<0 0 0 1>}
 deform_upperarm_bk_l {parent=deform_upperarm_bk_base_l, loc=<0 0 0>, rot=<0 0 0 1>}
 deform_upperarm_fr_base_l {parent=upperarm_twist_01_l, loc=<-0.3691 7.0506 -2.765>, rot=<0 0 0 1>}
 deform_upperarm_fr_l {parent=deform_upperarm_fr_base_l, loc=<0 0 0>, rot=<0 0 0 1>}
 deform_upperarm_up_base_l {parent=upperarm_twist_01_l, loc=<-3.2216 -1.1391 3.3311>, rot=<0 0 0 1>}
 deform_upperarm_up_l {parent=deform_upperarm_up_base_l, loc=<0 0 0>, rot=<0 0 0 1>}
 upperarm_twist_02_l {parent=upperarm_l, loc=<13.7941 0 0>, rot=<0 0 0 1>}
 extra_shoulder_strap_01 {parent=clavicle_l, loc=<11.354 3.288 7.603>, rot=<0.1887 -0.7601 0.1403 -0.6057>}
 extra_pipe_02 {parent=extra_shoulder_strap_01, loc=<-0.9253 -0.3228 2.2437>, rot=<0.0191 -0.0856 -0.0455 -0.9951>}
 extra_shoulder_strap_05 {parent=clavicle_l, loc=<9.0644 -5.9584 9.6151>, rot=<-0.3081 -0.7297 0.0444 -0.6088>}
 extra_pipe_01 {parent=extra_shoulder_strap_05, loc=<0.7617 0.9888 4.1504>, rot=<0.0636 -0.1232 0.0117 -0.9903>}
 extra_bolt_01 {parent=clavicle_l, loc=<10.8301 -6.7253 8.0875>, rot=<-0.2911 -0.7907 -0.0836 -0.5321>}
 extra_shoulder_strap_06 {parent=clavicle_l, loc=<3.09 11.9792 3.7827>, rot=<0.5673 -0.6843 -0.1358 -0.4375>}
 extra_shoulder_strap_07 {parent=clavicle_l, loc=<7.0808 9.9514 5.3697>, rot=<0.4639 -0.741 -0.0698 -0.4805>}
 clavicle_r {parent=spine_05, loc=<4.3575 -0.1466 2.3223>, rot=<0.7289 0.0465 -0.6829 0.009>}
 upperarm_r {parent=clavicle_r, loc=<-15.5111 0 0>, rot=<0.0226 -0.3234 -0.0088 -0.9459>}
 lowerarm_r {parent=upperarm_r, loc=<-26.6514 0 0>, rot=<0.0391 -0.0099 -0.2453 -0.9686>}
 hand_r {parent=lowerarm_r, loc=<-22.3491 0 0>, rot=<-0.7081 -0.0507 0.0975 -0.6975>}
 thumb_01_r {parent=hand_r, loc=<-1.3118 0.3191 2.459>, rot=<0.6546 -0.4042 0.0664 -0.6354>}
 thumb_02_r {parent=thumb_01_r, loc=<-5.9336 0.0448 -0.0401>, rot=<-0.0037 0.0213 0.298 -0.9543>}
 thumb_03_r {parent=thumb_02_r, loc=<-2.2223 -0.3682 0.0907>, rot=<0.0118 -0.013 0.0066 -0.9998>}
 index_metacarpal_r {parent=hand_r, loc=<-1.6535 -0.0183 2.96>, rot=<0.0158 -0.0124 -0.0173 -0.9996>}
 index_01_r {parent=index_metacarpal_r, loc=<-8.3008 -0.0357 0.0524>, rot=<-0.0044 0.0485 0.0924 -0.9945>}
 index_02_r {parent=index_01_r, loc=<-3.3587 0.3004 -0.0452>, rot=<-0.002 0.0027 0.3225 -0.9466>}
 index_03_r {parent=index_02_r, loc=<-2.5578 0.123 0.0252>, rot=<-0.002 0.0067 0.2558 -0.9667>}
 middle_metacarpal_r {parent=hand_r, loc=<-2.1856 0.1725 0.5984>, rot=<0.006 0.0168 -0.0329 -0.9993>}
 middle_01_r {parent=middle_metacarpal_r, loc=<-7.3275 0.2447 -0.0129>, rot=<-0.0208 0.0042 0.0858 -0.9961>}
 middle_02_r {parent=middle_01_r, loc=<-4.0123 -0.1401 -0.1175>, rot=<-0.0218 0.0258 0.3512 -0.9357>}
 middle_03_r {parent=middle_02_r, loc=<-2.6077 -0.0534 -0.1373>, rot=<0.0082 0.0188 0.1708 -0.9851>}
 ring_metacarpal_r {parent=hand_r, loc=<-2.2074 0.3625 -1.3638>, rot=<-0.0284 0.0722 -0.0458 -0.9959>}
 ring_01_r {parent=ring_metacarpal_r, loc=<-7.0295 0.6197 0.1446>, rot=<-0.0489 -0.0252 0.1892 -0.9804>}
 ring_02_r {parent=ring_01_r, loc=<-3.6174 -0.4647 0.1027>, rot=<0 0 0.2806 -0.9598>}
 ring_03_r {parent=ring_02_r, loc=<-2.4152 -0.114 -0.0199>, rot=<0 0 0.1189 -0.9929>}
 pinky_metacarpal_r {parent=hand_r, loc=<-1.7616 0.9753 -3.3814>, rot=<-0.0397 0.0814 -0.069 -0.9935>}
 pinky_01_r {parent=pinky_metacarpal_r, loc=<-7.1709 0.4831 0.1094>, rot=<-0.0654 -0.0035 0.2242 -0.9723>}
 pinky_02_r {parent=pinky_01_r, loc=<-2.8274 -0.0893 0.1043>, rot=<0.0037 0.0011 0.2854 -0.9584>}
 pinky_03_r {parent=pinky_02_r, loc=<-2.2046 -0.0307 0.0078>, rot=<-0.0002 0.0098 0.1445 -0.9895>}
 deform_wrist_dn_base_r {parent=hand_r, loc=<-0.0004 2.2345 -0.0001>, rot=<0 0 0 1>}
 deform_wrist_dn_r {parent=deform_wrist_dn_base_r, loc=<0 0 0>, rot=<0 0 0 1>}
 deform_wrist_up_base_r {parent=hand_r, loc=<0.0003 -2.2684 0.0001>, rot=<0 0 0 1>}
 deform_wrist_up_r {parent=deform_wrist_up_base_r, loc=<0 0 0>, rot=<0 0 0 1>}
 lowerarm_twist_02_r {parent=lowerarm_r, loc=<-9.6199 0 0>, rot=<0 0 0 -1>}
 lowerarm_twist_01_r {parent=lowerarm_r, loc=<-17.8549 0 0>, rot=<0 0 0 -1>}
 deform_elbow_in_base_r {parent=lowerarm_r, loc=<-0.2111 -5.2591 -0.8275>, rot=<0 0 0 -1>}
 deform_elbow_in_r {parent=deform_elbow_in_base_r, loc=<-0.0001 0 0.0002>, rot=<0 0 0 1>}
 deform_elbow_out_base_r {parent=lowerarm_r, loc=<0 4 -0.0001>, rot=<0 0 0 -1>}
 deform_elbow_out_r {parent=deform_elbow_out_base_r, loc=<0 0 0>, rot=<0 0 0 1>}
 upperarm_twist_01_r {parent=upperarm_r, loc=<-0.0041 0 0>, rot=<0 0 0 -1>}
 deform_upperarm_bk_base_r {parent=upperarm_twist_01_r, loc=<-1.0814 7.2192 1.9456>, rot=<0 0 0 -1>}
 deform_upperarm_bk_r {parent=deform_upperarm_bk_base_r, loc=<0 0 0>, rot=<0 0 0 1>}
 deform_upperarm_fr_base_r {parent=upperarm_twist_01_r, loc=<0.3689 -7.0506 2.7652>, rot=<0 0 0 -1>}
 deform_upperarm_fr_r {parent=deform_upperarm_fr_base_r, loc=<0 0 0>, rot=<0 0 0 1>}
 deform_upperarm_up_base_r {parent=upperarm_twist_01_r, loc=<3.2214 1.1391 -3.3308>, rot=<0 0 0 -1>}
 deform_upperarm_up_r {parent=deform_upperarm_up_base_r, loc=<0 0 0>, rot=<0 0 0 -1>}
 upperarm_twist_02_r {parent=upperarm_r, loc=<-13.7941 0 0>, rot=<0 0 0 -1>}
 extra_shoulder_pad {parent=upperarm_r, loc=<-1.3409 -0.1533 -4.7404>, rot=<0 0 0 1>}
 extra_shoulder_strap_04 {parent=extra_shoulder_pad, loc=<5.3533 9.539 2.0508>, rot=<0.9283 -0.1876 -0.0771 -0.3118>}
 extra_shoulder_strap_02 {parent=clavicle_r, loc=<-14.7574 -7.8329 -2.9728>, rot=<0.4402 -0.7706 0.2557 0.3834>}
 extra_shoulder_strap_03 {parent=clavicle_r, loc=<-13.8898 10.2279 -4.3474>, rot=<0.0395 0.3591 0.4283 -0.8283>}
 extra_shoulder_strap_08 {parent=clavicle_r, loc=<-5.106 -11.5864 -2.2096>, rot=<0.5645 -0.6849 0.3058 0.3446>}
 extra_shoulder_strap_09 {parent=clavicle_r, loc=<-5.063 12.7495 -3.6865>, rot=<0.2335 0.2578 0.6097 -0.7122>}
 deform_pec_aim_base_l {parent=spine_04, loc=<7.8732 12.6388 -2.4123>, rot=<0.0102 -0.4602 0.3979 -0.7936>}
 deform_pec_aim_l {parent=deform_pec_aim_base_l, loc=<0 0 0>, rot=<0 0 0 1>}
 deform_pec_base_l {parent=deform_pec_aim_l, loc=<10.766 0 0>, rot=<0 0 0 1>}
 deform_pec_l {parent=deform_pec_base_l, loc=<0.055 0 0>, rot=<0 0 0 1>}
 deform_pec_aim_base_r {parent=spine_04, loc=<7.8727 12.6388 2.4123>, rot=<-0.4602 -0.0102 0.7936 0.3979>}
 deform_pec_aim_r {parent=deform_pec_aim_base_r, loc=<0 0 0>, rot=<0 0 0 1>}
 deform_pec_base_r {parent=deform_pec_aim_r, loc=<-10.7662 -0.0006 -0.0007>, rot=<0 0 0 1>}
 deform_pec_r {parent=deform_pec_base_r, loc=<0 0 0>, rot=<0 0 0 1>}
 extra_chest_pocket {parent=spine_04, loc=<2.9206 11.9227 -10.9337>, rot=<0.6887 0.2602 0.6638 0.1317>}
 thigh_l {parent=pelvis, loc=<-2.5693 -0.0109 -7.7501>, rot=<0.0012 0.0572 0.0155 -0.9982>}
 calf_l {parent=thigh_l, loc=<-39.0484 0 0>, rot=<0.0006 -0.0161 -0.0391 -0.9991>}
 foot_l {parent=calf_l, loc=<-42.561 0.1077 0.4292>, rot=<0.0418 -0.0455 0.0343 -0.9975>}
 ball_l {parent=foot_l, loc=<-6.1599 13.8264 0.6806>, rot=<0.0001 -0.0001 -0.7063 -0.7079>}
 calf_twist_02_l {parent=calf_l, loc=<-18.8064 -0.0967 0.0397>, rot=<0.0237 0.0095 -0.0035 -0.9997>}
 calf_twist_01_l {parent=calf_l, loc=<-35.813 -0.0967 0.0397>, rot=<0.0237 0.0095 -0.0035 -0.9997>}
 deform_knee_in_base_l {parent=calf_l, loc=<1.298 -5.2957 -0.0496>, rot=<0 0 0 -1>}
 deform_knee_in_l {parent=deform_knee_in_base_l, loc=<0 0 0>, rot=<0 0 0 1>}
 deform_knee_out_base_l {parent=calf_l, loc=<0 4 0>, rot=<0 0 0 -1>}
 deform_knee_out_l {parent=deform_knee_out_base_l, loc=<0 0 0>, rot=<0 0 0 1>}
 extra_knee_pad {parent=calf_l, loc=<-1.1367 4.7265 -0.7528>, rot=<0 0 0 1>}
 thigh_twist_01_l {parent=thigh_l, loc=<-21.6303 0 0>, rot=<0.0013 0 0 -1>}
 deform_groin_base_l {parent=thigh_l, loc=<-2.1908 8.1142 -1.7442>, rot=<0 0 0 -1>}
 deform_groin_l {parent=deform_groin_base_l, loc=<0 0 0>, rot=<0 0 0 1>}
 dyn_leg_strap_base {parent=thigh_l, loc=<-15.3417 3.5432 -9.1817>, rot=<0.704 -0.6881 -0.1253 -0.1236>}
 dyn_leg_strap {parent=dyn_leg_strap_base, loc=<0.1 0 0>, rot=<0 0 0 1>}
 thigh_r {parent=pelvis, loc=<-2.5693 -0.0109 7.7501>, rot=<0.0572 -0.0012 0.9982 0.0155>}
 calf_r {parent=thigh_r, loc=<39.0484 0 0>, rot=<0.0006 -0.0161 -0.0391 -0.9991>}
 foot_r {parent=calf_r, loc=<42.561 -0.1077 -0.4292>, rot=<0.0418 -0.0455 0.0343 -0.9975>}
 ball_r {parent=foot_r, loc=<6.1599 -13.8264 -0.6806>, rot=<0.0001 -0.0001 -0.7063 -0.7079>}
 calf_twist_02_r {parent=calf_r, loc=<18.8064 0.0967 -0.0397>, rot=<0.0237 0.0095 -0.0035 -0.9997>}
 calf_twist_01_r {parent=calf_r, loc=<35.813 0.0967 -0.0397>, rot=<0.0237 0.0095 -0.0035 -0.9997>}
 deform_knee_in_base_r {parent=calf_r, loc=<-1.298 5.2957 0.0496>, rot=<0 0 0 -1>}
 deform_knee_in_r {parent=deform_knee_in_base_r, loc=<0 0 0>, rot=<0 0 0 1>}
 deform_knee_out_base_r {parent=calf_r, loc=<0 -4 0>, rot=<0 0 0 -1>}
 deform_knee_out_r {parent=deform_knee_out_base_r, loc=<0 0 0>, rot=<0 0 0 1>}
 thigh_twist_01_r {parent=thigh_r, loc=<21.6303 0 0>, rot=<0.0013 0 0 -1>}
 deform_groin_base_r {parent=thigh_r, loc=<2.1908 -8.1142 1.7442>, rot=<0 0 0 -1>}
 deform_groin_r {parent=deform_groin_base_r, loc=<0 0 0>, rot=<0 0 0 1>}
 deform_glute_base_l {parent=pelvis, loc=<-8.8064 -8.8222 -8.4287>, rot=<0.0275 0.0501 -0.4472 -0.8926>}
 deform_glute_l {parent=deform_glute_base_l, loc=<0 0 0>, rot=<0 0 0 1>}
 deform_glute_base_r {parent=pelvis, loc=<-8.8062 -8.8222 8.4287>, rot=<0.0501 -0.0275 0.8926 -0.4472>}
 deform_glute_r {parent=deform_glute_base_r, loc=<0 0 0>, rot=<0 0 0 1>}
 extra_jaket_fr_r {parent=pelvis, loc=<0.1222 9.124 11.2086>, rot=<-0.086 -0.2044 0.9704 0.0955>}
 dyn_grenade_01_base {parent=extra_jaket_fr_r, loc=<0.1725 -1.6799 0.6529>, rot=<0.0558 0 0 -0.9984>}
 dyn_grenade_01 {parent=dyn_grenade_01_base, loc=<-0.2 0 0>, rot=<0 0 0 1>}
 extra_jaket_side_fr_r {parent=pelvis, loc=<4.5808 3.1742 13.9198>, rot=<0.0789 -0.4993 0.8393 0.2002>}
 dyn_grenade_02_base {parent=extra_jaket_side_fr_r, loc=<6.7575 -0.8098 -2.6239>, rot=<-0.0434 -0.0679 -0.1293 -0.9883>}
 dyn_grenade_02 {parent=dyn_grenade_02_base, loc=<-0.2 0 0>, rot=<0 0 0 1>}
 dyn_grenade_03_base {parent=extra_jaket_side_fr_r, loc=<6.9854 -0.5393 1.6>, rot=<0.1246 0.0348 -0.1354 -0.9823>}
 dyn_grenade_03 {parent=dyn_grenade_03_base, loc=<-0.2 0 0>, rot=<0 0 0 1>}
 dyn_grenade_04_base {parent=extra_jaket_side_fr_r, loc=<5.8442 0.8317 5.4167>, rot=<0.2646 0.1271 -0.1793 -0.939>}
 dyn_grenade_04 {parent=dyn_grenade_04_base, loc=<-0.2 0 0>, rot=<0 0 0 1>}
 extra_jaket_side_fr_r_trg {parent=pelvis, loc=<-1.669 4.9456 16.2095>, rot=<0.0789 -0.4993 0.8393 0.2002>}
 extra_bag_fr_l {parent=pelvis, loc=<9.4393 6.586 -12.0567>, rot=<-0.2223 -0.6673 -0.1958 -0.6834>}
 dyn_strap_fr_l_base {parent=extra_bag_fr_l, loc=<3.8895 0.3039 -2.1392>, rot=<0.5713 -0.4391 -0.4159 -0.5548>}
 dyn_strap_fr_l {parent=dyn_strap_fr_l_base, loc=<0.1 0 0>, rot=<0 0 0 1>}
 extra_bag_bk_l {parent=pelvis, loc=<9.181 -0.6202 -14.692>, rot=<-0.0344 -0.7178 -0.0194 -0.6951>}
 dyn_strap_bk_l_base {parent=extra_bag_bk_l, loc=<3.924 0.0433 -1.9584>, rot=<0.5754 -0.411 -0.3954 -0.5862>}
 dyn_strap_bk_l {parent=dyn_strap_bk_l_base, loc=<0.0992 -0.0126 0.0002>, rot=<0 0 0 1>}
 extra_bag_fr_r {parent=pelvis, loc=<9.4393 6.586 12.0567>, rot=<0.6834 0.1958 0.6673 0.2223>}
 dyn_strap_fr_r_base {parent=extra_bag_fr_r, loc=<3.8393 -0.3023 -2.141>, rot=<0.5548 -0.4159 -0.4391 -0.5713>}
 dyn_strap_fr_r {parent=dyn_strap_fr_r_base, loc=<0.1001 0 0>, rot=<0 0 0 1>}
 extra_bag_bk_r {parent=pelvis, loc=<9.181 -0.6202 14.692>, rot=<0.6951 0.0194 0.7178 0.0344>}
 dyn_strap_bk_r_base {parent=extra_bag_bk_r, loc=<3.9315 -0.0431 -1.9576>, rot=<0.5862 -0.3954 -0.411 -0.5754>}
 dyn_strap_bk_r {parent=dyn_strap_bk_r_base, loc=<0.1002 0.0001 0>, rot=<0 0 0 1>}
 extra_jaket_side_r {parent=pelvis, loc=<4.5791 -1.4161 14.108>, rot=<0.1542 -0.7539 0.6292 0.109>}
 dyn_grenade_05_base {parent=extra_jaket_side_r, loc=<6.3074 -1.4134 2.4113>, rot=<0.8004 0.0892 0.0004 -0.5928>}
 dyn_grenade_05 {parent=dyn_grenade_05_base, loc=<-0.2 0 0>, rot=<0 0 0 1>}
 extra_jaket_side_r_trg {parent=pelvis, loc=<-1.8443 -2.0749 16.5868>, rot=<0.1542 -0.7539 0.6292 0.109>}
 extra_jaket_side_bk_r {parent=pelvis, loc=<5.6941 -7.9961 9.4007>, rot=<0.1561 -0.9374 0.3049 0.0621>}
 extra_jaket_side_bk_r_trg {parent=pelvis, loc=<-1.2268 -9.8599 10.6229>, rot=<0.1561 -0.9374 0.3049 0.0621>}
 extra_jaket_bk {parent=pelvis, loc=<5.7675 -10.1472 -0.0139>, rot=<0.1101 -0.9919 0.0632 0.0093>}
 extra_jaket_bk_trg {parent=pelvis, loc=<-1.9982 -11.8766 0.244>, rot=<0.1091 -0.994 -0.0083 0.0009>}
 extra_jaket_side_bk_l {parent=pelvis, loc=<5.6941 -7.9961 -9.4007>, rot=<0.1468 -0.957 -0.2452 -0.0502>}
 extra_jaket_side_bk_l_trg {parent=pelvis, loc=<-1.5066 -9.9357 -10.6856>, rot=<0.1468 -0.957 -0.2452 -0.0502>}
 extra_jaket_side_l {parent=pelvis, loc=<5.5722 1.7507 -12.8517>, rot=<-0.1129 0.641 0.7463 0.1393>}
 extra_jaket_side_l_trg {parent=pelvis, loc=<-1.5061 2.2413 -15.4821>, rot=<-0.1129 0.641 0.7463 0.1393>}
 extra_jaket_fr {parent=pelvis, loc=<4.9596 10.6623 -0.0113>, rot=<0.0093 -0.0005 0.9995 0.0295>}
 extra_jaket_fr_trg {parent=pelvis, loc=<-2.5925 11.1149 0.1411>, rot=<0.0093 -0.0005 0.9995 0.0295>}
 extra_jaket_side_fr_l {parent=pelvis, loc=<4.9502 9.3476 -7.1801>, rot=<-0.0225 0.1766 0.9821 0.0607>}
]
