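# Physics asset for the RoundThumb Nail medium male player body cosmetic, covering its dangling gear: five grenades, four bag straps, four bags and a leg strap.
# Each grenade and strap is a capsule pair, with the dyn_ bone held to its _base bone by narrow swing and twist limits.
# Grenade and leg-strap bases and the bags constrain to the pelvis capsule, and bag-strap bases to their bags, all with free rotation.
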PhysicsAsset /BRCosmetics/Characters/Player/Male/Medium/Bodies/M_MED_RoundThumb_Nail/Meshes/M_MED_RoundThumb_Nail_Physics
bodies {
 dyn_grenade_01 [{shapeType=Capsule, length=4, radius=2, shapeName=dyn_grenade_01_capsule, center=<2 0 0>, rotation=<rot 0 -90 -90>}]
 dyn_grenade_01_base [{shapeType=Capsule, length=0.505, radius=0.505, shapeName=dyn_grenade_01_base_capsule, rotation=<rot 0 -90 0>}]
 dyn_grenade_02 [{shapeType=Capsule, length=4, radius=2, shapeName=dyn_grenade_02_capsule, center=<2 0 0>, rotation=<rot 0 -90 -90>}]
 dyn_grenade_02_base [{shapeType=Capsule, length=0.505, radius=0.505, shapeName=dyn_grenade_02_base_capsule, rotation=<rot 0 -90 0>}]
 dyn_grenade_03 [{shapeType=Capsule, length=4, radius=2, shapeName=dyn_grenade_03_capsule, center=<2 0 0>, rotation=<rot 0 -90 -90>}]
 dyn_grenade_03_base [{shapeType=Capsule, length=0.505, radius=0.505, shapeName=dyn_grenade_03_base_capsule, rotation=<rot 0 -90 0>}]
 dyn_grenade_04 [{shapeType=Capsule, length=4, radius=2, shapeName=dyn_grenade_04_capsule, center=<2 0 0>, rotation=<rot 0 -90 -90>}]
 dyn_grenade_04_base [{shapeType=Capsule, length=0.505, radius=0.505, shapeName=dyn_grenade_04_base_capsule, rotation=<rot 0 -90 0>}]
 dyn_grenade_05 [{shapeType=Capsule, length=4, radius=2, shapeName=dyn_grenade_05_capsule, center=<2 0 0>, rotation=<rot 0 -90 -90>}]
 dyn_grenade_05_base [{shapeType=Capsule, length=0.505, radius=0.505, shapeName=dyn_grenade_05_base_capsule, rotation=<rot 0 -90 0>}]
 dyn_leg_strap [{shapeType=Capsule, length=4, radius=2, shapeName=dyn_leg_strap_capsule, center=<4 0 1>, rotation=<rot 0 -90 -90>}]
 dyn_leg_strap_base [{shapeType=Capsule, length=0.505, radius=0.505, shapeName=dyn_leg_strap_base_capsule, rotation=<rot 0 -90 0>}]
 dyn_strap_bk_l [{shapeType=Capsule, length=4, radius=1, shapeName=dyn_strap_bk_l_capsule, center=<4 -1 0>, rotation=<rot 0 90 -90>}]
 dyn_strap_bk_l_base [{shapeType=Capsule, length=0.505, radius=0.505, shapeName=dyn_strap_bk_l_base_capsule, rotation=<rot 0 -90 0>}]
 dyn_strap_bk_r [{shapeType=Capsule, length=4, radius=1, shapeName=dyn_strap_bk_r_capsule, center=<4 -1 0>, rotation=<rot 0 90 -90>}]
 dyn_strap_bk_r_base [{shapeType=Capsule, length=0.505, radius=0.505, shapeName=dyn_strap_bk_r_base_capsule, rotation=<rot 0 -90 0>}]
 dyn_strap_fr_l [{shapeType=Capsule, length=4, radius=1, shapeName=dyn_strap_fr_l_capsule, center=<4 -1 0>, rotation=<rot 0 90 -90>}]
 dyn_strap_fr_l_base [{shapeType=Capsule, length=0.505, radius=0.505, shapeName=dyn_strap_fr_l_base_capsule, rotation=<rot 0 -90 0>}]
 dyn_strap_fr_r [{shapeType=Capsule, length=4, radius=1, shapeName=dyn_strap_fr_r_capsule, center=<4 -1 0>, rotation=<rot 0 90 -90>}]
 dyn_strap_fr_r_base [{shapeType=Capsule, length=0.505, radius=0.505, shapeName=dyn_strap_fr_r_base_capsule, rotation=<rot 0 -90 0>}]
 extra_bag_bk_l [{
   shapeType=Capsule
   length=5.287
   radius=3.6682
   shapeName=extra_bag_bk_l_capsule
   center=<0.1991 -0.0866 0.0028>
   rotation=<rot 1.4713 -90 -5.9104>
  }]
 extra_bag_bk_r [{
   shapeType=Capsule
   length=5.4605
   radius=3.9709
   shapeName=extra_bag_bk_r_capsule
   center=<0.209 0.085 -0.0823>
   rotation=<rot -5.324 -90 -7.5477>
  }]
 extra_bag_fr_l [{
   shapeType=Capsule
   length=5.288
   radius=3.5584
   shapeName=extra_bag_fr_l_capsule
   center=<0.0674 0.0894 0.085>
   rotation=<rot -0.8339 -90 -1.5817>
  }]
 extra_bag_fr_r [{
   shapeType=Capsule
   length=5.3058
   radius=3.5808
   shapeName=extra_bag_fr_r_capsule
   center=<0.1008 -0.0872 0.0661>
   rotation=<rot 1.1427 -90 -2.103>
  }]
 pelvis [{shapeType=Capsule, length=7.5, radius=9.7688, shapeName=pelvis_capsule, center=<0.9577 0.6416 -0.0952>, rotation=<rot 4.3937 -90 1.8617>}]
}
constraints [
 {bone1Name=dyn_grenade_05, bone2Name=dyn_grenade_05_base, swing1LimitDegrees=10, swing2LimitDegrees=10, twistLimitDegrees=1}
 {
  bone1Name=dyn_grenade_01_base
  bone2Name=pelvis
  swing1LimitDegrees=45
  swing1Motion=Free
  swing2LimitDegrees=45
  swing2Motion=Free
  twistLimitDegrees=45
  twistMotion=Free
 }
 {bone1Name=dyn_grenade_01, bone2Name=dyn_grenade_01_base, swing1LimitDegrees=10, swing2LimitDegrees=10, twistLimitDegrees=1}
 {
  bone1Name=dyn_grenade_02_base
  bone2Name=pelvis
  swing1LimitDegrees=45
  swing1Motion=Free
  swing2LimitDegrees=45
  swing2Motion=Free
  twistLimitDegrees=45
  twistMotion=Free
 }
 {bone1Name=dyn_grenade_02, bone2Name=dyn_grenade_02_base, swing1LimitDegrees=10, swing2LimitDegrees=10, twistLimitDegrees=1}
 {
  bone1Name=dyn_grenade_03_base
  bone2Name=pelvis
  swing1LimitDegrees=45
  swing1Motion=Free
  swing2LimitDegrees=45
  swing2Motion=Free
  twistLimitDegrees=45
  twistMotion=Free
 }
 {bone1Name=dyn_grenade_03, bone2Name=dyn_grenade_03_base, swing1LimitDegrees=10, swing2LimitDegrees=10, twistLimitDegrees=1}
 {
  bone1Name=dyn_grenade_04_base
  bone2Name=pelvis
  swing1LimitDegrees=45
  swing1Motion=Free
  swing2LimitDegrees=45
  swing2Motion=Free
  twistLimitDegrees=45
  twistMotion=Free
 }
 {bone1Name=dyn_grenade_04, bone2Name=dyn_grenade_04_base, swing1LimitDegrees=10, swing2LimitDegrees=10, twistLimitDegrees=1}
 {
  bone1Name=dyn_grenade_05_base
  bone2Name=pelvis
  swing1LimitDegrees=45
  swing1Motion=Free
  swing2LimitDegrees=45
  swing2Motion=Free
  twistLimitDegrees=45
  twistMotion=Free
 }
 {bone1Name=dyn_strap_fr_l, bone2Name=dyn_strap_fr_l_base, swing1LimitDegrees=30, swing2LimitDegrees=15, twistLimitDegrees=5}
 {bone1Name=dyn_strap_fr_r, bone2Name=dyn_strap_fr_r_base, swing1LimitDegrees=30, swing2LimitDegrees=15, twistLimitDegrees=5}
 {bone1Name=dyn_strap_bk_r, bone2Name=dyn_strap_bk_r_base, swing1LimitDegrees=30, swing2LimitDegrees=15, twistLimitDegrees=5}
 {
  bone1Name=dyn_strap_bk_r_base
  bone2Name=extra_bag_bk_r
  swing1LimitDegrees=45
  swing1Motion=Free
  swing2LimitDegrees=45
  swing2Motion=Free
  twistLimitDegrees=45
  twistMotion=Free
 }
 {
  bone1Name=dyn_strap_fr_r_base
  bone2Name=extra_bag_fr_r
  swing1LimitDegrees=45
  swing1Motion=Free
  swing2LimitDegrees=45
  swing2Motion=Free
  twistLimitDegrees=45
  twistMotion=Free
 }
 {bone1Name=dyn_strap_bk_l, bone2Name=dyn_strap_bk_l_base, swing1LimitDegrees=30, swing2LimitDegrees=15, twistLimitDegrees=5}
 {
  bone1Name=dyn_strap_bk_l_base
  bone2Name=extra_bag_bk_l
  swing1LimitDegrees=45
  swing1Motion=Free
  swing2LimitDegrees=45
  swing2Motion=Free
  twistLimitDegrees=45
  twistMotion=Free
 }
 {
  bone1Name=extra_bag_bk_l
  bone2Name=pelvis
  swing1LimitDegrees=45
  swing1Motion=Free
  swing2LimitDegrees=45
  swing2Motion=Free
  twistLimitDegrees=45
  twistMotion=Free
 }
 {
  bone1Name=extra_bag_bk_r
  bone2Name=pelvis
  swing1LimitDegrees=45
  swing1Motion=Free
  swing2LimitDegrees=45
  swing2Motion=Free
  twistLimitDegrees=45
  twistMotion=Free
 }
 {
  bone1Name=dyn_strap_fr_l_base
  bone2Name=extra_bag_fr_l
  swing1LimitDegrees=45
  swing1Motion=Free
  swing2LimitDegrees=45
  swing2Motion=Free
  twistLimitDegrees=45
  twistMotion=Free
 }
 {
  bone1Name=extra_bag_fr_l
  bone2Name=pelvis
  swing1LimitDegrees=45
  swing1Motion=Free
  swing2LimitDegrees=45
  swing2Motion=Free
  twistLimitDegrees=45
  twistMotion=Free
 }
 {
  bone1Name=extra_bag_fr_r
  bone2Name=pelvis
  swing1LimitDegrees=45
  swing1Motion=Free
  swing2LimitDegrees=45
  swing2Motion=Free
  twistLimitDegrees=45
  twistMotion=Free
 }
 {bone1Name=dyn_leg_strap, bone2Name=dyn_leg_strap_base, swing1LimitDegrees=5, swing2LimitDegrees=10, twistLimitDegrees=1}
 {
  bone1Name=dyn_leg_strap_base
  bone2Name=pelvis
  swing1LimitDegrees=45
  swing1Motion=Free
  swing2LimitDegrees=45
  swing2Motion=Free
  twistLimitDegrees=45
  twistMotion=Free
 }
]
