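# Post-pose AnimBlueprint for the RoundThumb_Nail male player body cosmetic, animating its straps, pads and other accessory bones.
# It copies the attached parent's pose through a deform Control Rig, then weighted constraints, pelvis-space RigidBody physics and a second Control Rig.
# Final Look At and additive bone offsets correct the shoulder pad and strap, blended by the arm_up_r curve.
# The event graph updates that alpha and computes a gravity override via a deprecated function.
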AnimBlueprint /BRCosmetics/Characters/Player/Male/Medium/Bodies/M_MED_RoundThumb_Nail/Meshes/M_MED_RoundThumb_Nail_Skeleton_AnimBlueprint
target_skeleton=&/M_MED_RoundThumb_Nail_Skeleton
graphs [
 AnimGraph {nodes [
   n1 = Control Rig(Source=Copy Pose From Mesh() {bUseAttachedParent=true}) {
    alphaBoolBlend {blendOption=Linear}
    controlRigAssetReference {blueprintRigClass=&/M_MED_RoundThumb_Nail_DeformRig.M_MED_RoundThumb_Nail_DeformRig_C}
   }
   n2 = "Constraint - extra_shoulder_strap_02 (PPPPPPPP)"!(
    ComponentPose=Local To Component(LocalPose=@n1)
    Alpha=1
    ConstraintWeights_0=0.0227
    ConstraintWeights_1:float=0.073
    ConstraintWeights_2:float=0.1615
    ConstraintWeights_3:float=0.0113
    ConstraintWeights_4:float=0.1598
    ConstraintWeights_5:float=0.2444
    ConstraintWeights_6:float=0.1541
    ConstraintWeights_7:float=0.1732
   ) -> (Pose) {
    alphaBoolBlend {blendOption=Linear}
    boneToModify {boneName=extra_shoulder_strap_02}
    constraintSetup [
     {targetBone {boneName=spine_04}}
     {targetBone {boneName=spine_05}}
     {targetBone {boneName=clavicle_r}}
     {targetBone {boneName=upperarm_r}}
     {targetBone {boneName=upperarm_twist_01_r}}
     {targetBone {boneName=upperarm_twist_02_r}}
     {targetBone {boneName=deform_upperarm_fr_r}}
     {targetBone {boneName=deform_upperarm_up_r}}
    ]
    constraintWeights [1,1,1,1,1,1,1,1]
   }
   n3 = "Constraint - extra_shoulder_strap_03 (PPPPPPPP)"!(
    ComponentPose=@n2
    Alpha=1
    ConstraintWeights_0=0.0414
    ConstraintWeights_1:float=0.0371
    ConstraintWeights_2:float=0.2954
    ConstraintWeights_3:float=0.1292
    ConstraintWeights_4:float=0.1071
    ConstraintWeights_5:float=0.0585
    ConstraintWeights_6:float=0.1545
    ConstraintWeights_7:float=0.1769
   ) -> (Pose) {
    alphaBoolBlend {blendOption=Linear}
    boneToModify {boneName=extra_shoulder_strap_03}
    constraintSetup [
     {targetBone {boneName=spine_03}}
     {targetBone {boneName=spine_04}}
     {targetBone {boneName=spine_05}}
     {targetBone {boneName=clavicle_r}}
     {targetBone {boneName=upperarm_twist_01_r}}
     {targetBone {boneName=upperarm_twist_02_r}}
     {targetBone {boneName=deform_upperarm_bk_r}}
     {targetBone {boneName=deform_upperarm_up_r}}
    ]
    constraintWeights [1,1,1,1,1,1,1,1]
   }
   n4 = "Constraint - extra_shoulder_strap_01 (PPPPPPPPPPP)"!(
    ComponentPose=@n3
    Alpha=1
    ConstraintWeights_0=0.0076
    ConstraintWeights_1:float=0.0138
    ConstraintWeights_2:float=0.4071
    ConstraintWeights_3:float=0.0011
    ConstraintWeights_4:float=0.0238
    ConstraintWeights_5:float=0.3352
    ConstraintWeights_6:float=0.069
    ConstraintWeights_7:float=0.0217
    ConstraintWeights_8:float=0.0008
    ConstraintWeights_9:float=0.0254
    ConstraintWeights_10:float=0.0944
   ) -> (Pose) {
    alphaBoolBlend {blendOption=Linear}
    boneToModify {boneName=extra_shoulder_strap_01}
    constraintSetup [
     {targetBone {boneName=spine_03}}
     {targetBone {boneName=spine_04}}
     {targetBone {boneName=spine_05}}
     {targetBone {boneName=neck_01}}
     {targetBone {boneName=neck_02}}
     {targetBone {boneName=clavicle_l}}
     {targetBone {boneName=upperarm_twist_01_l}}
     {targetBone {boneName=upperarm_twist_02_l}}
     {targetBone {boneName=deform_pec_l}}
     {targetBone {boneName=deform_upperarm_fr_l}}
     {targetBone {boneName=deform_upperarm_up_l}}
    ]
    constraintWeights [1,1,1,1,1,1,1,1,1,1,1]
   }
   n5 = "Constraint - extra_shoulder_strap_05 (PPPPPPPP)"!(
    ComponentPose=@n4
    Alpha=1
    ConstraintWeights_0=0.0328
    ConstraintWeights_1:float=0.0162
    ConstraintWeights_2:float=0.5292
    ConstraintWeights_3:float=0.2065
    ConstraintWeights_4:float=0.0808
    ConstraintWeights_5:float=0.004
    ConstraintWeights_6:float=0.0279
    ConstraintWeights_7:float=0.1026
   ) -> (Pose) {
    alphaBoolBlend {blendOption=Linear}
    boneToModify {boneName=extra_shoulder_strap_05}
    constraintSetup [
     {targetBone {boneName=spine_03}}
     {targetBone {boneName=spine_04}}
     {targetBone {boneName=spine_05}}
     {targetBone {boneName=clavicle_l}}
     {targetBone {boneName=upperarm_twist_01_l}}
     {targetBone {boneName=upperarm_twist_02_l}}
     {targetBone {boneName=deform_upperarm_bk_l}}
     {targetBone {boneName=deform_upperarm_up_l}}
    ]
    constraintWeights [1,1,1,1,1,1,1,1]
   }
   n6 = "Constraint - extra_shoulder_strap_07 (PPPPPPPP)"!(
    ComponentPose=@n5
    Alpha=1
    ConstraintWeights_0=0.0472
    ConstraintWeights_1:float=0.1321
    ConstraintWeights_2:float=0.4754
    ConstraintWeights_3:float=0.2009
    ConstraintWeights_4:float=0.0118
    ConstraintWeights_5:float=0.0597
    ConstraintWeights_6:float=0.0263
    ConstraintWeights_7:float=0.0465
   ) -> (Pose) {
    alphaBoolBlend {blendOption=Linear}
    boneToModify {boneName=extra_shoulder_strap_07}
    constraintSetup [
     {targetBone {boneName=spine_03}}
     {targetBone {boneName=spine_04}}
     {targetBone {boneName=spine_05}}
     {targetBone {boneName=clavicle_l}}
     {targetBone {boneName=upperarm_twist_02_l}}
     {targetBone {boneName=deform_pec_l}}
     {targetBone {boneName=deform_upperarm_fr_l}}
     {targetBone {boneName=deform_upperarm_up_l}}
    ]
    constraintWeights [1,1,1,1,1,1,1,1]
   }
   n7 = "Constraint - extra_shoulder_strap_06 (PPPPPPPPPP)"!(
    ComponentPose=@n6
    Alpha=1
    ConstraintWeights_0=0.0004
    ConstraintWeights_1:float=0.0647
    ConstraintWeights_2:float=0.1728
    ConstraintWeights_3:float=0.5012
    ConstraintWeights_4:float=0.0176
    ConstraintWeights_5:float=0.1367
    ConstraintWeights_6:float=0.015
    ConstraintWeights_7:float=0.0795
    ConstraintWeights_8:float=0.0008
    ConstraintWeights_9:float=0.0114
   ) -> (Pose) {
    alphaBoolBlend {blendOption=Linear}
    boneToModify {boneName=extra_shoulder_strap_06}
    constraintSetup [
     {targetBone {boneName=spine_02}}
     {targetBone {boneName=spine_03}}
     {targetBone {boneName=spine_04}}
     {targetBone {boneName=spine_05}}
     {targetBone {boneName=neck_02}}
     {targetBone {boneName=clavicle_l}}
     {targetBone {boneName=upperarm_twist_01_l}}
     {targetBone {boneName=deform_pec_l}}
     {targetBone {boneName=deform_pec_r}}
     {targetBone {boneName=deform_upperarm_fr_l}}
    ]
    constraintWeights [1,1,1,1,1,1,1,1,1,1]
   }
   n8 = "Constraint - extra_shoulder_pad (PP)"!(ComponentPose=@n7, Alpha=1, ConstraintWeights_0=0.5, ConstraintWeights_1:float=0.5) -> (Pose) {
    alphaBoolBlend {blendOption=Linear}
    boneToModify {boneName=extra_shoulder_pad}
    constraintSetup [{targetBone {boneName=extra_shoulder_strap_03}}, {targetBone {boneName=extra_shoulder_strap_02}}]
    constraintWeights [1,1]
   }
   n9 = "Constraint - extra_shoulder_pad (R)"(ComponentPose=@n8, Alpha=1, ConstraintWeights_0=1) {
    alphaBoolBlend {blendOption=Linear}
    boneToModify {boneName=extra_shoulder_pad}
    constraintSetup [{transformType=Rotation, targetBone {boneName=extra_shoulder_strap_03}}]
    constraintWeights [1]
   }
   n10 = "Constraint - extra_bolt_01 (PPPPPPPP)"!(
    ComponentPose=@n9
    Alpha=1
    ConstraintWeights_0=0.0328
    ConstraintWeights_1:float=0.0162
    ConstraintWeights_2:float=0.5292
    ConstraintWeights_3:float=0.2065
    ConstraintWeights_4:float=0.0808
    ConstraintWeights_5:float=0.004
    ConstraintWeights_6:float=0.0279
    ConstraintWeights_7:float=0.1026
   ) -> (Pose) {
    alphaBoolBlend {blendOption=Linear}
    boneToModify {boneName=extra_bolt_01}
    constraintSetup [
     {targetBone {boneName=spine_03}}
     {targetBone {boneName=spine_04}}
     {targetBone {boneName=spine_05}}
     {targetBone {boneName=clavicle_l}}
     {targetBone {boneName=upperarm_twist_01_l}}
     {targetBone {boneName=upperarm_twist_02_l}}
     {targetBone {boneName=deform_upperarm_bk_l}}
     {targetBone {boneName=deform_upperarm_up_l}}
    ]
    constraintWeights [1,1,1,1,1,1,1,1]
   }
   n11 = "Constraint - extra_chest_pocket (PPPPPPPP)"!(
    ComponentPose=@n10
    Alpha=1
    ConstraintWeights_0=0.1381
    ConstraintWeights_1:float=0.3121
    ConstraintWeights_2:float=0.2166
    ConstraintWeights_3:float=0.0821
    ConstraintWeights_4:float=0.0511
    ConstraintWeights_5:float=0.0477
    ConstraintWeights_6:float=0.0838
    ConstraintWeights_7:float=0.0684
   ) -> (Pose) {
    alphaBoolBlend {blendOption=Linear}
    boneToModify {boneName=extra_chest_pocket}
    constraintSetup [
     {targetBone {boneName=spine_02}}
     {targetBone {boneName=spine_03}}
     {targetBone {boneName=spine_04}}
     {targetBone {boneName=spine_05}}
     {targetBone {boneName=clavicle_l}}
     {targetBone {boneName=upperarm_twist_01_l}}
     {targetBone {boneName=deform_pec_l}}
     {targetBone {boneName=deform_upperarm_fr_l}}
    ]
    constraintWeights [1,1,1,1,1,1,1,1]
   }
   n12 = "Constraint - extra_calf_belt_r (PPPPPPPP)"!(
    ComponentPose=@n11
    Alpha=1
    ConstraintWeights_0=0
    ConstraintWeights_1:float=0.4206
    ConstraintWeights_2:float=0.0626
    ConstraintWeights_3:float=0.2594
    ConstraintWeights_4:float=0.0176
    ConstraintWeights_5:float=0.0001
    ConstraintWeights_6:float=0.0333
    ConstraintWeights_7:float=0.2064
   ) -> (Pose) {
    alphaBoolBlend {blendOption=Linear}
    boneToModify {boneName=extra_calf_belt_r}
    constraintSetup [
     {targetBone {boneName=pelvis}}
     {targetBone {boneName=thigh_r}}
     {targetBone {boneName=thigh_twist_01_r}}
     {targetBone {boneName=calf_r}}
     {targetBone {boneName=calf_twist_02_r}}
     {targetBone {boneName=deform_groin_r}}
     {targetBone {boneName=deform_knee_in_r}}
     {targetBone {boneName=deform_knee_out_r}}
    ]
    constraintWeights [1,1,1,1,1,1,1,1]
   }
   n13 = "Constraint - extra_knee_pad (RR)"!(ComponentPose=@n12, Alpha=1, ConstraintWeights_0=0.4, ConstraintWeights_1:float=0.6) -> (Pose) {
    alphaBoolBlend {blendOption=Linear}
    boneToModify {boneName=extra_knee_pad}
    constraintSetup [{transformType=Rotation, targetBone {boneName=thigh_l}}, {transformType=Rotation, targetBone {boneName=calf_l}}]
    constraintWeights [1,1]
   }
   n14 = RigidBody!(ComponentPose=@n13, Alpha, OverrideWorldGravity:Vector="0,0,-980", ExternalForce) -> (Pose) {
    bClampLinearTranslationLimitToRefPose=true
    bFreezeIncomingPoseOnStart=true
    bOverrideWorldGravity=true
    bTransferBoneVelocities=true
    overridePhysicsAsset=&/M_MED_RoundThumb_Nail_Physics
    simulationSpace=BaseBoneSpace
    alphaBoolBlend {blendOption=Linear}
    baseBoneRef {boneName=pelvis}
    componentLinearAccScale=<1 1 10>
    componentLinearVelScale=<1 1 10>
    overrideWorldGravity {z=-980}
    simSpaceSettings {maxAngularAcceleration=0, maxLinearAcceleration=0, velocityScaleZ=5, worldAlpha=1}
   }
   n15 = Control Rig!(Source=Component To Local(ComponentPose=@n14), Alpha, crouch:bool=false, targeting:bool=false) -> (Pose) {
    alphaBoolBlend {blendOption=Linear}
    controlRigAssetReference {blueprintRigClass=&/M_MED_RoundThumb_Nail_CtrlRig.M_MED_RoundThumb_Nail_CtrlRig_C}
   }
   n16 = Get Alpha_arm_up_r() -> (Alpha_arm_up_r:float)
   n17 = Look At - Bone: extra_shoulder_pad(ComponentPose=Local To Component(LocalPose=@n15), Alpha=@n16) {
    alphaBoolBlend {blendOption=Linear}
    boneToModify {boneName=extra_shoulder_pad}
    lookAtLocation=<2.3674 -0.0448 2.7869>
    lookAtTarget {boneReference {boneName=lowerarm_r}}
    lookAt_Axis {axis=<-1 0>}
    lookUp_Axis {axis=<0 1>}
   }
   n18 = "Transform (Modify) Bone - Bone: extra_shoulder_pad"(ComponentPose=@n17, Alpha=@n16, Translation="0,0,0", Rotation="1.397,-4.387,0.393", Scale="1,1,1") {
    rotationMode=BMM_Additive
    rotationSpace=BCS_BoneSpace
    alphaBoolBlend {blendOption=Linear}
    boneToModify {boneName=extra_shoulder_pad}
    rotation=<rot 1.397 -4.387 0.393>
   }
   n19 = Look At - Bone: extra_shoulder_strap_03(ComponentPose=@n18, Alpha=@n16) {
    alphaBoolBlend {blendOption=Linear}
    boneToModify {boneName=extra_shoulder_strap_03}
    lookAtLocation=<2.3674 -0.0448 -7.9602>
    lookAtTarget {boneReference {boneName=extra_shoulder_strap_09}}
   }
   n20 = "Transform (Modify) Bone - Bone: extra_shoulder_strap_03"(ComponentPose=@n19, Alpha=@n16, Translation="0.759,-0.807,-2.285", Rotation="7.199,13.530,-50.263", Scale="1,1,1") {
    rotationMode=BMM_Additive
    rotationSpace=BCS_BoneSpace
    translationMode=BMM_Additive
    translationSpace=BCS_BoneSpace
    alphaBoolBlend {blendOption=Linear}
    boneToModify {boneName=extra_shoulder_strap_03}
    rotation=<rot 7.199 13.53 -50.263>
    translation=<0.759 -0.807 -2.285>
   }
   Output Pose(Result=Component To Local(ComponentPose=@n20))
  ]}
 EventGraph {nodes [
   Event BlueprintUpdateAnimation()
   > n1 = Gravity Override State Function DEPRECATED(
    CharacterPartAnimInstance=Self-Reference()
    GameTimeInSeconds=GetGameTimeInSeconds()
    GravityOverrideParams=Get Gravity Override Params()
    RBAN Params=Get RBAN Params()
   )
   > Set Gravity Override(Gravity Override=@n1)
   > Set Alpha_arm_up_r(Alpha_arm_up_r:float=GetCurveValue(CurveName=arm_up_r)) -> (Output_Get:float)
   TryGetPawnOwner()
  ]}
]
variables [
 Alpha_arm_up_r {type=number}
 Gravity Override {type=Vector}
 Gravity Override Params {type=GravityOverrideParamsStruct}
 RBAN Params {
  type=RBANParamsStruct
  default {
   dBNO {simSpaceSettings {
     maxAngularAcceleration=1000
     maxAngularVelocity=3
     maxLinearAcceleration=300
     maxLinearVelocity=700
     velocityScaleZ=0
     externalLinearDragV=<0 0.5 0>
    }}
   dBNO_Carried {simSpaceSettings {
     maxAngularAcceleration=1000
     maxAngularVelocity=3
     maxLinearAcceleration=300
     maxLinearVelocity=700
     velocityScaleZ=0
     externalLinearDragV=<0 0.5 0>
    }}
   emoteOrMelee {simSpaceSettings {
     maxAngularAcceleration=1000
     maxAngularVelocity=3
     maxLinearAcceleration=300
     maxLinearVelocity=700
     velocityScaleZ=0
     externalLinearDragV=<0 0.5 0>
    }}
   front_End {simSpaceSettings {
     maxAngularAcceleration=1000
     maxAngularVelocity=3
     maxLinearAcceleration=300
     maxLinearVelocity=700
     velocityScaleZ=0
     externalLinearDragV=<0 0.5 0>
    }}
   onGround_Crouch_Moving {simSpaceSettings {
     maxAngularAcceleration=1000
     maxAngularVelocity=3
     maxLinearAcceleration=300
     maxLinearVelocity=700
     velocityScaleZ=0
     externalLinearDragV=<0 0.5 0>
    }}
   onGround_Falling {simSpaceSettings {
     maxAngularAcceleration=1000
     maxAngularVelocity=3
     maxLinearAcceleration=300
     maxLinearVelocity=700
     velocityScaleZ=0
     externalLinearDragV=<0 0.5 0>
    }}
   onGround_Moving {simSpaceSettings {
     maxAngularAcceleration=1000
     maxAngularVelocity=3
     maxLinearAcceleration=300
     maxLinearVelocity=700
     velocityScaleZ=0
     externalLinearDragV=<0 0.5 0>
    }}
   onGround_Standing {simSpaceSettings {
     maxAngularAcceleration=1000
     maxAngularVelocity=3
     maxLinearAcceleration=300
     maxLinearVelocity=700
     velocityScaleZ=0
     externalLinearDragV=<0 0.5 0>
    }}
   rail_Grinding {simSpaceSettings {
     maxAngularAcceleration=1000
     maxAngularVelocity=3
     maxLinearAcceleration=300
     maxLinearVelocity=700
     velocityScaleZ=0
     worldAlpha=0.1
     externalLinearDragV=<0 0.5 0>
    }}
   riding_Moving {simSpaceSettings {
     maxAngularAcceleration=1000
     maxAngularVelocity=3
     maxLinearAcceleration=300
     maxLinearVelocity=700
     velocityScaleZ=0
     externalLinearDragV=<0 0.5 0>
    }}
   riding_Moving_Targeting {simSpaceSettings {
     maxAngularAcceleration=1000
     maxAngularVelocity=3
     maxLinearAcceleration=300
     maxLinearVelocity=700
     velocityScaleZ=0
     externalLinearDragV=<0 0.5 0>
    }}
   riding_Standing {simSpaceSettings {
     maxAngularAcceleration=1000
     maxAngularVelocity=3
     maxLinearAcceleration=300
     maxLinearVelocity=700
     velocityScaleZ=0
     externalLinearDragV=<0 0.5 0>
    }}
   skydive_DivingDown {simSpaceSettings {
     maxAngularAcceleration=1000
     maxAngularVelocity=3
     maxLinearAcceleration=300
     maxLinearVelocity=700
     velocityScaleZ=0
     externalLinearDragV=<0 0.5 0>
    }}
   skydive_FloatingDown {
    componentLinearAccScale=<0 0.5 0>
    componentLinearVelScale=<0 0.5 0>
    simSpaceSettings {
     maxAngularAcceleration=1000
     maxAngularVelocity=3
     maxLinearAcceleration=300
     maxLinearVelocity=700
     velocityScaleZ=0
     externalLinearDragV=<0 0.5 0>
    }
   }
   skydive_GoingUp {simSpaceSettings {
     maxAngularAcceleration=1000
     maxAngularVelocity=3
     maxLinearAcceleration=300
     maxLinearVelocity=700
     velocityScaleZ=0
     externalLinearDragV=<0 0.5 0>
    }}
   skydive_Parachuting {
    componentLinearAccScale=<0 0.1 0.3>
    componentLinearVelScale=<0 0.1 0.3>
    simSpaceSettings {
     maxAngularAcceleration=1000
     maxAngularVelocity=3
     maxLinearAcceleration=300
     maxLinearVelocity=700
     velocityScaleZ=0
     externalLinearDragV=<0 0.5 0>
    }
   }
   skydive_Surfing {simSpaceSettings {
     maxAngularAcceleration=1000
     maxAngularVelocity=3
     maxLinearAcceleration=300
     maxLinearVelocity=700
     velocityScaleZ=0
     externalLinearDragV=<0 0.5 0>
    }}
   swimming {simSpaceSettings {
     maxAngularAcceleration=1000
     maxAngularVelocity=3
     maxLinearAcceleration=300
     maxLinearVelocity=700
     velocityScaleZ=0
     externalLinearDragV=<0 0.5 0>
    }}
   tactical_Sprint {simSpaceSettings {
     maxAngularAcceleration=1000
     maxAngularVelocity=3
     maxLinearAcceleration=300
     maxLinearVelocity=700
     velocityScaleZ=0
     externalLinearDragV=<0 0.5 0>
    }}
   ziplining {simSpaceSettings {
     maxAngularAcceleration=1000
     maxAngularVelocity=3
     maxLinearAcceleration=300
     maxLinearVelocity=700
     velocityScaleZ=0
     externalLinearDragV=<0 0.5 0>
    }}
  }
 }
]
