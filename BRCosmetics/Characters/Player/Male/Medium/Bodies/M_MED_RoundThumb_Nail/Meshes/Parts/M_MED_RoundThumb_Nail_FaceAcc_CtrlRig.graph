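# Control Rig for the face accessory on the male medium RoundThumb_Nail body, applying pose-driven corrections to hair bones.
# Head and right clavicle spherical pose readers multiply to adjust right and center hair plus extra_clavicle_r; the clavicle alone also shifts right hair.
# A second head reader, smoothed by AlphaInterp, offsets and shrinks all three hair roots and writes its inverse to the HeadUp curve.
# All readers have debug drawing enabled.
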ControlRigBlueprint /BRCosmetics/Characters/Player/Male/Medium/Bodies/M_MED_RoundThumb_Nail/Meshes/Parts/M_MED_RoundThumb_Nail_FaceAcc_CtrlRig
graphs [
 RigVMFunctionLibrary {links [], nodes []}
 RigVMFunction_Sequence_3_1_ContainedGraph {nodes [
   n1 = RigVMFunctionEntryNode()
   n2 = &RigVMFunction_Sequence(ExecuteContext=@n1.ExecuteContext)
   n3 = &RigVMFunction_Sequence(ExecuteContext=@n2.B)
   RigVMFunctionReturnNode(A=@n2.A, B=@n3.A, C=@n3.B)
  ]}
 RigVMModel {nodes [
   RigUnit_BeginExecution = &RigUnit_BeginExecution()
   RigVMFunction_Sequence_3 = RigVMAggregateNode(ExecuteContext=@RigUnit_BeginExecution.ExecutePin)
   n1 = &RigUnit_SphericalPoseReader(
    ActiveRegionScaleFactors=(PositiveWidth=1,NegativeWidth=0.5,PositiveHeight=0.75,NegativeHeight=0.75)
    ActiveRegionSize=0.01
    Debug=(bDrawDebug=true,bDraw2D=False,bDrawLocalAxes=False,DebugScale=25,DebugSegments=20,DebugThickness=0.25)
    DriverAxis=(X=0,Y=1,Z=0)
    DriverItem=(Type=Bone,Name="head")
    FalloffRegionScaleFactors=(PositiveWidth=1,NegativeWidth=0.5,PositiveHeight=1,NegativeHeight=1)
    FalloffSize=0.1
    OptionalParentItem=(Type=Bone)
    RotationOffset=(X=105,Y=0,Z=-7)
    ExecutePin=@RigVMFunction_Sequence_3.A
   )
   > n2 = &RigUnit_SphericalPoseReader(
    ActiveRegionScaleFactors=(PositiveWidth=1,NegativeWidth=1,PositiveHeight=1,NegativeHeight=1)
    ActiveRegionSize=0.01
    Debug=(bDrawDebug=true,bDraw2D=False,bDrawLocalAxes=False,DebugScale=25,DebugSegments=20,DebugThickness=0.25)
    DriverAxis=(X=-1,Y=0,Z=0)
    DriverItem=(Type=Bone,Name="clavicle_r")
    FalloffRegionScaleFactors=(PositiveWidth=1,NegativeWidth=1,PositiveHeight=1,NegativeHeight=1)
    FalloffSize=0.1
    OptionalParentItem=(Type=Bone)
    RotationOffset=(X=-105.4232,Y=0,Z=-88.6268)
   )
   Multiply = &RigVMFunction_MathDoubleMul(A=@n1.OutputParam, B=@n2.OutputParam)
   > &RigUnit_ModifyTransforms(
    ItemToModify=((Item=(Type=Bone,Name="extra_clavicle_r"),Transform=(Rotation=(X=0,Y=0,Z=0,W=1),Translation=(X=0,Y=0,Z=0.5857),Scale3D=(X=1,Y=1,Z=1))))
    Weight=@Multiply.Result
   )
   > &RigUnit_ModifyTransforms(
    ItemToModify=((Item=(Type=Bone,Name="dyn_hair_01"),Transform=(Rotation=(X=-0.0421,Y=0.0113,Z=-0.2586,W=0.965),Translation=(X=0,Y=0,Z=0),Scale3D=(X=1,Y=1,Z=1))))
    Weight=@Multiply.Result
   )
   > &RigUnit_ModifyTransforms(
    ItemToModify=((Item=(Type=Bone,Name="dyn_hair_01_r"),Transform=(Rotation=(X=0,Y=0,Z=-0.2588,W=0.9659),Translation=(X=0,Y=0,Z=1),Scale3D=(X=1,Y=1,Z=1))))
    Weight=@Multiply.Result
   )
   &RigUnit_ModifyTransforms(
    ItemToModify=((Item=(Type=Bone,Name="dyn_hair_01_r"),Transform=(Rotation=(X=0,Y=0,Z=0,W=1),Translation=(X=-0.5,Y=0,Z=0),Scale3D=(X=1,Y=1,Z=1))))
    ExecutePin=@RigVMFunction_Sequence_3.B
    Weight=@n2.OutputParam
   )
   n3 = &RigUnit_SphericalPoseReader(
    ActiveRegionScaleFactors=(PositiveWidth=1,NegativeWidth=2,PositiveHeight=2,NegativeHeight=2)
    ActiveRegionSize=0.05
    Debug=(bDrawDebug=true,bDraw2D=False,bDrawLocalAxes=False,DebugScale=25,DebugSegments=20,DebugThickness=0.25)
    DriverAxis=(X=0,Y=1,Z=0)
    DriverItem=(Type=Bone,Name="head")
    FalloffRegionScaleFactors=(PositiveWidth=1,NegativeWidth=0.5,PositiveHeight=2,NegativeHeight=2)
    FalloffSize=0.2
    OptionalParentItem=(Type=Bone)
    RotationOffset=(X=85,Y=0,Z=-45)
    ExecutePin=@RigVMFunction_Sequence_3.C
   )
   n4 = &RigVMFunction_Sequence(ExecuteContext=@n3.ExecutePin)
   n5 = &RigVMFunction_AlphaInterp(
    InterpSpeedDecreasing=10
    InterpSpeedIncreasing=10
    OutRange=(Min=1,Max=0)
    Scale=1
    bClampResult=False
    bMapRange=false
    Value=@n3.OutputParam
   )
   &RigUnit_ModifyTransforms(
    ItemToModify=((Item=(Type=Bone,Name="dyn_hair_01_l"),Transform=(Rotation=(X=0,Y=0,Z=0,W=1),Translation=(X=0,Y=3,Z=0),Scale3D=(X=0.1,Y=0.1,Z=0.1))))
    ExecutePin=@n4.A
    Weight=@n5.Result
   )
   > &RigUnit_ModifyTransforms(
    ItemToModify=((Item=(Type=Bone,Name="dyn_hair_01"),Transform=(Rotation=(X=0,Y=0,Z=0,W=1),Translation=(X=0,Y=3,Z=0),Scale3D=(X=0.1,Y=0.1,Z=0.1))))
    Weight=@n5.Result
   )
   > &RigUnit_ModifyTransforms(
    ItemToModify=((Item=(Type=Bone,Name="dyn_hair_01_r"),Transform=(Rotation=(X=0,Y=0,Z=0,W=1),Translation=(X=0,Y=3,Z=0),Scale3D=(X=0.1,Y=0.1,Z=0.1))))
    Weight=@n5.Result
   )
   AlphaInterp = &RigVMFunction_AlphaInterp(
    InterpSpeedDecreasing=10
    InterpSpeedIncreasing=10
    OutRange=(Min=1,Max=0)
    Scale=1
    bClampResult=False
    bInterpResult=false
    bMapRange=true
    Value=@n5.Result
   )
   /Script/ControlRig.RigUnit_SetCurveValue(Curve=HeadUp, Value=@AlphaInterp.Result, ExecutePin=@n4.B) {is=struct RigVMUnitNode}
  ]}
]
hierarchy [
 root {loc=<0 0 0>, rot=<0 0 0 1>}
 pelvis {parent=root, loc=<0 -0.7071 93.1191>, rot=<0.0127 0.707 -0.0127 -0.707>}
 spine_01 {parent=pelvis, loc=<6.7494 -0.6388 0>, rot=<0 0 0.034 -0.9994>}
 spine_02 {parent=spine_01, loc=<7.2518 -0.035 0>, rot=<0 0 0.0238 -0.9997>}
 spine_03 {parent=spine_02, loc=<6.0104 0 0>, rot=<0 0 0.0149 -0.9999>}
 spine_04 {parent=spine_03, loc=<8.1024 0 0>, rot=<0 0 -0.0132 -0.9999>}
 spine_05 {parent=spine_04, loc=<10.0146 0 0>, rot=<0 0 -0.0015 -1>}
 neck_01 {parent=spine_05, loc=<11.3138 0.1529 0>, rot=<0 0 -0.2003 -0.9797>}
 neck_02 {parent=neck_01, loc=<4.5627 0 0>, rot=<0 0 0.0504 -0.9987>}
 head {parent=neck_02, loc=<5.3914 0 0>, rot=<0 0 0.115 -0.9934>}
 dyn_hair_r {parent=head, loc=<-0.8371 -5.9696 3.1867>, rot=<0.2864 -0.9282 0.2347 0.037>}
 dyn_hair {parent=head, loc=<-0.8006 -6.7802 0>, rot=<0.2292 -0.9734 0 0>}
 dyn_hair_l {parent=head, loc=<-0.8369 -5.9696 -3.1867>, rot=<0.2864 -0.9282 -0.2347 -0.037>}
 dyn_hair_01_r {parent=head, loc=<-0.8648 -5.4722 3.8356>, rot=<0.2721 -0.9105 0.3034 0.0696>}
 dyn_hair_02_r {parent=dyn_hair_01_r, loc=<1.7 0 0>, rot=<0 0 0 -1>}
 curl_13 {parent=dyn_hair_02_r, loc=<1.3059 0.1851 0.9308>, rot=<0.156 0.2806 -0.1261 -0.9386>}
 curl_12 {parent=dyn_hair_02_r, loc=<1.0918 -0.4525 1.0288>, rot=<0.1483 0.1579 0.1722 -0.961>}
 curl_15 {parent=dyn_hair_02_r, loc=<1.0075 0.12 -0.8091>, rot=<0.0324 0.1463 -0.0004 -0.9887>}
 curl_14 {parent=dyn_hair_02_r, loc=<0.9874 -0.7111 -0.0257>, rot=<-0.2057 -0.2787 0.3078 -0.8862>}
 curl_16 {parent=dyn_hair_02_r, loc=<-0.5915 0.3501 -1.9534>, rot=<-0.1899 -0.3299 0.0011 -0.9247>}
 dyn_curl_12_base {parent=dyn_hair_02_r, loc=<1.0555 -0.4409 1.0143>, rot=<0.1483 0.1579 0.1722 -0.961>}
 dyn_curl_12 {parent=dyn_curl_12_base, loc=<0.0408 0 0>, rot=<0 0 0 1>}
 dyn_curl_13_base {parent=dyn_hair_02_r, loc=<1.2728 0.1718 0.9109>, rot=<0.156 0.2806 -0.1261 -0.9386>}
 dyn_curl_13 {parent=dyn_curl_13_base, loc=<0.0408 0 0>, rot=<0 0 0 1>}
 dyn_curl_14_base {parent=dyn_hair_02_r, loc=<0.9606 -0.6936 -0.0004>, rot=<-0.2057 -0.2787 0.3078 -0.8862>}
 dyn_curl_14 {parent=dyn_curl_14_base, loc=<0.0408 0 0>, rot=<0 0 0 1>}
 dyn_curl_15_base {parent=dyn_hair_02_r, loc=<0.9685 0.1196 -0.8209>, rot=<0.0324 0.1463 -0.0004 -0.9887>}
 dyn_curl_15 {parent=dyn_curl_15_base, loc=<0.0408 0 0>, rot=<0 0 0 1>}
 dyn_curl_16_base {parent=dyn_hair_02_r, loc=<-0.6234 0.3451 -1.9285>, rot=<-0.1899 -0.3299 0.0011 -0.9247>}
 dyn_curl_16 {parent=dyn_curl_16_base, loc=<0.0408 0 0>, rot=<0 0 0 1>}
 dyn_hair_01 {parent=head, loc=<-0.7188 -6.7394 0>, rot=<0.2292 -0.9734 0 0>}
 dyn_hair_02 {parent=dyn_hair_01, loc=<1.7 0 0>, rot=<0 0 0 1>}
 curl_06 {parent=dyn_hair_02, loc=<0.8015 0.6219 2.3604>, rot=<0.0798 -0.0309 0.0985 -0.9914>}
 curl_07 {parent=dyn_hair_02, loc=<0.6997 -0.6698 1.6956>, rot=<0.1712 0.1627 0.2925 -0.9267>}
 curl_08 {parent=dyn_hair_02, loc=<1.8653 0.0055 1.0089>, rot=<-0.1173 -0.0609 0.0243 -0.9909>}
 curl_09 {parent=dyn_hair_02, loc=<1.9246 0.2143 -0.2212>, rot=<-0.0836 -0.0598 0.1651 -0.9809>}
 curl_10 {parent=dyn_hair_02, loc=<0.3452 -0.2997 -1.8853>, rot=<-0.0463 -0.0494 0.3552 -0.9323>}
 curl_11 {parent=dyn_hair_02, loc=<1.4295 0.4856 -2.0603>, rot=<0.0174 0.1051 0.2946 -0.9497>}
 dyn_curl_06_base {parent=dyn_hair_02, loc=<0.7616 0.6301 2.3623>, rot=<0.0798 -0.0309 0.0985 -0.9914>}
 dyn_curl_06 {parent=dyn_curl_06_base, loc=<0.0408 0 0>, rot=<0 0 0 1>}
 dyn_curl_07_base {parent=dyn_hair_02, loc=<0.668 -0.6499 1.6792>, rot=<0.1712 0.1627 0.2925 -0.9267>}
 dyn_curl_07 {parent=dyn_curl_07_base, loc=<0.0408 0 0>, rot=<0 0 0 1>}
 dyn_curl_08_base {parent=dyn_hair_02, loc=<1.8249 0.0068 1.014>, rot=<-0.1173 -0.0609 0.0243 -0.9909>}
 dyn_curl_08 {parent=dyn_curl_08_base, loc=<0.0408 0 0>, rot=<0 0 0 1>}
 dyn_curl_09_base {parent=dyn_hair_02, loc=<1.8863 0.2271 -0.2153>, rot=<-0.0836 -0.0598 0.1651 -0.9809>}
 dyn_curl_09 {parent=dyn_curl_09_base, loc=<0.0408 0 0>, rot=<0 0 0 1>}
 dyn_curl_10_base {parent=dyn_hair_02, loc=<0.3149 -0.2728 -1.8802>, rot=<-0.0463 -0.0494 0.3552 -0.9323>}
 dyn_curl_10 {parent=dyn_curl_10_base, loc=<0.0408 0 0>, rot=<0 0 0 1>}
 dyn_curl_11_base {parent=dyn_hair_02, loc=<1.3967 0.5083 -2.0689>, rot=<0.0174 0.1051 0.2946 -0.9497>}
 dyn_curl_11 {parent=dyn_curl_11_base, loc=<0.0408 0 0>, rot=<0 0 0 1>}
 dyn_hair_01_l {parent=head, loc=<-0.8645 -5.4722 -3.8356>, rot=<0.2721 -0.9105 -0.3034 -0.0696>}
 dyn_hair_02_l {parent=dyn_hair_01_l, loc=<1.7 0 0>, rot=<0 0 0 1>}
 curl_05 {parent=dyn_hair_02_l, loc=<0.32 -0.2697 -1.6528>, rot=<0.0265 0.0507 0.2293 -0.9717>}
 curl_04 {parent=dyn_hair_02_l, loc=<1.0667 0.26 -0.7415>, rot=<-0.1613 -0.248 0.0365 -0.9545>}
 curl_03 {parent=dyn_hair_02_l, loc=<1.422 -0.0473 0.5772>, rot=<0.0691 0.1637 -0.1304 -0.9754>}
 curl_02 {parent=dyn_hair_02_l, loc=<0.6816 -0.3465 1.4121>, rot=<-0.0552 -0.0416 0.3204 -0.9448>}
 curl_01 {parent=dyn_hair_02_l, loc=<-0.425 0.1982 2.3135>, rot=<0.2668 0.3773 -0.1416 -0.8754>}
 dyn_curl_01_base {parent=dyn_hair_02_l, loc=<0.2837 -0.2516 -1.6573>, rot=<0.0265 0.0507 0.2293 -0.9717>}
 dyn_curl_01 {parent=dyn_curl_01_base, loc=<0.0408 0 0>, rot=<0 0 0 1>}
 dyn_curl_02_base {parent=dyn_hair_02_l, loc=<1.0311 0.2596 -0.7217>, rot=<-0.1613 -0.248 0.0365 -0.9545>}
 dyn_curl_02 {parent=dyn_curl_02_base, loc=<0.0408 0 0>, rot=<0 0 0 1>}
 dyn_curl_03_base {parent=dyn_hair_02_l, loc=<1.3847 -0.0586 0.5649>, rot=<0.0691 0.1637 -0.1304 -0.9754>}
 dyn_curl_03 {parent=dyn_curl_03_base, loc=<0.0408 0 0>, rot=<0 0 0 1>}
 dyn_curl_04_base {parent=dyn_hair_02_l, loc=<0.6493 -0.322 1.4168>, rot=<-0.0552 -0.0416 0.3204 -0.9448>}
 dyn_curl_04 {parent=dyn_curl_04_base, loc=<0.0408 0 0>, rot=<0 0 0 1>}
 dyn_curl_05_base {parent=dyn_hair_02_l, loc=<-0.4526 0.1799 2.2896>, rot=<0.2668 0.3773 -0.1416 -0.8754>}
 dyn_curl_05 {parent=dyn_curl_05_base, loc=<0.0408 0 0>, rot=<0 0 0 1>}
 clavicle_l {parent=spine_05, loc=<4.3575 -0.1466 -2.3223>, rot=<0.0465 -0.7289 -0.009 -0.6829>}
 upperarm_l {parent=clavicle_l, loc=<15.5111 0 0>, rot=<0.0226 -0.3234 -0.0088 -0.9459>}
 extra_clavicle_l {parent=clavicle_l, loc=<0.0488 0 0>, rot=<0 0 0 1>}
 clavicle_r {parent=spine_05, loc=<4.3575 -0.1466 2.3223>, rot=<0.7289 0.0465 -0.6829 0.009>}
 upperarm_r {parent=clavicle_r, loc=<-15.5111 0 0>, rot=<0.0226 -0.3234 -0.0088 -0.9459>}
 extra_clavicle_r {parent=clavicle_r, loc=<-0.0723 0 0>, rot=<0 0 0 1>}
 HeadUp {type=Curve, loc=<0 0 0>, rot=<0 0 0 1>}
]
