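# Physics asset for the face accessory part of the RoundThumb_Nail male medium player cosmetic, covering three two-bone dynamic hair chains.
# Capsule bodies on head, neck and extra clavicle bones surround the center, left and right chains.
# Hair roots joint to the head with free angular motion, as do the head-neck links, while hair tips have tight swing and twist limits.
# Clavicle bodies are unconstrained; the right carries two overlapping capsules, one named for the left.
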PhysicsAsset /BRCosmetics/Characters/Player/Male/Medium/Bodies/M_MED_RoundThumb_Nail/Meshes/Parts/M_MED_RoundThumb_Nail_FaceAcc_Physics
bodies {
 dyn_hair_01 [{shapeType=Capsule, length=1, radius=1, shapeName=dyn_hair_01_capsule}]
 dyn_hair_01_l [{shapeType=Capsule, length=1, radius=1, shapeName=dyn_hair_01_l_capsule}]
 dyn_hair_01_r [{shapeType=Capsule, length=1, radius=1, shapeName=dyn_hair_01_r_capsule}]
 dyn_hair_02 [{shapeType=Capsule, length=4, radius=2, shapeName=dyn_hair_02_capsule, center=<1 0 0>}]
 dyn_hair_02_l [{shapeType=Capsule, length=4, radius=2, shapeName=dyn_hair_02_l_capsule, center=<1 0 0>}]
 dyn_hair_02_r [{shapeType=Capsule, length=4, radius=2, shapeName=dyn_hair_02_r_capsule, center=<1 0 0>}]
 extra_clavicle_l [{shapeType=Capsule, length=8, radius=6, shapeName=clavicle_l_capsule, center=<6 -3 5>, rotation=<rot 0 -90 90>}]
 extra_clavicle_r [
  {shapeType=Capsule, length=8, radius=6, shapeName=clavicle_l_capsule, center=<-6 3 -5>, rotation=<rot 0 -90 -90>}
  {shapeType=Capsule, length=8, radius=6, shapeName=clavicle_r_capsule, center=<-6 -1.8472 -5>, rotation=<rot 0 -90 -90>}
 ]
 head [{shapeType=Capsule, length=4.6647, radius=8, center=<0 2.3836 0>, rotation=<rot 90 0 0>}]
 neck_01 [{shapeType=Capsule, length=4.5, radius=6, shapeName=neck_01_capsule, rotation=<rot 0 -90.0011 90>}]
 neck_02 [{shapeType=Capsule, length=6, radius=5, shapeName=neck_02_capsule, rotation=<rot 0 -90.0004 90>}]
}
constraints [
 {
  bone1Name=head
  bone2Name=neck_02
  swing1LimitDegrees=45
  swing1Motion=Free
  swing2LimitDegrees=45
  swing2Motion=Free
  twistLimitDegrees=45
  twistMotion=Free
 }
 {
  bone1Name=neck_02
  bone2Name=neck_01
  swing1LimitDegrees=45
  swing1Motion=Free
  swing2LimitDegrees=45
  swing2Motion=Free
  twistLimitDegrees=45
  twistMotion=Free
 }
 {bone1Name=dyn_hair_02, bone2Name=dyn_hair_01, swing1LimitDegrees=25, swing2LimitDegrees=5, twistLimitDegrees=1}
 {bone1Name=dyn_hair_02_l, bone2Name=dyn_hair_01_l, swing1LimitDegrees=25, swing2LimitDegrees=5, twistLimitDegrees=1}
 {bone1Name=dyn_hair_02_r, bone2Name=dyn_hair_01_r, swing1LimitDegrees=35, swing2LimitDegrees=5, twistLimitDegrees=3}
 {
  bone1Name=dyn_hair_01
  bone2Name=head
  swing1LimitDegrees=45
  swing1Motion=Free
  swing2LimitDegrees=45
  swing2Motion=Free
  twistLimitDegrees=45
  twistMotion=Free
 }
 {
  bone1Name=dyn_hair_01_l
  bone2Name=head
  swing1LimitDegrees=45
  swing1Motion=Free
  swing2LimitDegrees=45
  swing2Motion=Free
  twistLimitDegrees=45
  twistMotion=Free
 }
 {
  bone1Name=dyn_hair_01_r
  bone2Name=head
  swing1LimitDegrees=45
  swing1Motion=Free
  swing2LimitDegrees=45
  swing2Motion=Free
  twistLimitDegrees=45
  twistMotion=Free
 }
]
