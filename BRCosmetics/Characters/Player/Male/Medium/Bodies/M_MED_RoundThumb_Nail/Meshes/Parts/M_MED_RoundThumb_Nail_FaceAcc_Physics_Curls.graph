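# Physics asset that simulates curly hair on the face-accessory mesh of the M_MED_RoundThumb_Nail male character.
# Three hair bodies hang off the head with tight angular limits, carrying sixteen curls in center, left and right groups.
# Each curl pairs a small base capsule, angularly free on its hair body, with a limited-swing curl capsule.
# Head, neck and clavicle capsules add body volumes, with clavicles left unconstrained.
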PhysicsAsset /BRCosmetics/Characters/Player/Male/Medium/Bodies/M_MED_RoundThumb_Nail/Meshes/Parts/M_MED_RoundThumb_Nail_FaceAcc_Physics_Curls
bodies {
 dyn_curl_01 [{shapeType=Capsule, length=2, radius=1, shapeName=dyn_curl_01_capsule, center=<1.9798 0 0>, rotation=<rot 90 0 0>}]
 dyn_curl_01_base [{shapeType=Capsule, length=0.5, radius=0.5, shapeName=dyn_curl_01_base_capsule}]
 dyn_curl_02 [{shapeType=Capsule, length=2, radius=1, shapeName=dyn_curl_02_capsule, center=<1.9798 0 0>, rotation=<rot 90 0 0>}]
 dyn_curl_02_base [{shapeType=Capsule, length=0.5, radius=0.5, shapeName=dyn_curl_02_base_capsule}]
 dyn_curl_03 [{shapeType=Capsule, length=2, radius=1, shapeName=dyn_curl_03_capsule, center=<1.9798 0 0>, rotation=<rot 90 0 0>}]
 dyn_curl_03_base [{shapeType=Capsule, length=0.5, radius=0.5, shapeName=dyn_curl_03_base_capsule}]
 dyn_curl_04 [{shapeType=Capsule, length=2, radius=1, shapeName=dyn_curl_04_capsule, center=<1.9798 0 0>, rotation=<rot 90 0 0>}]
 dyn_curl_04_base [{shapeType=Capsule, length=0.5, radius=0.5, shapeName=dyn_curl_04_base_capsule}]
 dyn_curl_05 [{shapeType=Capsule, length=2, radius=1, shapeName=dyn_curl_05_capsule, center=<1.9798 0 0>, rotation=<rot 90 0 0>}]
 dyn_curl_05_base [{shapeType=Capsule, length=0.5, radius=0.5, shapeName=dyn_curl_05_base_capsule}]
 dyn_curl_06 [{shapeType=Capsule, length=2, radius=1, shapeName=dyn_curl_06_capsule, center=<1.9798 0 0>, rotation=<rot 90 0 0>}]
 dyn_curl_06_base [{shapeType=Capsule, length=0.5, radius=0.5, shapeName=dyn_curl_06_base_capsule}]
 dyn_curl_07 [{shapeType=Capsule, length=2, radius=1, shapeName=dyn_curl_07_capsule, center=<1.9798 0 0>, rotation=<rot 90 0 0>}]
 dyn_curl_07_base [{shapeType=Capsule, length=0.5, radius=0.5, shapeName=dyn_curl_07_base_capsule}]
 dyn_curl_08 [{shapeType=Capsule, length=2, radius=1, shapeName=dyn_curl_08_capsule, center=<1.9798 0 0>, rotation=<rot 90 0 0>}]
 dyn_curl_08_base [{shapeType=Capsule, length=0.5, radius=0.5, shapeName=dyn_curl_08_base_capsule}]
 dyn_curl_09 [{shapeType=Capsule, length=2, radius=0.5, shapeName=dyn_curl_09_capsule, center=<1.9798 0 0>, rotation=<rot 90 0 0>}]
 dyn_curl_09_base [{shapeType=Capsule, length=0.5, radius=0.5, shapeName=dyn_curl_09_base_capsule}]
 dyn_curl_10 [{shapeType=Capsule, length=2, radius=1, shapeName=dyn_curl_10_capsule, center=<1.9798 0 0>, rotation=<rot 90 0 0>}]
 dyn_curl_10_base [{shapeType=Capsule, length=0.5, radius=0.5, shapeName=dyn_curl_10_base_capsule}]
 dyn_curl_11 [{shapeType=Capsule, length=1, radius=0.5, shapeName=dyn_curl_11_capsule, center=<1.2518 0 0>, rotation=<rot 90 0 0>}]
 dyn_curl_11_base [{shapeType=Capsule, length=0.5, radius=0.5, shapeName=dyn_curl_11_base_capsule}]
 dyn_curl_12 [{shapeType=Capsule, length=2, radius=1, shapeName=dyn_curl_12_capsule, center=<1.9798 0 0>, rotation=<rot 90 0 0>}]
 dyn_curl_12_base [{shapeType=Capsule, length=0.5, radius=0.5, shapeName=dyn_curl_12_base_capsule}]
 dyn_curl_13 [{shapeType=Capsule, length=3, radius=1, shapeName=dyn_curl_13_capsule, center=<1.2629 0.1264 0>, rotation=<rot 90 -55.3171 -55.3183>}]
 dyn_curl_13_base [{shapeType=Capsule, length=0.5, radius=0.5, shapeName=dyn_curl_13_base_capsule}]
 dyn_curl_14 [{shapeType=Capsule, length=2, radius=1, shapeName=dyn_curl_14_capsule, center=<1.9798 0 0>, rotation=<rot 90 0 0>}]
 dyn_curl_14_base [{shapeType=Capsule, length=0.5, radius=0.5, shapeName=dyn_curl_14_base_capsule}]
 dyn_curl_15 [{shapeType=Capsule, length=1, radius=0.5, shapeName=dyn_curl_15_capsule, center=<1.2518 0 0>, rotation=<rot 90 0 0>}]
 dyn_curl_15_base [{shapeType=Capsule, length=0.5, radius=0.5, shapeName=dyn_curl_15_base_capsule}]
 dyn_curl_16 [{shapeType=Capsule, length=0.6, radius=0.4, shapeName=dyn_curl_16_capsule, center=<1.4181 0 0>, rotation=<rot 90 0 0>}]
 dyn_curl_16_base [{shapeType=Capsule, length=0.5, radius=0.5, shapeName=dyn_curl_16_base_capsule}]
 dyn_hair_02 [{shapeType=Capsule, length=4, radius=2, shapeName=dyn_hair_02_capsule, center=<1 0 0>}]
 dyn_hair_02_l [{shapeType=Capsule, length=4, radius=2, shapeName=dyn_hair_02_l_capsule, center=<1 0 0>}]
 dyn_hair_02_r [{shapeType=Capsule, length=3, radius=1, shapeName=dyn_hair_02_r_capsule, center=<1 0 0>}]
 extra_clavicle_l [{shapeType=Capsule, length=8, radius=6, shapeName=clavicle_l_capsule, center=<6 -3 5>, rotation=<rot 0 -90 90>}]
 extra_clavicle_r [
  {shapeType=Capsule, length=8, radius=6, shapeName=clavicle_l_capsule, center=<-6 3 -5>, rotation=<rot 0 -90 -90>}
  {shapeType=Capsule, length=8, radius=6, shapeName=clavicle_r_capsule, center=<-6 -1.8472 -5>, rotation=<rot 0 -90 -90>}
 ]
 head [{shapeType=Capsule, length=4.6647, radius=8, center=<0 2.3836 0>, rotation=<rot 90 0 0>}]
 neck_01 [{shapeType=Capsule, length=4.5, radius=6, shapeName=neck_01_capsule, rotation=<rot 0 -90.0011 90>}]
 neck_02 [{shapeType=Capsule, length=6, radius=5, shapeName=neck_02_capsule, rotation=<rot 0 -90.0004 90>}]
}
constraints [
 {
  bone1Name=head
  bone2Name=neck_02
  swing1LimitDegrees=45
  swing1Motion=Free
  swing2LimitDegrees=45
  swing2Motion=Free
  twistLimitDegrees=45
  twistMotion=Free
 }
 {
  bone1Name=neck_02
  bone2Name=neck_01
  swing1LimitDegrees=45
  swing1Motion=Free
  swing2LimitDegrees=45
  swing2Motion=Free
  twistLimitDegrees=45
  twistMotion=Free
 }
 {bone1Name=dyn_hair_02, bone2Name=head, swing1LimitDegrees=25, swing2LimitDegrees=5, twistLimitDegrees=1}
 {bone1Name=dyn_hair_02_l, bone2Name=head, swing1LimitDegrees=25, swing2LimitDegrees=5, twistLimitDegrees=1}
 {bone1Name=dyn_hair_02_r, bone2Name=head, swing1LimitDegrees=25, swing2LimitDegrees=5, twistLimitDegrees=1}
 {bone1Name=dyn_curl_11, bone2Name=dyn_curl_11_base, swing1LimitDegrees=15, swing2LimitDegrees=15, twistLimitDegrees=5}
 {
  bone1Name=dyn_curl_06_base
  bone2Name=dyn_hair_02
  swing1LimitDegrees=45
  swing1Motion=Free
  swing2LimitDegrees=45
  swing2Motion=Free
  twistLimitDegrees=45
  twistMotion=Free
 }
 {bone1Name=dyn_curl_06, bone2Name=dyn_curl_06_base, swing1LimitDegrees=15, swing2LimitDegrees=15, twistLimitDegrees=5}
 {
  bone1Name=dyn_curl_07_base
  bone2Name=dyn_hair_02
  swing1LimitDegrees=45
  swing1Motion=Free
  swing2LimitDegrees=45
  swing2Motion=Free
  twistLimitDegrees=45
  twistMotion=Free
 }
 {bone1Name=dyn_curl_07, bone2Name=dyn_curl_07_base, swing1LimitDegrees=15, swing2LimitDegrees=15, twistLimitDegrees=5}
 {
  bone1Name=dyn_curl_08_base
  bone2Name=dyn_hair_02
  swing1LimitDegrees=45
  swing1Motion=Free
  swing2LimitDegrees=45
  swing2Motion=Free
  twistLimitDegrees=45
  twistMotion=Free
 }
 {bone1Name=dyn_curl_08, bone2Name=dyn_curl_08_base, swing1LimitDegrees=15, swing2LimitDegrees=15, twistLimitDegrees=5}
 {
  bone1Name=dyn_curl_09_base
  bone2Name=dyn_hair_02
  swing1LimitDegrees=45
  swing1Motion=Free
  swing2LimitDegrees=45
  swing2Motion=Free
  twistLimitDegrees=45
  twistMotion=Free
 }
 {bone1Name=dyn_curl_09, bone2Name=dyn_curl_09_base, swing1LimitDegrees=15, swing2LimitDegrees=15, twistLimitDegrees=5}
 {
  bone1Name=dyn_curl_10_base
  bone2Name=dyn_hair_02
  swing1LimitDegrees=45
  swing1Motion=Free
  swing2LimitDegrees=45
  swing2Motion=Free
  twistLimitDegrees=45
  twistMotion=Free
 }
 {bone1Name=dyn_curl_10, bone2Name=dyn_curl_10_base, swing1LimitDegrees=15, swing2LimitDegrees=15, twistLimitDegrees=5}
 {
  bone1Name=dyn_curl_11_base
  bone2Name=dyn_hair_02
  swing1LimitDegrees=45
  swing1Motion=Free
  swing2LimitDegrees=45
  swing2Motion=Free
  twistLimitDegrees=45
  twistMotion=Free
 }
 {bone1Name=dyn_curl_05, bone2Name=dyn_curl_05_base, swing1LimitDegrees=15, swing2LimitDegrees=15, twistLimitDegrees=5}
 {
  bone1Name=dyn_curl_05_base
  bone2Name=dyn_hair_02_l
  swing1LimitDegrees=45
  swing1Motion=Free
  swing2LimitDegrees=45
  swing2Motion=Free
  twistLimitDegrees=45
  twistMotion=Free
 }
 {bone1Name=dyn_curl_03, bone2Name=dyn_curl_03_base, swing1LimitDegrees=15, swing2LimitDegrees=15, twistLimitDegrees=5}
 {
  bone1Name=dyn_curl_01_base
  bone2Name=dyn_hair_02_l
  swing1LimitDegrees=45
  swing1Motion=Free
  swing2LimitDegrees=45
  swing2Motion=Free
  twistLimitDegrees=45
  twistMotion=Free
 }
 {bone1Name=dyn_curl_01, bone2Name=dyn_curl_01_base, swing1LimitDegrees=15, swing2LimitDegrees=15, twistLimitDegrees=5}
 {
  bone1Name=dyn_curl_02_base
  bone2Name=dyn_hair_02_l
  swing1LimitDegrees=45
  swing1Motion=Free
  swing2LimitDegrees=45
  swing2Motion=Free
  twistLimitDegrees=45
  twistMotion=Free
 }
 {bone1Name=dyn_curl_02, bone2Name=dyn_curl_02_base, swing1LimitDegrees=15, swing2LimitDegrees=15, twistLimitDegrees=5}
 {
  bone1Name=dyn_curl_03_base
  bone2Name=dyn_hair_02_l
  swing1LimitDegrees=45
  swing1Motion=Free
  swing2LimitDegrees=45
  swing2Motion=Free
  twistLimitDegrees=45
  twistMotion=Free
 }
 {
  bone1Name=dyn_curl_04_base
  bone2Name=dyn_hair_02_l
  swing1LimitDegrees=45
  swing1Motion=Free
  swing2LimitDegrees=45
  swing2Motion=Free
  twistLimitDegrees=45
  twistMotion=Free
 }
 {bone1Name=dyn_curl_04, bone2Name=dyn_curl_04_base, swing1LimitDegrees=15, swing2LimitDegrees=15, twistLimitDegrees=5}
 {
  bone1Name=dyn_curl_12_base
  bone2Name=dyn_hair_02_r
  swing1LimitDegrees=45
  swing1Motion=Free
  swing2LimitDegrees=45
  swing2Motion=Free
  twistLimitDegrees=45
  twistMotion=Free
 }
 {
  bone1Name=dyn_curl_13_base
  bone2Name=dyn_hair_02_r
  swing1LimitDegrees=45
  swing1Motion=Free
  swing2LimitDegrees=45
  swing2Motion=Free
  twistLimitDegrees=45
  twistMotion=Free
 }
 {bone1Name=dyn_curl_12, bone2Name=dyn_curl_12_base, swing1LimitDegrees=15, swing2LimitDegrees=15, twistLimitDegrees=5}
 {
  bone1Name=dyn_curl_14_base
  bone2Name=dyn_hair_02_r
  swing1LimitDegrees=45
  swing1Motion=Free
  swing2LimitDegrees=45
  swing2Motion=Free
  twistLimitDegrees=45
  twistMotion=Free
 }
 {bone1Name=dyn_curl_13, bone2Name=dyn_curl_13_base, swing1LimitDegrees=15, swing2LimitDegrees=15, twistLimitDegrees=5}
 {bone1Name=dyn_curl_14, bone2Name=dyn_curl_14_base, swing1LimitDegrees=15, swing2LimitDegrees=15, twistLimitDegrees=5}
 {
  bone1Name=dyn_curl_15_base
  bone2Name=dyn_hair_02_r
  swing1LimitDegrees=45
  swing1Motion=Free
  swing2LimitDegrees=45
  swing2Motion=Free
  twistLimitDegrees=45
  twistMotion=Free
 }
 {bone1Name=dyn_curl_15, bone2Name=dyn_curl_15_base, swing1LimitDegrees=15, swing2LimitDegrees=15, twistLimitDegrees=5}
 {
  bone1Name=dyn_curl_16_base
  bone2Name=dyn_hair_02_r
  swing1LimitDegrees=45
  swing1Motion=Free
  swing2LimitDegrees=45
  swing2Motion=Free
  twistLimitDegrees=45
  twistMotion=Free
 }
 {bone1Name=dyn_curl_16, bone2Name=dyn_curl_16_base, swing1LimitDegrees=15, swing2LimitDegrees=15, twistLimitDegrees=5}
]
