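# Post-copy animation Blueprint for the RoundThumb Nail character's face accessory, driving its dynamic hair and curls.
# It layers BodyMesh arms over the parent pose, then partly counters head rotation for the face-plate variant outside emotes.
# Constraints and a Control Rig feed head-space hair physics, whose bones weight the curl bases before a second curl simulation.
# A bool blend offers fixed hair-bone offsets instead, and customization loading detects the matching cape.
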AnimBlueprint /BRCosmetics/Characters/Player/Male/Medium/Bodies/M_MED_RoundThumb_Nail/Meshes/Parts/M_MED_RoundThumb_Nail_FaceAcc_Skeleton_AnimBlueprint
target_skeleton=&/M_MED_RoundThumb_Nail_FaceAcc_Skeleton
graphs [
 AnimGraph {nodes [
   n1 = "Transform (Modify) Bone - Bone: dyn_hair_01_r"(ComponentPose=Local To Component(LocalPose=Use cached pose 'pre'()), Alpha=1, Translation="0,0,0", Rotation="0,20,0", Scale="1,1,1") {
    rotationMode=BMM_Additive
    rotationSpace=BCS_BoneSpace
    alphaBoolBlend {blendOption=Linear}
    boneToModify {boneName=dyn_hair_01_r}
    rotation {yaw=20}
   }
   n2 = "Transform (Modify) Bone - Bone: dyn_hair_01"(ComponentPose=@n1, Alpha=1, Translation="0,0,0", Rotation="0,20,0", Scale="1,1,1") {
    rotationMode=BMM_Additive
    rotationSpace=BCS_BoneSpace
    alphaBoolBlend {blendOption=Linear}
    boneToModify {boneName=dyn_hair_01}
    rotation {yaw=20}
   }
   n3 = "Transform (Modify) Bone - Bone: dyn_hair_01_l"(ComponentPose=@n2, Alpha=1, Translation="0,0,0", Rotation="0,20,0", Scale="1,1,1") {
    rotationMode=BMM_Additive
    rotationSpace=BCS_BoneSpace
    alphaBoolBlend {blendOption=Linear}
    boneToModify {boneName=dyn_hair_01_l}
    rotation {yaw=20}
   }
   n4 = "Constraint - extra_clavicle_l (P)"(ComponentPose=Local To Component(LocalPose=Use cached pose 'pre'()), Alpha=0.85, ConstraintWeights_0=1) {
    alphaBoolBlend {blendOption=Linear}
    boneToModify {boneName=extra_clavicle_l}
    constraintSetup [{targetBone {boneName=spine_05}}]
    constraintWeights [1]
   }
   n5 = "Constraint - extra_clavicle_r (P)"(ComponentPose=@n4, Alpha=0.85, ConstraintWeights_0=1) {
    alphaBoolBlend {blendOption=Linear}
    boneToModify {boneName=extra_clavicle_r}
    constraintSetup [{targetBone {boneName=spine_05}}]
    constraintWeights [1]
   }
   n6 = Control Rig(Source=Component To Local(ComponentPose=@n5)) {
    alphaBoolBlend {blendOption=Linear}
    controlRigAssetReference {blueprintRigClass=&/M_MED_RoundThumb_Nail_FaceAcc_CtrlRig.M_MED_RoundThumb_Nail_FaceAcc_CtrlRig_C}
   }
   n7 = RigidBody!(ComponentPose=Local To Component(LocalPose=@n6), Alpha, OverrideWorldGravity:Vector="0,0,-980", ExternalForce) -> (Pose) {
    bOverrideWorldGravity=true
    overridePhysicsAsset=&/M_MED_RoundThumb_Nail_FaceAcc_Physics
    simulationSpace=BaseBoneSpace
    alphaBoolBlend {blendOption=Linear}
    baseBoneRef {boneName=head}
    componentAppliedLinearAccClamp=<2500 2500 2500>
    componentLinearAccScale=<0.35 0.35 0.35>
    componentLinearVelScale=<0.35 0.35 0.35>
    simSpaceSettings {maxAngularAcceleration=10, maxAngularVelocity=10, maxLinearAcceleration=10, maxLinearVelocity=10, worldAlpha=1}
   }
   n8 = "Constraint - dyn_curl_01_base (PPPP)"!(
    ComponentPose=@n7
    Alpha=1
    ConstraintWeights_0=0.4338
    ConstraintWeights_1:float=0.276
    ConstraintWeights_2:float=0.1866
    ConstraintWeights_3:float=0.1036
   ) -> (Pose) {
    alphaBoolBlend {blendOption=Linear}
    boneToModify {boneName=dyn_curl_01_base}
    constraintSetup [
     {targetBone {boneName=dyn_hair_02}}
     {targetBone {boneName=dyn_hair_02_l}}
     {targetBone {boneName=dyn_hair_01}}
     {targetBone {boneName=dyn_hair_01_l}}
    ]
    constraintWeights [1,1,1,1]
   }
   n9 = "Constraint - dyn_curl_02_base (PPPP)"!(
    ComponentPose=@n8
    Alpha=1
    ConstraintWeights_0=0.6963
    ConstraintWeights_1:float=0.2193
    ConstraintWeights_2:float=0.0444
    ConstraintWeights_3:float=0.0401
   ) -> (Pose) {
    alphaBoolBlend {blendOption=Linear}
    boneToModify {boneName=dyn_curl_02_base}
    constraintSetup [
     {targetBone {boneName=dyn_hair_02_l}}
     {targetBone {boneName=dyn_hair_02}}
     {targetBone {boneName=dyn_hair_01_l}}
     {targetBone {boneName=dyn_hair_01}}
    ]
    constraintWeights [1,1,1,1]
   }
   n10 = "Constraint - dyn_curl_06_base (PPP)"!(ComponentPose=@n9, Alpha=1, ConstraintWeights_0=0.6112, ConstraintWeights_1:float=0.337, ConstraintWeights_2:float=0.0517) -> (Pose) {
    alphaBoolBlend {blendOption=Linear}
    boneToModify {boneName=dyn_curl_06_base}
    constraintSetup [{targetBone {boneName=dyn_hair_02}}, {targetBone {boneName=dyn_hair_02_l}}, {targetBone {boneName=dyn_hair_01}}]
    constraintWeights [1,1,1]
   }
   n11 = "Constraint - dyn_curl_07_base (PPPP)"!(
    ComponentPose=@n10
    Alpha=1
    ConstraintWeights_0=0.7513
    ConstraintWeights_1:float=0.1685
    ConstraintWeights_2:float=0.0556
    ConstraintWeights_3:float=0.0245
   ) -> (Pose) {
    alphaBoolBlend {blendOption=Linear}
    boneToModify {boneName=dyn_curl_07_base}
    constraintSetup [
     {targetBone {boneName=dyn_hair_02}}
     {targetBone {boneName=dyn_hair_02_l}}
     {targetBone {boneName=dyn_hair_01}}
     {targetBone {boneName=dyn_hair_01_l}}
    ]
    constraintWeights [1,1,1,1]
   }
   n12 = "Constraint - dyn_curl_10_base (PPPP)"!(
    ComponentPose=@n11
    Alpha=1
    ConstraintWeights_0=0.2903
    ConstraintWeights_1:float=0.2679
    ConstraintWeights_2:float=0.2393
    ConstraintWeights_3:float=0.2024
   ) -> (Pose) {
    alphaBoolBlend {blendOption=Linear}
    boneToModify {boneName=dyn_curl_10_base}
    constraintSetup [
     {targetBone {boneName=dyn_hair_02}}
     {targetBone {boneName=dyn_hair_02_r}}
     {targetBone {boneName=dyn_hair_01}}
     {targetBone {boneName=dyn_hair_01_r}}
    ]
    constraintWeights [1,1,1,1]
   }
   n13 = "Constraint - dyn_curl_11_base (PPP)"!(ComponentPose=@n12, Alpha=1, ConstraintWeights_0=0.6053, ConstraintWeights_1:float=0.362, ConstraintWeights_2:float=0.0327) -> (Pose) {
    alphaBoolBlend {blendOption=Linear}
    boneToModify {boneName=dyn_curl_11_base}
    constraintSetup [{targetBone {boneName=dyn_hair_02_r}}, {targetBone {boneName=dyn_hair_02}}, {targetBone {boneName=dyn_hair_01_r}}]
    constraintWeights [1,1,1]
   }
   n14 = "Constraint - dyn_curl_12_base (PPP)"!(ComponentPose=@n13, Alpha=1, ConstraintWeights_0=0.8784, ConstraintWeights_1:float=0.0876, ConstraintWeights_2:float=0.034) -> (Pose) {
    alphaBoolBlend {blendOption=Linear}
    boneToModify {boneName=dyn_curl_12_base}
    constraintSetup [{targetBone {boneName=dyn_hair_02_r}}, {targetBone {boneName=dyn_hair_01_r}}, {targetBone {boneName=dyn_hair_02}}]
    constraintWeights [1,1,1]
   }
   n15 = "Constraint - dyn_curl_13_base (PPP)"!(ComponentPose=@n14, Alpha=1, ConstraintWeights_0=0.9441, ConstraintWeights_1:float=0.0346, ConstraintWeights_2:float=0.0213) -> (Pose) {
    alphaBoolBlend {blendOption=Linear}
    boneToModify {boneName=dyn_curl_13_base}
    constraintSetup [{targetBone {boneName=dyn_hair_02_r}}, {targetBone {boneName=dyn_hair_01_r}}, {targetBone {boneName=dyn_hair_02}}]
    constraintWeights [1,1,1]
   }
   n16 = RigidBody(ComponentPose=@n15, OverrideWorldGravity="0,0,-980") {
    bFreezeIncomingPoseOnStart=true
    bOverrideWorldGravity=true
    overridePhysicsAsset=&/M_MED_RoundThumb_Nail_FaceAcc_Physics_Curls
    simulationSpace=BaseBoneSpace
    alphaBoolBlend {blendOption=Linear}
    baseBoneRef {boneName=head}
    componentAppliedLinearAccClamp=<2500 2500 2500>
    componentLinearAccScale=<0.1 0.1 0.1>
    componentLinearVelScale=<0.1 0.1 0.1>
    simSpaceSettings {maxAngularAcceleration=10, maxAngularVelocity=10, maxLinearAcceleration=10, maxLinearVelocity=10, worldAlpha=1}
   }
   Output Pose(Result=Blend Poses by bool(BlendPose_0=Component To Local(ComponentPose=@n3), BlendPose_1=Component To Local(ComponentPose=@n16)) {blendType=Linear})
   n17 = Copy Pose From Mesh(SourceMeshComponent=Get BodyMesh() -> (BodyMesh:Skeletal Mesh Component Object Reference)) {bUseAttachedParent=true}
   n18 = Copy Pose From Mesh() {bUseAttachedParent=true}
   n19 = "Transform (Modify) Bone - Bone: head"(
    ComponentPose=Local To Component(LocalPose=Layered blend per bone(BasePose=@n18, BlendPoses_0=@n17) {layerSetup [{branchFilters [{boneName=upperarm_l}, {boneName=upperarm_r}]}]})
    Alpha=1
    Translation="0,0,0"
    Rotation=Get head rot() -> (head rot:Rotator)
    Scale="1,1,1"
   ) {rotationMode=BMM_Additive, rotationSpace=BCS_BoneSpace, alphaBoolBlend {blendOption=Linear}, boneToModify {boneName=head}}
   Save cached pose 'pre'(Pose=Component To Local(ComponentPose=@n19)) {cacheName=pre}
  ]}
 EventGraph {nodes [
   Event BlueprintUpdateAnimation()
   > n1 = Sequence()
   n2 = Get bIsPlayingEmote()
   > n3 = Branch(Condition=AND Boolean(A=Get FacePlate() -> (FacePlate:bool), B=NOT Boolean(A=@n2)))
   n4 = GetOwnerSkeletalMeshComponent()
   n5 = BreakRotator(InRot=BreakTransform(InTransform=GetDeltaTransformFromRefPose(self=@n4, BoneName=head, BaseName=neck_02)).Rotation)
   n6 = Get Y percentage() -> (Y percentage:float)
   n7 = Get X percentage() -> (X percentage:float)
   n8 = Get Z percentage() -> (Z percentage:float)
   > Set head rot(head rot:Rotator=MakeRotator(
     Roll=float * float!(A=@n5, B=-1, C:double=@n7) -> (ReturnValue)
     Pitch=float * float(A=@n5.Pitch, B=-1, C=@n6)
     Yaw=float * float(A=@n5.Yaw, B=-1, C=@n8)
    )) -> (Output_Get:Rotator)
   Set head rot(execute=@n3.else, head rot="0, 0, 0")
   Set AlphaOffDyn(execute=@n1.then_1, AlphaOffDyn:double=SelectFloat(A=GetCurveValue(CurveName=HeadUp), B=1, bPickA=Get bIsPlayingEmote())) -> (Output_Get:double)
   n9 = Event BlueprintInitializeAnimation()
   n10 = Sequence(execute=[@n14.then, @n9.then])
   > Set FacePlate(FacePlate:bool=CharacterItemDefinitionHasAnyMetaTag(Tags=Make Array("[0]"=(TagName="Cosmetics.Variant.Property.Stage1")))) -> (Output_Get:bool)
   n11 = Event OnCharacterCustomizationLoaded()
   n12 = "Get (a copy)"(Array:Array of Skeletal Mesh Component Budgeted Object References=@n11.CustomMeshParts, Dimension 1=3) -> (Output:Skeletal Mesh Component Budgeted Object Reference)
   Is Valid(exec=@n10.then_1, InputObject=@n12)
   > Set CapeNail(CapeNail:bool="Equal (Object)"(
     A=Get SkeletalMesh!(self:Skinned Mesh Component Object Reference=@n12) -> (SkeletalMesh:Skeletal Mesh Object Reference)
     B=Get CapeNailMesh() -> (CapeNailMesh:Skeletal Mesh Object Reference)
    )) -> (Output_Get:bool)
   n13 = "Get (a copy)"(Array=@n11.CustomMeshParts, Dimension 1=1)
   n14 = Set BodyMesh(execute=@n11.then, BodyMesh=@n13)
  ]}
]
variables [
 AlphaCapeNail {type=number, default=1}
 AlphaOffDyn {type=number, default=1}
 BodyMesh {type=SkeletalMeshComponent}
 CapeNail {type=boolean}
 CapeNailMesh {type=SkeletalMesh, default=/BRCosmetics/Accessories/FORT_Capes/Cape_M_MED_RoundThumb_Nail/Meshes/M_MED_RoundThumb_Nail_Cape}
 FacePlate {type=boolean}
 head rot {type=Rotator}
 X percentage {type=number}
 Y percentage {type=number}
 Z percentage {type=number, default=0.8}
]
